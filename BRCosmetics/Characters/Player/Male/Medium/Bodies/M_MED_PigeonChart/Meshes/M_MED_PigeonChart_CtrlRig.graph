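# Control Rig for the M_MED_PigeonChart male body cosmetic that poses its skirt bones to follow leg movement.
# A leg range-of-motion function turns spherical pose readers on thighs and calves, knee yaw and foot distances into optionally interpolated weights, scaled by a blend-out alpha.
# Curve-shaped averages add combined poses such as both thighs up.
# These weights drive mirrored left/right transform offsets on front, side and rear skirt bones.
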ControlRigBlueprint /BRCosmetics/Characters/Player/Male/Medium/Bodies/M_MED_PigeonChart/Meshes/M_MED_PigeonChart_CtrlRig
graphs [
 New Function_ContainedGraph {nodes [
   n1 = RigVMFunctionEntryNode()
   n2 = &RigVMFunction_Sequence(ExecuteContext=@n1.ExecuteContext)
   RigVMVariableNode(ExecuteContext=@n2.A, Value=@n1.Interp Results) {variable=Entry_Interp}
   > RigVMVariableNode(Value=@n1.Interp Output Speed) {variable=Entry_Interp Output Speed}
   > RigVMVariableNode(Value=@n1.Result Alpha) {variable=Entry_Alpha}
   RigUnit_GetTransform = &RigUnit_GetTransform(Item=(Type=Bone,Name="root"), bInitial=False)
   > RigVMVariableNode(Value=@RigUnit_GetTransform.Transform.Translation) {variable=root position}
   > RigVMFunction_Sequence = RigVMAggregateNode()
   n3 = &RigUnit_SphericalPoseReader(
    ActiveRegionScaleFactors=(PositiveWidth=3,NegativeWidth=1,PositiveHeight=1.2,NegativeHeight=1)
    ActiveRegionSize=0.1
    Debug=(bDrawDebug=false,bDraw2D=False,bDrawLocalAxes=False,DebugScale=25,DebugSegments=20,DebugThickness=0.25)
    DriverAxis=(X=1,Y=0,Z=0)
    DriverItem=(Type=Bone,Name="thigh_r")
    FalloffRegionScaleFactors=(PositiveWidth=1,NegativeWidth=1,PositiveHeight=0.75,NegativeHeight=0.75)
    FalloffSize=0.58
    FlipHeightScaling=true
    FlipWidthScaling=true
    OptionalParentItem=(Type=Bone,Name="None")
    RotationOffset=(X=-90,Y=0,Z=-15)
    ExecutePin=@RigVMFunction_Sequence.B
   )
   n4 = RigVMVariableNode() {variable=Entry_Interp Output Speed}
   n5 = &RigVMFunction_AlphaInterp(
    Scale=1
    bClampResult=true
    Value=@n3.OutputParam
    bInterpResult=RigVMVariableNode() {variable=Entry_Interp}.Value
    InterpSpeedDecreasing=@n4.Value
    InterpSpeedIncreasing=@n4.Value
   )
   > RigVMVariableNode(Value=@n5.Result) {variable=R Thigh Up}
   > n6 = &RigUnit_SphericalPoseReader(
    ActiveRegionScaleFactors=(PositiveWidth=3,NegativeWidth=1,PositiveHeight=1.3,NegativeHeight=1.2)
    ActiveRegionSize=0.1
    Debug=(bDrawDebug=false,bDraw2D=False,bDrawLocalAxes=False,DebugScale=25,DebugSegments=20,DebugThickness=0.25)
    DriverAxis=(X=1,Y=0,Z=0)
    DriverItem=(Type=Bone,Name="thigh_r")
    FalloffRegionScaleFactors=(PositiveWidth=1,NegativeWidth=1,PositiveHeight=0.5,NegativeHeight=0.5)
    FalloffSize=0.57
    FlipHeightScaling=true
    FlipWidthScaling=true
    OptionalParentItem=(Type=Bone,Name="None")
    RotationOffset=(X=-50,Y=5,Z=-12)
   )
   n7 = RigVMVariableNode() {variable=Entry_Interp Output Speed}
   n8 = &RigVMFunction_AlphaInterp(
    Scale=1
    bClampResult=true
    Value=@n6.OutputParam
    bInterpResult=RigVMVariableNode() {variable=Entry_Interp}.Value
    InterpSpeedDecreasing=@n7.Value
    InterpSpeedIncreasing=@n7.Value
   )
   > RigVMVariableNode(Value=@n8.Result) {variable=R Thigh Up Side Inbetween}
   > n9 = &RigUnit_SphericalPoseReader(
    ActiveRegionScaleFactors=(PositiveWidth=3,NegativeWidth=1,PositiveHeight=1,NegativeHeight=1)
    ActiveRegionSize=0.1
    Debug=(bDrawDebug=false,bDraw2D=False,bDrawLocalAxes=False,DebugScale=25,DebugSegments=20,DebugThickness=0.25)
    DriverAxis=(X=1,Y=0,Z=0)
    DriverItem=(Type=Bone,Name="thigh_r")
    FalloffRegionScaleFactors=(PositiveWidth=1,NegativeWidth=1,PositiveHeight=1,NegativeHeight=1)
    FalloffSize=0.5
    FlipHeightScaling=true
    FlipWidthScaling=true
    OptionalParentItem=(Type=Bone,Name="None")
    RotationOffset=(X=0,Y=0,Z=0)
   )
   n10 = RigVMVariableNode() {variable=Entry_Interp Output Speed}
   n11 = &RigVMFunction_AlphaInterp(
    Scale=1
    bClampResult=true
    Value=@n9.OutputParam
    bInterpResult=RigVMVariableNode() {variable=Entry_Interp}.Value
    InterpSpeedDecreasing=@n10.Value
    InterpSpeedIncreasing=@n10.Value
   )
   > RigVMVariableNode(Value=@n11.Result) {variable=R Thigh Side}
   > n12 = &RigUnit_SphericalPoseReader(
    ActiveRegionScaleFactors=(PositiveWidth=3,NegativeWidth=1,PositiveHeight=1.5,NegativeHeight=1.5)
    ActiveRegionSize=0.1
    Debug=(bDrawDebug=false,bDraw2D=False,bDrawLocalAxes=False,DebugScale=25,DebugSegments=20,DebugThickness=0.25)
    DriverAxis=(X=1,Y=0,Z=0)
    DriverItem=(Type=Bone,Name="thigh_r")
    FalloffRegionScaleFactors=(PositiveWidth=1,NegativeWidth=1,PositiveHeight=1,NegativeHeight=1)
    FalloffSize=0.5
    FlipHeightScaling=true
    FlipWidthScaling=true
    OptionalParentItem=(Type=Bone,Name="None")
    RotationOffset=(X=55,Y=0,Z=0)
   )
   n13 = RigVMVariableNode() {variable=Entry_Interp Output Speed}
   n14 = &RigVMFunction_AlphaInterp(
    Scale=1
    bClampResult=true
    Value=@n12.OutputParam
    bInterpResult=RigVMVariableNode() {variable=Entry_Interp}.Value
    InterpSpeedDecreasing=@n13.Value
    InterpSpeedIncreasing=@n13.Value
   )
   > RigVMVariableNode(Value=@n14.Result) {variable=R Thigh Rear Side Inbetween}
   > n15 = &RigUnit_SphericalPoseReader(
    ActiveRegionScaleFactors=(PositiveWidth=3,NegativeWidth=1,PositiveHeight=3,NegativeHeight=3)
    ActiveRegionSize=0.1
    Debug=(bDrawDebug=false,bDraw2D=False,bDrawLocalAxes=False,DebugScale=25,DebugSegments=20,DebugThickness=0.25)
    DriverAxis=(X=1,Y=0,Z=0)
    DriverItem=(Type=Bone,Name="thigh_r")
    FalloffRegionScaleFactors=(PositiveWidth=1,NegativeWidth=0.95,PositiveHeight=1,NegativeHeight=1)
    FalloffSize=0.5
    FlipHeightScaling=true
    FlipWidthScaling=true
    OptionalParentItem=(Type=Bone,Name="None")
    RotationOffset=(X=90,Y=0,Z=0)
   )
   n16 = RigVMVariableNode() {variable=Entry_Interp Output Speed}
   n17 = &RigVMFunction_AlphaInterp(
    Scale=1
    bClampResult=true
    Value=@n15.OutputParam
    bInterpResult=RigVMVariableNode() {variable=Entry_Interp}.Value
    InterpSpeedDecreasing=@n16.Value
    InterpSpeedIncreasing=@n16.Value
   )
   > RigVMVariableNode(Value=@n17.Result) {variable=R Thigh Rear}
   > n18 = &RigUnit_SphericalPoseReader(
    ActiveRegionScaleFactors=(PositiveWidth=3,NegativeWidth=1,PositiveHeight=1.2,NegativeHeight=1)
    ActiveRegionSize=0.1
    Debug=(bDrawDebug=false,bDraw2D=False,bDrawLocalAxes=False,DebugScale=25,DebugSegments=20,DebugThickness=0.25)
    DriverAxis=(X=1,Y=0,Z=0)
    DriverItem=(Type=Bone,Name="thigh_r")
    FalloffRegionScaleFactors=(PositiveWidth=1,NegativeWidth=1,PositiveHeight=0.3,NegativeHeight=0.3)
    FalloffSize=0.58
    FlipHeightScaling=true
    FlipWidthScaling=true
    OptionalParentItem=(Type=Bone,Name="None")
    RotationOffset=(X=-135,Y=0,Z=-15)
   )
   n19 = RigVMVariableNode() {variable=Entry_Interp Output Speed}
   n20 = &RigVMFunction_AlphaInterp(
    Scale=1
    bClampResult=true
    Value=@n18.OutputParam
    bInterpResult=RigVMVariableNode() {variable=Entry_Interp}.Value
    InterpSpeedDecreasing=@n19.Value
    InterpSpeedIncreasing=@n19.Value
   )
   > RigVMVariableNode(Value=@n20.Result) {variable=R Thigh Across}
   n21 = &RigUnit_SphericalPoseReader(
    ActiveRegionScaleFactors=(PositiveWidth=3,NegativeWidth=1,PositiveHeight=1.2,NegativeHeight=1)
    ActiveRegionSize=0.1
    Debug=(bDrawDebug=false,bDraw2D=False,bDrawLocalAxes=False,DebugScale=25,DebugSegments=20,DebugThickness=0.25)
    DriverAxis=(X=-1,Y=0,Z=0)
    DriverItem=(Type=Bone,Name="thigh_l")
    FalloffRegionScaleFactors=(PositiveWidth=1,NegativeWidth=1,PositiveHeight=0.75,NegativeHeight=0.75)
    FalloffSize=0.58
    FlipHeightScaling=False
    FlipWidthScaling=False
    OptionalParentItem=(Type=Bone,Name="None")
    RotationOffset=(X=90,Y=0,Z=-15)
    ExecutePin=@RigVMFunction_Sequence.A
   )
   n22 = RigVMVariableNode() {variable=Entry_Interp Output Speed}
   n23 = &RigVMFunction_AlphaInterp(
    Scale=1
    bClampResult=true
    Value=@n21.OutputParam
    bInterpResult=RigVMVariableNode() {variable=Entry_Interp}.Value
    InterpSpeedDecreasing=@n22.Value
    InterpSpeedIncreasing=@n22.Value
   )
   > RigVMVariableNode(Value=@n23.Result) {variable=L Thigh Up}
   > n24 = &RigUnit_SphericalPoseReader(
    ActiveRegionScaleFactors=(PositiveWidth=3,NegativeWidth=1,PositiveHeight=1.3,NegativeHeight=1.2)
    ActiveRegionSize=0.1
    Debug=(bDrawDebug=false,bDraw2D=False,bDrawLocalAxes=False,DebugScale=25,DebugSegments=20,DebugThickness=0.25)
    DriverAxis=(X=-1,Y=0,Z=0)
    DriverItem=(Type=Bone,Name="thigh_l")
    FalloffRegionScaleFactors=(PositiveWidth=1,NegativeWidth=1,PositiveHeight=0.5,NegativeHeight=0.5)
    FalloffSize=0.57
    FlipHeightScaling=False
    FlipWidthScaling=False
    OptionalParentItem=(Type=Bone,Name="None")
    RotationOffset=(X=130,Y=5,Z=-12)
   )
   n25 = RigVMVariableNode() {variable=Entry_Interp Output Speed}
   n26 = &RigVMFunction_AlphaInterp(
    Scale=1
    bClampResult=true
    Value=@n24.OutputParam
    bInterpResult=RigVMVariableNode() {variable=Entry_Interp}.Value
    InterpSpeedDecreasing=@n25.Value
    InterpSpeedIncreasing=@n25.Value
   )
   > RigVMVariableNode(Value=@n26.Result) {variable=L Thigh Up Side Inbetween}
   > n27 = &RigUnit_SphericalPoseReader(
    ActiveRegionScaleFactors=(PositiveWidth=3,NegativeWidth=1,PositiveHeight=1,NegativeHeight=1)
    ActiveRegionSize=0.1
    Debug=(bDrawDebug=false,bDraw2D=False,bDrawLocalAxes=False,DebugScale=25,DebugSegments=20,DebugThickness=0.25)
    DriverAxis=(X=-1,Y=0,Z=0)
    DriverItem=(Type=Bone,Name="thigh_l")
    FalloffRegionScaleFactors=(PositiveWidth=1,NegativeWidth=1,PositiveHeight=1,NegativeHeight=1)
    FalloffSize=0.5
    FlipHeightScaling=False
    FlipWidthScaling=False
    OptionalParentItem=(Type=Bone,Name="None")
    RotationOffset=(X=180,Y=0,Z=0)
   )
   n28 = RigVMVariableNode() {variable=Entry_Interp Output Speed}
   n29 = &RigVMFunction_AlphaInterp(
    Scale=1
    bClampResult=true
    Value=@n27.OutputParam
    bInterpResult=RigVMVariableNode() {variable=Entry_Interp}.Value
    InterpSpeedDecreasing=@n28.Value
    InterpSpeedIncreasing=@n28.Value
   )
   > RigVMVariableNode(Value=@n29.Result) {variable=L Thigh Side}
   > n30 = &RigUnit_SphericalPoseReader(
    ActiveRegionScaleFactors=(PositiveWidth=3,NegativeWidth=1,PositiveHeight=1.5,NegativeHeight=1.5)
    ActiveRegionSize=0.1
    Debug=(bDrawDebug=false,bDraw2D=False,bDrawLocalAxes=False,DebugScale=25,DebugSegments=20,DebugThickness=0.25)
    DriverAxis=(X=-1,Y=0,Z=0)
    DriverItem=(Type=Bone,Name="thigh_l")
    FalloffRegionScaleFactors=(PositiveWidth=1,NegativeWidth=1,PositiveHeight=1,NegativeHeight=1)
    FalloffSize=0.5
    FlipHeightScaling=False
    FlipWidthScaling=False
    OptionalParentItem=(Type=Bone,Name="None")
    RotationOffset=(X=-125,Y=0,Z=0)
   )
   n31 = RigVMVariableNode() {variable=Entry_Interp Output Speed}
   n32 = &RigVMFunction_AlphaInterp(
    Scale=1
    bClampResult=true
    Value=@n30.OutputParam
    bInterpResult=RigVMVariableNode() {variable=Entry_Interp}.Value
    InterpSpeedDecreasing=@n31.Value
    InterpSpeedIncreasing=@n31.Value
   )
   > RigVMVariableNode(Value=@n32.Result) {variable=L Thigh Rear Side Inbetween}
   > n33 = &RigUnit_SphericalPoseReader(
    ActiveRegionScaleFactors=(PositiveWidth=3,NegativeWidth=1,PositiveHeight=3,NegativeHeight=3)
    ActiveRegionSize=0.1
    Debug=(bDrawDebug=false,bDraw2D=False,bDrawLocalAxes=False,DebugScale=25,DebugSegments=20,DebugThickness=0.25)
    DriverAxis=(X=-1,Y=0,Z=0)
    DriverItem=(Type=Bone,Name="thigh_l")
    FalloffRegionScaleFactors=(PositiveWidth=1,NegativeWidth=0.95,PositiveHeight=1,NegativeHeight=1)
    FalloffSize=0.5
    FlipHeightScaling=False
    FlipWidthScaling=False
    OptionalParentItem=(Type=Bone,Name="None")
    RotationOffset=(X=-90,Y=0,Z=0)
   )
   n34 = RigVMVariableNode() {variable=Entry_Interp Output Speed}
   n35 = &RigVMFunction_AlphaInterp(
    Scale=1
    bClampResult=true
    Value=@n33.OutputParam
    bInterpResult=RigVMVariableNode() {variable=Entry_Interp}.Value
    InterpSpeedDecreasing=@n34.Value
    InterpSpeedIncreasing=@n34.Value
   )
   > RigVMVariableNode(Value=@n35.Result) {variable=L Thigh Rear}
   > n36 = &RigUnit_SphericalPoseReader(
    ActiveRegionScaleFactors=(PositiveWidth=3,NegativeWidth=1,PositiveHeight=1.2,NegativeHeight=1)
    ActiveRegionSize=0.1
    Debug=(bDrawDebug=false,bDraw2D=False,bDrawLocalAxes=False,DebugScale=25,DebugSegments=20,DebugThickness=0.25)
    DriverAxis=(X=-1,Y=0,Z=0)
    DriverItem=(Type=Bone,Name="thigh_l")
    FalloffRegionScaleFactors=(PositiveWidth=1,NegativeWidth=1,PositiveHeight=0.3,NegativeHeight=0.3)
    FalloffSize=0.58
    FlipHeightScaling=False
    FlipWidthScaling=False
    OptionalParentItem=(Type=Bone,Name="None")
    RotationOffset=(X=45,Y=0,Z=-15)
   )
   n37 = RigVMVariableNode() {variable=Entry_Interp Output Speed}
   n38 = &RigVMFunction_AlphaInterp(
    Scale=1
    bClampResult=true
    Value=@n36.OutputParam
    bInterpResult=RigVMVariableNode() {variable=Entry_Interp}.Value
    InterpSpeedDecreasing=@n37.Value
    InterpSpeedIncreasing=@n37.Value
   )
   > RigVMVariableNode(Value=@n38.Result) {variable=L Thigh Across}
   Get Delta Ref = RigVMFunctionReferenceNode(Item=(Type=Bone,Name="calf_l"), ExecuteContext=@RigVMFunction_Sequence.C)
   n39 = RigVMVariableNode() {variable=Entry_Interp Output Speed}
   n40 = &RigVMFunction_AlphaInterp(
    InRange=(Min=0,Max=140)
    Scale=1
    bClampResult=true
    bMapRange=true
    Value=@"Get Delta Ref".Result.Rotation.Yaw
    bInterpResult=RigVMVariableNode() {variable=Entry_Interp}.Value
    InterpSpeedDecreasing=@n39.Value
    InterpSpeedIncreasing=@n39.Value
   )
   > RigVMVariableNode(Value=@n40.Result) {variable=L Knee Bend}
   > n41 = &RigUnit_SphericalPoseReader(
    ActiveRegionScaleFactors=(PositiveWidth=3,NegativeWidth=0.5,PositiveHeight=5,NegativeHeight=5)
    ActiveRegionSize=0.1
    Debug=(bDrawDebug=false,bDraw2D=False,bDrawLocalAxes=False,DebugScale=25,DebugSegments=20,DebugThickness=0.25)
    DriverAxis=(X=-1,Y=0,Z=0)
    DriverItem=(Type=Bone,Name="calf_l")
    FalloffRegionScaleFactors=(PositiveWidth=0.2,NegativeWidth=0.92,PositiveHeight=1,NegativeHeight=1)
    FalloffSize=0.7
    FlipHeightScaling=False
    FlipWidthScaling=False
    OptionalParentItem=(Type=Bone,Name="thigh_l")
    RotationOffset=(X=-90,Y=0,Z=27)
   )
   Subtract = &RigVMFunction_MathDoubleSub(A=1, B=RigVMVariableNode() {variable=L Thigh Up}.Value)
   Multiply = &RigVMFunction_MathDoubleMul(A=@n41.OutputParam, B=@Subtract.Result)
   > RigVMVariableNode(Value=@Multiply.Result) {variable=L Thigh Rear Up Knee Bend}
   n42 = RigVMVariableNode() {variable=L Thigh Up}
   n43 = RigVMVariableNode() {variable=L Knee Bend}
   n44 = &RigVMFunction_AnimEvalRichCurve(
    Curve=(EditorCurveData=(Keys=((InterpMode=RCIM_Cubic,TangentMode=RCTM_User,ArriveTangent=0.002,LeaveTangent=0.002),(InterpMode=RCIM_Cubic,TangentMode=RCTM_User,Time=0.5,Value=0.0387,ArriveTangent=0.225,LeaveTangent=0.225),(Time=1,Value=1))))
    SourceMinimum=0
    TargetMinimum=0
    Value=&RigVMFunction_MathDoubleMul(B=0.5, A=&RigVMFunction_MathDoubleAdd(A=@n42.Value, B=@n43.Value).Result).Result
   )
   > RigVMVariableNode(Value=@n44.Result) {variable=L Thigh Up Knee Bend}
   n45 = &RigUnit_GetRelativeTransformForItem(Child=(Type=Bone,Name="foot_l"), Parent=(Type=Bone,Name="root"), bChildInitial=False, bParentInitial=False)
   n46 = &RigUnit_GetRelativeTransformForItem(Child=(Type=Bone,Name="pelvis"), Parent=(Type=Bone,Name="root"), bChildInitial=False, bParentInitial=False)
   > RigVMVariableNode(Value=&RigVMFunction_MathDoubleRemap(
     SourceMaximum=45
     SourceMinimum=23
     TargetMaximum=0
     TargetMinimum=1
     Value=&RigVMFunction_MathVectorDistance(A=@n45.RelativeTransform.Translation, B=@n46.RelativeTransform.Translation).Result
    ).Result) {variable=L Foot Close to Pelvis}
   Get Delta Ref_1 = RigVMFunctionReferenceNode(Item=(Type=Bone,Name="calf_r"), ExecuteContext=@RigVMFunction_Sequence.E)
   n47 = RigVMVariableNode() {variable=Entry_Interp Output Speed}
   n48 = &RigVMFunction_AlphaInterp(
    InRange=(Min=0,Max=140)
    Scale=1
    bClampResult=true
    bMapRange=true
    Value=@"Get Delta Ref_1".Result.Rotation.Yaw
    bInterpResult=RigVMVariableNode() {variable=Entry_Interp}.Value
    InterpSpeedDecreasing=@n47.Value
    InterpSpeedIncreasing=@n47.Value
   )
   > RigVMVariableNode(Value=@n48.Result) {variable=R Knee Bend}
   > n49 = &RigUnit_SphericalPoseReader(
    ActiveRegionScaleFactors=(PositiveWidth=3,NegativeWidth=0.5,PositiveHeight=5,NegativeHeight=5)
    ActiveRegionSize=0.1
    Debug=(bDrawDebug=false,bDraw2D=False,bDrawLocalAxes=False,DebugScale=25,DebugSegments=20,DebugThickness=0.25)
    DriverAxis=(X=1,Y=0,Z=0)
    DriverItem=(Type=Bone,Name="calf_r")
    FalloffRegionScaleFactors=(PositiveWidth=0.2,NegativeWidth=0.92,PositiveHeight=1,NegativeHeight=1)
    FalloffSize=0.7
    FlipHeightScaling=true
    FlipWidthScaling=true
    OptionalParentItem=(Type=Bone,Name="thigh_r")
    RotationOffset=(X=90,Y=0,Z=27)
   )
   n50 = RigVMVariableNode() {variable=R Thigh Up}
   > RigVMVariableNode(Value=&RigVMFunction_MathDoubleMul(A=@n49.OutputParam, B=&RigVMFunction_MathDoubleSub(A=1, B=@n50.Value).Result).Result) {variable=R Thigh Rear Up Knee Bend}
   n51 = RigVMVariableNode() {variable=R Thigh Up}
   n52 = RigVMVariableNode() {variable=R Knee Bend}
   n53 = &RigVMFunction_AnimEvalRichCurve(
    Curve=(EditorCurveData=(Keys=((InterpMode=RCIM_Cubic,TangentMode=RCTM_User,ArriveTangent=0.002,LeaveTangent=0.002),(InterpMode=RCIM_Cubic,TangentMode=RCTM_User,Time=0.5,Value=0.02,ArriveTangent=0.2592,LeaveTangent=0.2592),(Time=1,Value=1))))
    SourceMinimum=0
    TargetMinimum=0
    Value=&RigVMFunction_MathDoubleMul(B=0.5, A=&RigVMFunction_MathDoubleAdd(A=@n51.Value, B=@n52.Value).Result).Result
   )
   > RigVMVariableNode(Value=@n53.Result) {variable=R Thigh Up Knee Bend}
   n54 = &RigUnit_GetRelativeTransformForItem(Child=(Type=Bone,Name="pelvis"), Parent=(Type=Bone,Name="root"), bChildInitial=False, bParentInitial=False)
   n55 = &RigUnit_GetRelativeTransformForItem(Child=(Type=Bone,Name="foot_r"), Parent=(Type=Bone,Name="root"), bChildInitial=False, bParentInitial=False)
   > RigVMVariableNode(Value=&RigVMFunction_MathDoubleRemap(
     SourceMaximum=45
     SourceMinimum=23
     TargetMaximum=0
     TargetMinimum=1
     Value=&RigVMFunction_MathVectorDistance(B=@n54.RelativeTransform.Translation, A=@n55.RelativeTransform.Translation).Result
    ).Result) {variable=R Foot Close to Pelvis}
   n56 = RigVMVariableNode() {variable=L Thigh Up}
   n57 = RigVMVariableNode() {variable=R Thigh Up}
   RigVMFunction_AnimEvalRichCurve = &RigVMFunction_AnimEvalRichCurve(
    Curve=(EditorCurveData=(Keys=((InterpMode=RCIM_Cubic,TangentMode=RCTM_User,ArriveTangent=0.002,LeaveTangent=0.002),(InterpMode=RCIM_Cubic,TangentMode=RCTM_User,Time=0.3991,Value=0.0766,ArriveTangent=0.6213,LeaveTangent=0.6213),(Time=1,Value=1))))
    SourceMinimum=0
    TargetMinimum=0
    Value=&RigVMFunction_MathDoubleMul(B=0.5, A=&RigVMFunction_MathDoubleAdd(A=@n56.Value, B=@n57.Value).Result).Result
   )
   RigVMVariableNode(Value=@RigVMFunction_AnimEvalRichCurve.Result, ExecuteContext=@RigVMFunction_Sequence.D) {variable=Both Thighs Up}
   n58 = RigVMVariableNode() {variable=L Thigh Rear}
   n59 = RigVMVariableNode() {variable=R Thigh Rear}
   n60 = &RigVMFunction_AnimEvalRichCurve(
    Curve=(EditorCurveData=(Keys=((InterpMode=RCIM_Cubic,TangentMode=RCTM_User,ArriveTangent=0.002,LeaveTangent=0.002),(InterpMode=RCIM_Cubic,TangentMode=RCTM_User,Time=0.3991,Value=0.0766,ArriveTangent=0.6213,LeaveTangent=0.6213),(Time=1,Value=1))))
    SourceMinimum=0
    TargetMinimum=0
    Value=&RigVMFunction_MathDoubleMul(B=0.5, A=&RigVMFunction_MathDoubleAdd(A=@n58.Value, B=@n59.Value).Result).Result
   )
   > RigVMVariableNode(Value=@n60.Result) {variable=Both Thighs Rear}
   n61 = RigVMVariableNode() {variable=L Thigh Side}
   n62 = RigVMVariableNode() {variable=R Thigh Up}
   n63 = &RigVMFunction_AnimEvalRichCurve(
    Curve=(EditorCurveData=(Keys=((InterpMode=RCIM_Cubic,TangentMode=RCTM_User,ArriveTangent=0.002,LeaveTangent=0.002),(InterpMode=RCIM_Cubic,TangentMode=RCTM_User,Time=0.3991,Value=0.0766,ArriveTangent=0.6213,LeaveTangent=0.6213),(Time=1,Value=1))))
    SourceMinimum=0
    TargetMinimum=0
    Value=&RigVMFunction_MathDoubleMul(B=0.5, A=&RigVMFunction_MathDoubleAdd(A=@n61.Value, B=@n62.Value).Result).Result
   )
   > RigVMVariableNode(Value=@n63.Result) {variable=L Thigh Side R Thigh Up}
   n64 = RigVMVariableNode() {variable=R Thigh Side}
   n65 = RigVMVariableNode() {variable=L Thigh Up}
   n66 = &RigVMFunction_AnimEvalRichCurve(
    Curve=(EditorCurveData=(Keys=((InterpMode=RCIM_Cubic,TangentMode=RCTM_User,ArriveTangent=0.002,LeaveTangent=0.002),(InterpMode=RCIM_Cubic,TangentMode=RCTM_User,Time=0.3991,Value=0.0766,ArriveTangent=0.6213,LeaveTangent=0.6213),(Time=1,Value=1))))
    SourceMinimum=0
    TargetMinimum=0
    Value=&RigVMFunction_MathDoubleMul(B=0.5, A=&RigVMFunction_MathDoubleAdd(A=@n64.Value, B=@n65.Value).Result).Result
   )
   > RigVMVariableNode(Value=@n66.Result) {variable=R Thigh Side L Thigh Up}
   n67 = &RigUnit_GetRelativeTransformForItem(Child=(Type=Bone,Name="foot_l"), Parent=(Type=Bone,Name="root"), bChildInitial=False, bParentInitial=False)
   n68 = /Script/RigVM.RigVMFunction_MathVectorAdd(A=(X=-35,Y=-20,Z=0), A.Z=@n67.RelativeTransform.Translation.Z, B=RigVMVariableNode() {variable=root position}.Value) {is=struct RigVMUnitNode, template="Add::Execute(in A,in B,out Result)"}
   n69 = &RigVMFunction_AlphaInterp(
    ClampMax=0
    ClampMin=1
    InRange=(Min=10,Max=28)
    InterpSpeedDecreasing=10
    InterpSpeedIncreasing=10
    OutRange=(Min=1,Max=0)
    Scale=1
    bClampResult=false
    bMapRange=false
    Value=&RigVMFunction_MathDoubleRemap(
     SourceMaximum=18
     SourceMinimum=7
     TargetMaximum=0
     TargetMinimum=1
     Value=&RigVMFunction_MathVectorDistance(B=@n68.Result, A=@n67.RelativeTransform.Translation).Result
    ).Result
   )
   RigVMVariableNode(Value=@n69.Result, ExecuteContext=@RigVMFunction_Sequence.F) {variable=L Leg Across Side}
   n70 = &RigUnit_GetRelativeTransformForItem(Child=(Type=Bone,Name="foot_r"), Parent=(Type=Bone,Name="root"), bChildInitial=False, bParentInitial=False)
   n71 = &RigVMFunction_MathVectorAdd(A=(X=35,Y=-20,Z=-20), A.Z=@n70.RelativeTransform.Translation.Z, B=RigVMVariableNode() {variable=root position}.Value) {template="Add::Execute(in A,in B,out Result)"}
   AlphaInterp = &RigVMFunction_AlphaInterp(
    ClampMax=0
    ClampMin=1
    InRange=(Min=10,Max=28)
    InterpSpeedDecreasing=10
    InterpSpeedIncreasing=10
    OutRange=(Min=1,Max=0)
    Scale=1
    bClampResult=false
    bMapRange=false
    Value=&RigVMFunction_MathDoubleRemap(
     SourceMaximum=28
     SourceMinimum=10
     TargetMaximum=0
     TargetMinimum=1
     Value=&RigVMFunction_MathVectorDistance(B=@n71.Result, A=@n70.RelativeTransform.Translation).Result
    ).Result
   )
   > RigVMVariableNode(Value=@AlphaInterp.Result) {variable=R Leg Across Side}
   n72 = RigVMVariableNode() {variable=Entry_Alpha}
   RigVMFunctionReturnNode(
    L Thigh Raised Fwd=&RigVMFunction_MathDoubleMul(A=RigVMVariableNode() {variable=L Thigh Up}.Value, B=@n72.Value).Result
    R Thigh Raised Fwd=&RigVMFunction_MathDoubleMul(A=RigVMVariableNode() {variable=R Thigh Up}.Value, B=@n72.Value).Result
    R Leg Across L Leg Rear=&RigVMFunction_MathDoubleMul(A=RigVMVariableNode() {variable=R Leg Across Rear}.Value, B=@n72.Value).Result
    L Leg Across R Leg Side=&RigVMFunction_MathDoubleMul(A=RigVMVariableNode() {variable=L Leg Across Side}.Value, B=@n72.Value).Result
    L Knee=&RigVMFunction_MathDoubleMul(A=RigVMVariableNode() {variable=L Knee Bend}.Value, B=@n72.Value).Result
    Both Thighs Raised=&RigVMFunction_MathDoubleMul(B=@n72.Value, A=RigVMVariableNode() {variable=Both Thighs Up}.Value).Result
    L Foot Near Pelvis=&RigVMFunction_MathDoubleMul(B=@n72.Value, A=RigVMVariableNode() {variable=L Foot Close to Pelvis}.Value).Result
    L Thigh Side R Thigh Raise=&RigVMFunction_MathDoubleMul(B=@n72.Value, A=RigVMVariableNode() {variable=L Thigh Side R Thigh Up}.Value).Result
    R Knee=&RigVMFunction_MathDoubleMul(A=RigVMVariableNode() {variable=R Knee Bend}.Value, B=@n72.Value).Result
    L Thigh Rear Knee Bend=&RigVMFunction_MathDoubleMul(A=RigVMVariableNode() {variable=L Thigh Rear Up Knee Bend}.Value, B=@n72.Value).Result
    L Thigh Fwd Knee Bend=&RigVMFunction_MathDoubleMul(B=@n72.Value, A=RigVMVariableNode() {variable=L Thigh Up Knee Bend}.Value).Result
    R Thigh Rear Knee Bend=&RigVMFunction_MathDoubleMul(A=RigVMVariableNode() {variable=R Thigh Rear Up Knee Bend}.Value, B=@n72.Value).Result
    R Thigh Fwd Knee Bend=&RigVMFunction_MathDoubleMul(B=@n72.Value, A=RigVMVariableNode() {variable=R Thigh Up Knee Bend}.Value).Result
    Both Thighs Back=&RigVMFunction_MathDoubleMul(B=@n72.Value, A=RigVMVariableNode() {variable=Both Thighs Rear}.Value).Result
    R Foot Near Pelvis=&RigVMFunction_MathDoubleMul(B=@n72.Value, A=RigVMVariableNode() {variable=R Foot Close to Pelvis}.Value).Result
    R Thigh Side L Thigh Raise=&RigVMFunction_MathDoubleMul(B=@n72.Value, A=RigVMVariableNode() {variable=R Thigh Side L Thigh Up}.Value).Result
    L Thigh Raised Fwd Side Inbetween=&RigVMFunction_MathDoubleMul(A=RigVMVariableNode() {variable=L Thigh Up Side Inbetween}.Value, B=@n72.Value).Result
    R Thigh Raised Fwd Side Inbetween=&RigVMFunction_MathDoubleMul(A=RigVMVariableNode() {variable=R Thigh Up Side Inbetween}.Value, B=@n72.Value).Result
    R Thigh Raised Side=&RigVMFunction_MathDoubleMul(A=RigVMVariableNode() {variable=R Thigh Side}.Value, B=@n72.Value).Result
    L Thigh Raised Rear=&RigVMFunction_MathDoubleMul(A=RigVMVariableNode() {variable=L Thigh Rear}.Value, B=@n72.Value).Result
    L Leg Across R Leg Front=&RigVMFunction_MathDoubleMul(A=RigVMVariableNode() {variable=L Leg Across Front}.Value, B=@n72.Value).Result
    L Leg Across R Leg Rear=&RigVMFunction_MathDoubleMul(A=RigVMVariableNode() {variable=L Leg Across Rear}.Value, B=@n72.Value).Result
    R Leg Across L Leg Side=&RigVMFunction_MathDoubleMul(A=RigVMVariableNode() {variable=R Leg Across Side}.Value, B=@n72.Value).Result
    L Thigh Raised Side=&RigVMFunction_MathDoubleMul(A=RigVMVariableNode() {variable=L Thigh Side}.Value, B=@n72.Value).Result
    L Thigh Raised Side Rear Inbetween=&RigVMFunction_MathDoubleMul(A=RigVMVariableNode() {variable=L Thigh Rear Side Inbetween}.Value, B=@n72.Value).Result
    R Thigh Raised Side Rear Inbetween=&RigVMFunction_MathDoubleMul(A=RigVMVariableNode() {variable=R Thigh Rear Side Inbetween}.Value, B=@n72.Value).Result
    R Thigh Raised Rear=&RigVMFunction_MathDoubleMul(A=RigVMVariableNode() {variable=R Thigh Rear}.Value, B=@n72.Value).Result
    L Thigh Across R=&RigVMFunction_MathDoubleMul(A=RigVMVariableNode() {variable=L Thigh Across}.Value, B=@n72.Value).Result
    R Thigh Across L=&RigVMFunction_MathDoubleMul(A=RigVMVariableNode() {variable=R Thigh Across}.Value, B=@n72.Value).Result
    R Leg Across L Leg Front=&RigVMFunction_MathDoubleMul(A=RigVMVariableNode() {variable=R Leg Across Front}.Value, B=@n72.Value).Result
    ExecuteContext=@n2.B
   )
   n73 = &RigUnit_GetRelativeTransformForItem(Child=(Type=Bone,Name="foot_r"), Parent=(Type=Bone,Name="root"), bChildInitial=False, bParentInitial=False)
   n74 = &RigVMFunction_MathVectorAdd(B=(X=18,Y=10,Z=7), B.Z=@n73.RelativeTransform.Translation.Z, A=RigVMVariableNode() {variable=root position}.Value) {template="Add::Execute(in A,in B,out Result)"}
   n75 = &RigVMFunction_AlphaInterp(
    ClampMax=0
    ClampMin=1
    InRange=(Min=10,Max=28)
    InterpSpeedDecreasing=10
    InterpSpeedIncreasing=10
    OutRange=(Min=1,Max=0)
    Scale=1
    bClampResult=false
    bMapRange=false
    Value=&RigVMFunction_MathDoubleRemap(
     SourceMaximum=24
     SourceMinimum=9
     TargetMaximum=0
     TargetMinimum=1
     Value=&RigVMFunction_MathVectorDistance(B=@n74.Result, A=@n73.RelativeTransform.Translation).Result
    ).Result
   )
   RigVMVariableNode(ExecuteContext=(), Value=@n75.Result) {variable=R Leg Across Front}
   n76 = &RigUnit_GetRelativeTransformForItem(Child=(Type=Bone,Name="foot_r"), Parent=(Type=Bone,Name="root"), bChildInitial=False, bParentInitial=False)
   n77 = &RigVMFunction_MathVectorAdd(B=(X=12,Y=-40,Z=7), B.Z=@n76.RelativeTransform.Translation.Z, A=RigVMVariableNode() {variable=root position}.Value) {template="Add::Execute(in A,in B,out Result)"}
   n78 = &RigVMFunction_AlphaInterp(
    ClampMax=0
    ClampMin=1
    InRange=(Min=10,Max=28)
    InterpSpeedDecreasing=10
    InterpSpeedIncreasing=10
    OutRange=(Min=1,Max=0)
    Scale=1
    bClampResult=false
    bMapRange=false
    Value=&RigVMFunction_MathDoubleRemap(
     SourceMaximum=18
     SourceMinimum=7
     TargetMaximum=0
     TargetMinimum=1
     Value=&RigVMFunction_MathVectorDistance(B=@n77.Result, A=@n76.RelativeTransform.Translation).Result
    ).Result
   )
   > RigVMVariableNode(Value=@n78.Result) {variable=R Leg Across Rear}
   RigUnit_GetRelativeTransformForItem = &RigUnit_GetRelativeTransformForItem(Child=(Type=Bone,Name="foot_l"), Parent=(Type=Bone,Name="root"), bChildInitial=False, bParentInitial=False)
   Add = &RigVMFunction_MathVectorAdd(
    B=(X=-18,Y=10,Z=7)
    B.Z=@RigUnit_GetRelativeTransformForItem.RelativeTransform.Translation.Z
    A=RigVMVariableNode() {variable=root position}.Value
   ) {template="Add::Execute(in A,in B,out Result)"}
   RigVMFunction_MathVectorDistance = &RigVMFunction_MathVectorDistance(B=@Add.Result, A=@RigUnit_GetRelativeTransformForItem.RelativeTransform.Translation)
   Remap = &RigVMFunction_MathDoubleRemap(SourceMaximum=24, SourceMinimum=9, TargetMaximum=0, TargetMinimum=1, Value=@RigVMFunction_MathVectorDistance.Result)
   n79 = &RigVMFunction_AlphaInterp(
    ClampMax=0
    ClampMin=1
    InRange=(Min=10,Max=28)
    InterpSpeedDecreasing=10
    InterpSpeedIncreasing=10
    OutRange=(Min=1,Max=0)
    Scale=1
    bClampResult=false
    bMapRange=false
    Value=@Remap.Result
   )
   RigVMVariableNode(ExecuteContext=(), Value=@n79.Result) {variable=L Leg Across Front}
   > RigVMFunction_DebugTransformMutableNoSpace = /Script/RigVM.RigVMFunction_DebugTransformMutableNoSpace(
    Color=(R=1,G=0,B=0,A=1)
    DebugDrawSettings=(DepthPriority=SDPG_Foreground,Lifetime=-1)
    Mode=Axes
    Scale=10
    Thickness=0
    Transform=(Rotation=(X=0,Y=0,Z=0,W=1),Translation=(X=0,Y=0,Z=0),Scale3D=(X=1,Y=1,Z=1))
    WorldOffset=(Rotation=(X=0,Y=0,Z=0,W=1),Translation=(X=0,Y=0,Z=0),Scale3D=(X=1,Y=1,Z=1))
    bEnabled=true
    Transform.Translation=@Add.Result
   ) {is=struct RigVMUnitNode}
   n80 = &RigUnit_GetRelativeTransformForItem(Child=(Type=Bone,Name="foot_l"), Parent=(Type=Bone,Name="root"), bChildInitial=False, bParentInitial=False)
   n81 = &RigVMFunction_MathVectorAdd(B=(X=-12,Y=-40,Z=7), B.Z=@n80.RelativeTransform.Translation.Z, A=RigVMVariableNode() {variable=root position}.Value) {template="Add::Execute(in A,in B,out Result)"}
   n82 = &RigVMFunction_AlphaInterp(
    ClampMax=0
    ClampMin=1
    InRange=(Min=10,Max=28)
    InterpSpeedDecreasing=10
    InterpSpeedIncreasing=10
    OutRange=(Min=1,Max=0)
    Scale=1
    bClampResult=false
    bMapRange=false
    Value=&RigVMFunction_MathDoubleRemap(
     SourceMaximum=18
     SourceMinimum=7
     TargetMaximum=0
     TargetMinimum=1
     Value=&RigVMFunction_MathVectorDistance(B=@n81.Result, A=@n80.RelativeTransform.Translation).Result
    ).Result
   )
   > RigVMVariableNode(Value=@n82.Result) {variable=L Leg Across Rear}
   &RigVMFunction_MathDoubleMul(B=@n72.Value, A=RigVMVariableNode() {variable=L Knee Bend Pelvis Rel}.Value)
   &RigVMFunction_MathDoubleMul(B=@n72.Value, A=RigVMVariableNode() {variable=R Knee Bend Pelvis Rel}.Value)
   &RigVMFunction_DebugTransformMutableNoSpace(
    Color=(R=1,G=0,B=0,A=1)
    DebugDrawSettings=(DepthPriority=SDPG_Foreground,Lifetime=-1)
    ExecutePin=()
    Mode=Axes
    Scale=10
    Thickness=0
    Transform=(Rotation=(X=0,Y=0,Z=0,W=1),Translation=(X=0,Y=0,Z=0),Scale3D=(X=1,Y=1,Z=1))
    WorldOffset=(Rotation=(X=0,Y=0,Z=0,W=1),Translation=(X=0,Y=0,Z=0),Scale3D=(X=1,Y=1,Z=1))
    bEnabled=true
    Transform.Translation=@n81.Result
   )
   &RigVMFunction_DebugTransformMutableNoSpace(
    Color=(R=1,G=0,B=0,A=1)
    DebugDrawSettings=(DepthPriority=SDPG_Foreground,Lifetime=-1)
    ExecutePin=()
    Mode=Axes
    Scale=10
    Thickness=0
    Transform=(Rotation=(X=0,Y=0,Z=0,W=1),Translation=(X=0,Y=0,Z=0),Scale3D=(X=1,Y=1,Z=1))
    WorldOffset=(Rotation=(X=0,Y=0,Z=0,W=1),Translation=(X=0,Y=0,Z=0),Scale3D=(X=1,Y=1,Z=1))
    bEnabled=false
    Transform.Translation=@n77.Result
   )
   &RigVMFunction_DebugTransformMutableNoSpace(
    Color=(R=1,G=0,B=0,A=1)
    DebugDrawSettings=(DepthPriority=SDPG_Foreground,Lifetime=-1)
    ExecutePin=()
    Mode=Axes
    Scale=10
    Thickness=0
    Transform=(Rotation=(X=0,Y=0,Z=0,W=1),Translation=(X=0,Y=0,Z=0),Scale3D=(X=1,Y=1,Z=1))
    WorldOffset=(Rotation=(X=0,Y=0,Z=0,W=1),Translation=(X=0,Y=0,Z=0),Scale3D=(X=1,Y=1,Z=1))
    bEnabled=false
   )
   &RigVMFunction_DebugTransformMutableNoSpace(
    Color=(R=1,G=0,B=0,A=1)
    DebugDrawSettings=(DepthPriority=SDPG_Foreground,Lifetime=-1)
    ExecutePin=()
    Mode=Axes
    Scale=10
    Thickness=0
    Transform=(Rotation=(X=0,Y=0,Z=0,W=1),Translation=(X=0,Y=0,Z=0),Scale3D=(X=1,Y=1,Z=1))
    WorldOffset=(Rotation=(X=0,Y=0,Z=0,W=1),Translation=(X=0,Y=0,Z=0),Scale3D=(X=1,Y=1,Z=1))
    bEnabled=false
   )
   &RigVMFunction_DebugTransformMutableNoSpace(
    Color=(R=1,G=0,B=0,A=1)
    DebugDrawSettings=(DepthPriority=SDPG_Foreground,Lifetime=-1)
    ExecutePin=()
    Mode=Axes
    Scale=10
    Thickness=0
    Transform=(Rotation=(X=0,Y=0,Z=0,W=1),Translation=(X=0,Y=0,Z=0),Scale3D=(X=1,Y=1,Z=1))
    WorldOffset=(Rotation=(X=0,Y=0,Z=0,W=1),Translation=(X=0,Y=0,Z=0),Scale3D=(X=1,Y=1,Z=1))
    bEnabled=false
    Transform.Translation=@n74.Result
   )
  ]}
 New Function_ContainedGraph {nodes [
   RigVMFunctionEntryNode()
   > RigVMFunction_Sequence = RigVMAggregateNode()
   Full Leg ROM = RigVMFunctionReferenceNode(
    Interp Output Speed=5
    Interp Results=false
    ExecuteContext=@RigVMFunction_Sequence.B
    Result Alpha=RigVMVariableNode() {variable=BlendOutLegRom}.Value
   )
   > RigVMFunction_Sequence_1 = RigVMAggregateNode()
   Modify Transforms with Mirror_1 = RigVMFunctionReferenceNode(
    Base Modify Data="((Item=(Type=Bone,Name=\"dyn_front_skirt_l_01\"),Transform=(Rotation=(X=-0.1949,Y=0.0493,Z=-0.7054,W=0.6797),Translation=(X=-15,Y=-5,Z=-1.75),Scale3D=(X=1,Y=1,Z=1))),(Item=(Type=Bone,Name=\"dyn_front_skirt_corner_l_01\"),Transform=(Rotation=(X=0.0526,Y=-0.4758 ...(truncated)"
    Mirror Modify Data="((Item=(Type=Bone,Name=\"dyn_front_skirt_r_01\"),Transform=(Rotation=(X=0,Y=0,Z=0,W=1),Translation=(X=0,Y=0,Z=0),Scale3D=(X=1,Y=1,Z=1))),(Item=(Type=Bone,Name=\"dyn_front_skirt_corner_r_01\"),Transform=(Rotation=(X=0,Y=0,Z=-0.0436,W=0.999),Translation=(X=1,Y=0 ...(truncated)"
    ExecuteContext=@RigVMFunction_Sequence_1.A
    "Mirror Weight "=@"Full Leg ROM".R Thigh Raised Fwd
    Weight=@"Full Leg ROM".L Thigh Raised Fwd
   )
   > Modify Transforms with Mirror_1_1 = RigVMFunctionReferenceNode(
    Base Modify Data="((Item=(Type=Bone,Name=\"dyn_front_skirt_l_01\"),Transform=(Rotation=(X=0.1827,Y=0.0625,Z=-0.1431,W=0.9707),Translation=(X=-0.5,Y=-3.5,Z=0),Scale3D=(X=1,Y=1,Z=1))),(Item=(Type=Bone,Name=\"dyn_front_skirt_corner_l_01\"),Transform=(Rotation=(X=-0.1294,Y=0.017,Z= ...(truncated)"
    Mirror Modify Data="((Item=(Type=Bone,Name=\"dyn_front_skirt_r_01\"),Transform=(Rotation=(X=0,Y=0,Z=0,W=1),Translation=(X=0,Y=0,Z=0),Scale3D=(X=1,Y=1,Z=1))),(Item=(Type=Bone,Name=\"dyn_front_skirt_corner_r_01\"),Transform=(Rotation=(X=0,Y=0,Z=0,W=1),Translation=(X=0,Y=0,Z=0),Scal ...(truncated)"
    "Mirror Weight "=@"Full Leg ROM".R Thigh Raised Fwd Side Inbetween
    Weight=@"Full Leg ROM".L Thigh Raised Fwd Side Inbetween
   )
   > Modify Transforms with Mirror_1_2 = RigVMFunctionReferenceNode(
    Base Modify Data="((Item=(Type=Bone,Name=\"dyn_front_skirt_l_01\"),Transform=(Rotation=(X=0.0226,Y=0.2578,Z=-0.0842,W=0.9623),Translation=(X=0,Y=0,Z=0),Scale3D=(X=1,Y=1,Z=1))),(Item=(Type=Bone,Name=\"dyn_front_skirt_corner_l_01\"),Transform=(Rotation=(X=0.183,Y=0.183,Z=-0.683,W ...(truncated)"
    Mirror Modify Data="((Item=(Type=Bone,Name=\"dyn_front_skirt_r_01\"),Transform=(Rotation=(X=0,Y=0,Z=0,W=1),Translation=(X=0,Y=0,Z=0),Scale3D=(X=1,Y=1,Z=1))),(Item=(Type=Bone,Name=\"dyn_front_skirt_corner_r_01\"),Transform=(Rotation=(X=0,Y=0,Z=0,W=1),Translation=(X=0,Y=0,Z=0),Scal ...(truncated)"
    "Mirror Weight "=@"Full Leg ROM".R Thigh Raised Side
    Weight=@"Full Leg ROM".L Thigh Raised Side
   )
   > Modify Transforms with Mirror_1_3 = RigVMFunctionReferenceNode(
    Base Modify Data="((Item=(Type=Bone,Name=\"dyn_front_skirt_l_01\"),Transform=(Rotation=(X=0,Y=0,Z=0,W=1),Translation=(X=0,Y=0,Z=0),Scale3D=(X=1,Y=1,Z=1))),(Item=(Type=Bone,Name=\"dyn_front_skirt_corner_l_01\"),Transform=(Rotation=(X=0,Y=0,Z=0,W=1),Translation=(X=0,Y=0,Z=0),Scal ...(truncated)"
    Mirror Modify Data="((Item=(Type=Bone,Name=\"dyn_front_skirt_r_01\"),Transform=(Rotation=(X=0,Y=0,Z=0,W=1),Translation=(X=0,Y=0,Z=0),Scale3D=(X=1,Y=1,Z=1))),(Item=(Type=Bone,Name=\"dyn_front_skirt_corner_r_01\"),Transform=(Rotation=(X=0,Y=0,Z=0,W=1),Translation=(X=0,Y=0,Z=0),Scal ...(truncated)"
    "Mirror Weight "=@"Full Leg ROM".R Thigh Raised Side Rear Inbetween
    Weight=@"Full Leg ROM".L Thigh Raised Side Rear Inbetween
   )
   > Modify Transforms with Mirror_1_4 = RigVMFunctionReferenceNode(
    Base Modify Data="((Item=(Type=Bone,Name=\"dyn_rear_skirt_corner_l_01\"),Transform=(Rotation=(X=-0.171,Y=0.171,Z=0.0302,W=0.9698),Translation=(X=0,Y=0,Z=-5),Scale3D=(X=1,Y=1,Z=1))),(Item=(Type=Bone,Name=\"dyn_rear_skirt_l_01\"),Transform=(Rotation=(X=0.0689,Y=0.0898,Z=-0.6048,W ...(truncated)"
    Mirror Modify Data="((Item=(Type=Bone,Name=\"dyn_rear_skirt_corner_r_01\"),Transform=(Rotation=(X=0,Y=0,Z=0,W=1),Translation=(X=0,Y=0,Z=0),Scale3D=(X=1,Y=1,Z=1))),(Item=(Type=Bone,Name=\"dyn_rear_skirt_r_01\"),Transform=(Rotation=(X=0,Y=0,Z=0,W=1),Translation=(X=0,Y=0,Z=0),Scale3 ...(truncated)"
    "Mirror Weight "=@"Full Leg ROM".R Thigh Raised Rear
    Weight=@"Full Leg ROM".L Thigh Raised Rear
   )
   > Modify Transforms with Mirror_1_5 = RigVMFunctionReferenceNode(
    Base Modify Data="((Item=(Type=Bone,Name=\"dyn_front_skirt_l_01\"),Transform=(Rotation=(X=-0.0302,Y=-0.171,Z=-0.171,W=0.9698),Translation=(X=0,Y=0,Z=0),Scale3D=(X=1,Y=1,Z=1))),(Item=(Type=Bone,Name=\"dyn_front_skirt_corner_l_01\"),Transform=(Rotation=(X=0,Y=0,Z=0,W=1),Translati ...(truncated)"
    Mirror Modify Data="((Item=(Type=Bone,Name=\"dyn_front_skirt_r_01\"),Transform=(Rotation=(X=0,Y=0,Z=0,W=1),Translation=(X=0,Y=0,Z=0),Scale3D=(X=1,Y=1,Z=1))),(Item=(Type=Bone,Name=\"dyn_front_skirt_corner_r_01\"),Transform=(Rotation=(X=0,Y=0,Z=0,W=1),Translation=(X=0,Y=0,Z=0),Scal ...(truncated)"
    Weight=@"Full Leg ROM".L Thigh Across R
    "Mirror Weight "=@"Full Leg ROM".R Thigh Across L
   )
   Modify Transforms with Mirror_1_8 = RigVMFunctionReferenceNode(
    Base Modify Data="((Item=(Type=Bone,Name=\"dyn_side_skirt_r_MSTR\"),Transform=(Rotation=(X=0,Y=0,Z=-0.3827,W=0.9239),Translation=(X=0,Y=0,Z=0),Scale3D=(X=1,Y=1,Z=1))),(Item=(Type=Bone,Name=\"dyn_rear_skirt_corner_r_MSTR\"),Transform=(Rotation=(X=0,Y=0,Z=-0.2588,W=0.9659),Transl ...(truncated)"
    Mirror Modify Data="((Item=(Type=Bone,Name=\"dyn_side_skirt_l_MSTR\"),Transform=(Rotation=(X=0,Y=0,Z=0,W=1),Translation=(X=0,Y=0,Z=0),Scale3D=(X=1,Y=1,Z=1))),(Item=(Type=Bone,Name=\"dyn_rear_skirt_corner_l_MSTR\"),Transform=(Rotation=(X=0,Y=0,Z=0,W=1),Translation=(X=0,Y=0,Z=0),Sc ...(truncated)"
    ExecuteContext=@RigVMFunction_Sequence_1.B
    Weight=@"Full Leg ROM".L Leg Across R Leg Side
    "Mirror Weight "=@"Full Leg ROM".R Leg Across L Leg Side
   )
   &RigUnit_ModifyTransforms(
    ItemToModify="((Item=(Type=Bone,Name=\"dyn_rear_skirt_l_02_MSTR\"),Transform=(Rotation=(X=0,Y=-0.0175,Z=0,W=0.9998),Translation=(X=0,Y=0,Z=1),Scale3D=(X=1,Y=1,Z=1))),(Item=(Type=Bone,Name=\"dyn_rear_skirt_r_02_MSTR\"),Transform=(Rotation=(X=0,Y=-0.0175,Z=0,W=0.9998),Transla ...(truncated)"
    Weight=1
    WeightMinimum=0
    ExecutePin=@RigVMFunction_Sequence_1.D
   )
   &RigUnit_ModifyTransforms(
    ItemToModify="((Item=(Type=Bone,Name=\"dyn_front_skirt_l_MSTR\"),Transform=(Rotation=(X=0.0736,Y=0.0004,Z=0.0006,W=0.9973),Translation=(X=3.4059,Y=0,Z=0),Scale3D=(X=1,Y=1,Z=1))),(Item=(Type=Bone,Name=\"dyn_front_skirt_r_MSTR\"),Transform=(Rotation=(X=-0.0002,Y=0.0025,Z=0.02 ...(truncated)"
    WeightMinimum=0
    Weight=@"Full Leg ROM".Both Thighs Raised
    ExecutePin=@RigVMFunction_Sequence_1.C
   )
   > &RigUnit_ModifyTransforms(
    ItemToModify="((Item=(Type=Bone,Name=\"dyn_rear_skirt_l_01\"),Transform=(Rotation=(X=0.0057,Y=-0.0432,Z=0.1304,W=0.9905),Translation=(X=0,Y=0,Z=0),Scale3D=(X=1,Y=1,Z=1))),(Item=(Type=Bone,Name=\"dyn_rear_skirt_r_01\"),Transform=(Rotation=(X=0.0057,Y=-0.0432,Z=0.1304,W=0.990 ...(truncated)"
    WeightMinimum=0
    Weight=@"Full Leg ROM".Both Thighs Back
   )
   > Modify Transforms with Mirror_2 = RigVMFunctionReferenceNode(
    Base Modify Data="((Item=(Type=Bone,Name=\"dyn_front_skirt_l_03\"),Transform=(Rotation=(X=0,Y=0,Z=0.6152,W=0.7883),Translation=(X=0,Y=0,Z=0),Scale3D=(X=1,Y=1,Z=1))),(Item=(Type=Bone,Name=\"dyn_front_skirt_corner_l_03\"),Transform=(Rotation=(X=0.1726,Y=0.4087,Z=0.3418,W=0.8285), ...(truncated)"
    Mirror Modify Data="((Item=(Type=Bone,Name=\"dyn_front_skirt_r_03\"),Transform=(Rotation=(X=0,Y=0,Z=0,W=1),Translation=(X=0,Y=0,Z=0),Scale3D=(X=1,Y=1,Z=1))),(Item=(Type=Bone,Name=\"dyn_front_skirt_corner_r_03\"),Transform=(Rotation=(X=0,Y=0,Z=0,W=1),Translation=(X=0,Y=0,Z=0),Scal ...(truncated)"
    Weight=@"Full Leg ROM".L Thigh Fwd Knee Bend
    "Mirror Weight "=@"Full Leg ROM".R Thigh Fwd Knee Bend
   )
   > Modify Transforms with Mirror_1_6_1 = RigVMFunctionReferenceNode(
    Base Modify Data="((Item=(Type=Bone,Name=\"dyn_rear_skirt_l_01\"),Transform=(Rotation=(X=0,Y=0,Z=0,W=1),Translation=(X=-5,Y=-10,Z=0),Scale3D=(X=1,Y=1,Z=1))),(Item=(Type=Bone,Name=\"dyn_rear_skirt_l_02_MSTR\"),Transform=(Rotation=(X=0,Y=0,Z=0.0872,W=0.9962),Translation=(X=0,Y=0, ...(truncated)"
    Mirror Modify Data="((Item=(Type=Bone,Name=\"dyn_rear_skirt_r_01\"),Transform=(Rotation=(X=0,Y=0,Z=0,W=1),Translation=(X=0,Y=0,Z=0),Scale3D=(X=1,Y=1,Z=1))),(Item=(Type=Bone,Name=\"dyn_rear_skirt_corner_r_02_MSTR\"),Transform=(Rotation=(X=0,Y=0,Z=0,W=1),Translation=(X=0,Y=0,Z=0),S ...(truncated)"
    "Mirror Weight "=@"Full Leg ROM".R Foot Near Pelvis
    Weight=@"Full Leg ROM".L Foot Near Pelvis
   )
   > Modify Transforms with Mirror_1_6 = RigVMFunctionReferenceNode(
    Base Modify Data="((Item=(Type=Bone,Name=\"dyn_rear_skirt_l_02_MSTR\"),Transform=(Rotation=(X=-0.0118,Y=-0.0129,Z=-0.6755,W=0.7372),Translation=(X=-10,Y=0,Z=0),Scale3D=(X=1,Y=1,Z=1))),(Item=(Type=Bone,Name=\"dyn_rear_skirt_corner_l_02_MSTR\"),Transform=(Rotation=(X=-0.0295,Y=-0 ...(truncated)"
    Mirror Modify Data="((Item=(Type=Bone,Name=\"dyn_rear_skirt_r_02_MSTR\"),Transform=(Rotation=(X=0,Y=0,Z=0,W=1),Translation=(X=0,Y=0,Z=0),Scale3D=(X=1,Y=1,Z=1))),(Item=(Type=Bone,Name=\"dyn_rear_skirt_corner_r_02_MSTR\"),Transform=(Rotation=(X=0,Y=0,Z=0,W=1),Translation=(X=0,Y=0,Z ...(truncated)"
    "Mirror Weight "=@"Full Leg ROM".R Thigh Rear Knee Bend
    Weight=@"Full Leg ROM".L Thigh Rear Knee Bend
   )
   > Modify Transforms with Mirror_1_6_2 = RigVMFunctionReferenceNode(
    Base Modify Data=((Item=(Type=Bone,Name="dyn_front_skirt_l_01"),Transform=(Rotation=(X=0.0996,Y=0.1422,Z=-0.5649,W=0.8067),Translation=(X=-5,Y=0,Z=-5),Scale3D=(X=1,Y=1,Z=1))))
    Mirror Modify Data=((Item=(Type=Bone,Name="dyn_front_skirt_r_01"),Transform=(Rotation=(X=0,Y=0,Z=0,W=1),Translation=(X=0,Y=0,Z=0),Scale3D=(X=1,Y=1,Z=1))))
    "Mirror Weight "=@"Full Leg ROM".R Thigh Side L Thigh Raise
    Weight=@"Full Leg ROM".L Thigh Side R Thigh Raise
   )
   RigVMFunctionReturnNode(ExecuteContext=@RigVMFunction_Sequence.F)
   RigVMFunction_MathBoolOr = &RigVMFunction_MathBoolOr(A=RigVMVariableNode() {variable=IsKneeSliding}.Value, B=RigVMVariableNode() {variable=IsBaseballSliding}.Value)
   n1 = &RigVMFunction_AlphaInterp(
    InterpSpeedDecreasing=4
    InterpSpeedIncreasing=20
    Scale=1
    bClampResult=true
    Value=RigVMDispatchNode(False=1, True=0, Condition=@RigVMFunction_MathBoolOr.Result) {template="DISPATCH_RigVMDispatch_If(in Condition,in True,in False,out Result)"}.Result
   )
   RigVMVariableNode(Value=@n1.Result, ExecuteContext=@RigVMFunction_Sequence.A) {variable=BlendOutLegRom}
   RigVMFunction_MathBoolToFloat = &RigVMFunction_MathBoolToFloat(Value=RigVMVariableNode() {variable=IsCrouching}.Value)
   n2 = &RigVMFunction_AlphaInterp(InterpSpeedDecreasing=10, InterpSpeedIncreasing=10, Scale=1, bClampResult=true, Value=@RigVMFunction_MathBoolToFloat.Result)
   &RigUnit_ModifyTransforms(
    ItemToModify="((Item=(Type=Bone,Name=\"dyn_root_offset\"),Transform=(Rotation=(X=0,Y=0,Z=0,W=1),Translation=(X=0,Y=0,Z=10),Scale3D=(X=1,Y=1,Z=1))),(Item=(Type=Bone,Name=\"dyn_rear_skirt_l_01\"),Transform=(Rotation=(X=0,Y=0,Z=-0.1305,W=0.9914),Translation=(X=0,Y=0,Z=0),Scale ...(truncated)"
    WeightMinimum=0
    Weight=@n2.Result
    ExecutePin=@RigVMFunction_Sequence.D
   )
   n3 = &RigVMFunction_AlphaInterp(
    InterpSpeedDecreasing=10
    InterpSpeedIncreasing=10
    Scale=1
    bClampResult=true
    Value=&RigVMFunction_MathBoolToFloat(Value=RigVMVariableNode() {variable=MoveGroundCollision}.Value).Result
   )
   > &RigUnit_ModifyTransforms(
    ItemToModify=((Item=(Type=Bone,Name="dyn_root_offset"),Transform=(Rotation=(X=0,Y=0,Z=0,W=1),Translation=(X=0,Y=0,Z=-500),Scale3D=(X=1,Y=1,Z=1))))
    WeightMinimum=0
    Weight=@n3.Result
   )
   n4 = RigVMVariableNode() {variable=IsBaseballSliding}
   n5 = RigVMVariableNode() {variable=IsKneeSliding}
   RigVMFunction_ControlFlowBranch = &RigVMFunction_ControlFlowBranch(Condition=&RigVMFunction_MathBoolOr(B=@n4.Value, A=@n5.Value).Result, ExecuteContext=@RigVMFunction_Sequence.E)
   Subtract = &RigVMFunction_MathDoubleSub(A=1, B=RigVMVariableNode() {variable=BlendOutLegRom}.Value)
   Multiply = &RigVMFunction_MathDoubleMul(B=&RigVMFunction_MathBoolToFloat(Value=@n5.Value).Result, A=@Subtract.Result)
   Modify Transforms with Mirror_1_7 = RigVMFunctionReferenceNode(
    Base Modify Data="((Item=(Type=Bone,Name=\"dyn_front_skirt_l_01\"),Transform=(Rotation=(X=0.05,Y=0.0714,Z=-0.5714,W=0.816),Translation=(X=-10,Y=0,Z=0),Scale3D=(X=1,Y=1,Z=1))),(Item=(Type=Bone,Name=\"dyn_front_skirt_corner_l_01\"),Transform=(Rotation=(X=0.0996,Y=0.1422,Z=-0.5649 ...(truncated)"
    Mirror Modify Data="((Item=(Type=Bone,Name=\"dyn_front_skirt_r_01\"),Transform=(Rotation=(X=0,Y=0,Z=0,W=1),Translation=(X=0,Y=0,Z=0),Scale3D=(X=1,Y=1,Z=1))),(Item=(Type=Bone,Name=\"dyn_front_skirt_corner_r_01\"),Transform=(Rotation=(X=0,Y=0,Z=0,W=1),Translation=(X=0,Y=0,Z=0),Scal ...(truncated)"
    "Mirror Weight "=@Multiply.Result
    Weight=@Multiply.Result
    ExecuteContext=@RigVMFunction_ControlFlowBranch.True
   )
   > &RigUnit_ModifyTransforms(
    ItemToModify=((Item=(Type=Bone,Name="dyn_bow_loop_MSTR"),Transform=(Rotation=(X=-0.1951,Y=0,Z=0,W=0.9808),Translation=(X=0,Y=-2.4335,Z=-1.0143),Scale3D=(X=1,Y=1,Z=1))))
    WeightMinimum=0
    Weight=@Multiply.Result
   )
   n6 = RigVMVariableNode() {variable=BlendOutLegRom}
   > &RigUnit_ModifyTransforms(
    ItemToModify="((Item=(Type=Bone,Name=\"dyn_front_skirt_l_01\"),Transform=(Rotation=(X=0,Y=0,Z=-0.5,W=0.866),Translation=(X=-10,Y=0,Z=0),Scale3D=(X=1,Y=1,Z=1))),(Item=(Type=Bone,Name=\"dyn_front_skirt_corner_l_01\"),Transform=(Rotation=(X=0,Y=0,Z=-0.5,W=0.866),Translation=(X ...(truncated)"
    WeightMinimum=0
    Weight=&RigVMFunction_MathDoubleMul(B=&RigVMFunction_MathBoolToFloat(Value=@n4.Value).Result, A=&RigVMFunction_MathDoubleSub(A=1, B=@n6.Value).Result).Result
   )
   n7 = &RigVMFunction_MathTransformMul(
    A=(Rotation=(X=0,Y=0,Z=0,W=1),Translation=(X=-15,Y=23,Z=28),Scale3D=(X=1,Y=1,Z=1))
    B=&RigUnit_GetRelativeTransformForItem(Child=(Type=Bone,Name="pelvis"), Parent=(Type=Bone,Name="root"), bChildInitial=False, bParentInitial=true).RelativeTransform
   ) {template="Multiply::Execute(in A,in B,out Result)"}
   n8 = &RigUnit_GetRelativeTransformForItem(Child=(Type=Bone,Name="foot_l"), Parent=(Type=Bone,Name="root"), bChildInitial=False, bParentInitial=true)
   &RigUnit_ModifyTransforms(
    ItemToModify="((Item=(Type=Bone,Name=\"dyn_front_skirt_r_MSTR\"),Transform=(Rotation=(X=0.1624,Y=0.0677,Z=-0.1549,W=0.9721),Translation=(X=3.2776,Y=9.1224,Z=2.9912),Scale3D=(X=1,Y=1,Z=1))),(Item=(Type=Bone,Name=\"dyn_front_skirt_corner_r_MSTR\"),Transform=(Rotation=(X=0.059 ...(truncated)"
    WeightMinimum=0
    Weight=&RigVMFunction_MathDoubleRemap(
     SourceMaximum=10
     SourceMinimum=1.4
     TargetMaximum=0
     TargetMinimum=1
     Value=&RigVMFunction_MathVectorDistance(A=@n7.Result.Translation, B=@n8.RelativeTransform.Translation).Result
    ).Result
    ExecutePin=@RigVMFunction_Sequence.C
   )
  ]}
 New Function_ContainedGraph {nodes [
   RigVMFunctionEntryNode()
   > RigVMFunction_Sequence = RigVMAggregateNode()
   &RigUnit_ParentConstraint(
    AdvancedSettings=(RotationOrderForFilter=XZY)
    Child=(Type=Bone,Name="dyn_collar_side_l_01")
    Filter=(TranslationFilter=(bX=True,bY=True,bZ=True),RotationFilter=(bX=true,bY=true,bZ=true),ScaleFilter=(bX=True,bY=True,bZ=True))
    Parents="((Item=(Type=Bone,Name=\"deform_pec_l\"),Weight=0.004),(Item=(Type=Bone,Name=\"spine_03\"),Weight=0.0114),(Item=(Type=Bone,Name=\"spine_04\"),Weight=0.0213),(Item=(Type=Bone,Name=\"deform_upperarm_up_l\"),Weight=0.0228),(Item=(Type=Bone,Name=\"neck_01\"),Weight=0.04 ...(truncated)"
    Weight=1
    bMaintainOffset=true
    ExecutePin=@RigVMFunction_Sequence.A
   )
   > &RigUnit_ParentConstraint(
    AdvancedSettings=(RotationOrderForFilter=XZY)
    Child=(Type=Bone,Name="dyn_collar_front_01")
    Filter=(TranslationFilter=(bX=True,bY=True,bZ=True),RotationFilter=(bX=true,bY=true,bZ=true),ScaleFilter=(bX=True,bY=True,bZ=True))
    Parents="((Item=(Type=Bone,Name=\"deform_pec_r\"),Weight=0.0021),(Item=(Type=Bone,Name=\"clavicle_r\"),Weight=0.032),(Item=(Type=Bone,Name=\"clavicle_l\"),Weight=0.0334),(Item=(Type=Bone,Name=\"spine_03\"),Weight=0.0336),(Item=(Type=Bone,Name=\"neck_01\"),Weight=0.0416),(Ite ...(truncated)"
    Weight=1
    bMaintainOffset=true
   )
   > &RigUnit_ParentConstraint(
    AdvancedSettings=(RotationOrderForFilter=XZY)
    Child=(Type=Bone,Name="dyn_collar_rear_01")
    Filter=(TranslationFilter=(bX=True,bY=True,bZ=True),RotationFilter=(bX=true,bY=true,bZ=true),ScaleFilter=(bX=True,bY=True,bZ=True))
    Parents="((Item=(Type=Bone,Name=\"deform_upperarm_up_l\"),Weight=0.0013),(Item=(Type=Bone,Name=\"spine_04\"),Weight=0.0013),(Item=(Type=Bone,Name=\"spine_03\"),Weight=0.0197),(Item=(Type=Bone,Name=\"neck_01\"),Weight=0.1017),(Item=(Type=Bone,Name=\"clavicle_l\"),Weight=0.114 ...(truncated)"
    Weight=1
    bMaintainOffset=true
   )
   > &RigUnit_ParentConstraint(
    AdvancedSettings=(RotationOrderForFilter=XZY)
    Child=(Type=Bone,Name="dyn_collar_rear_corner_l_01")
    Filter=(TranslationFilter=(bX=True,bY=True,bZ=True),RotationFilter=(bX=true,bY=true,bZ=true),ScaleFilter=(bX=True,bY=True,bZ=True))
    Parents="((Item=(Type=Bone,Name=\"upperarm_twist_01_l\"),Weight=0.0107),(Item=(Type=Bone,Name=\"spine_03\"),Weight=0.0127),(Item=(Type=Bone,Name=\"deform_upperarm_up_l\"),Weight=0.0155),(Item=(Type=Bone,Name=\"clavicle_r\"),Weight=0.0199),(Item=(Type=Bone,Name=\"neck_01\"),W ...(truncated)"
    Weight=1
    bMaintainOffset=true
   )
   > &RigUnit_ParentConstraint(
    AdvancedSettings=(RotationOrderForFilter=XZY)
    Child=(Type=Bone,Name="dyn_collar_side_r_01")
    Filter=(TranslationFilter=(bX=True,bY=True,bZ=True),RotationFilter=(bX=true,bY=true,bZ=true),ScaleFilter=(bX=True,bY=True,bZ=True))
    Parents="((Item=(Type=Bone,Name=\"deform_pec_r\"),Weight=0.0054),(Item=(Type=Bone,Name=\"deform_upperarm_up_r\"),Weight=0.0138),(Item=(Type=Bone,Name=\"spine_03\"),Weight=0.0143),(Item=(Type=Bone,Name=\"spine_04\"),Weight=0.0363),(Item=(Type=Bone,Name=\"neck_01\"),Weight=0.0 ...(truncated)"
    Weight=1
    bMaintainOffset=true
   )
   > &RigUnit_ParentConstraint(
    AdvancedSettings=(RotationOrderForFilter=XZY)
    Child=(Type=Bone,Name="dyn_collar_rear_corner_r_01")
    Filter=(TranslationFilter=(bX=True,bY=True,bZ=True),RotationFilter=(bX=true,bY=true,bZ=true),ScaleFilter=(bX=True,bY=True,bZ=True))
    Parents="((Item=(Type=Bone,Name=\"upperarm_twist_01_r\"),Weight=0.0083),(Item=(Type=Bone,Name=\"spine_03\"),Weight=0.0115),(Item=(Type=Bone,Name=\"deform_upperarm_up_r\"),Weight=0.012),(Item=(Type=Bone,Name=\"clavicle_l\"),Weight=0.019),(Item=(Type=Bone,Name=\"neck_01\"),Wei ...(truncated)"
    Weight=1
    bMaintainOffset=true
   )
   n1 = &RigUnit_GetTransform(Item=(Type=Control,Name="spine_05_ctrl"), bInitial=False)
   ClampSpatially = /Script/RigVM.RigVMFunction_MathTransformClampSpatially(
    Axis=X
    DebugColor=(R=1,G=0,B=0,A=1)
    DebugThickness=1
    Maximum=30
    Minimum=10
    Type=Capsule
    bDrawDebug=false
    Value=&RigUnit_GetTransform(Item=(Type=Bone,Name="dyn_collar_side_l_01"), bInitial=False).Transform
    Space=@n1.Transform
   ) {
    is=struct RigVMUnitNode
    template="ClampSpatially::Execute(in Value,in Axis,in Type,in Minimum,in Maximum,in Space,in bDrawDebug,in DebugColor,in DebugThickness,out Result)"
   }
   &RigUnit_SetTransform(Item=(Type=Bone,Name="dyn_collar_side_l_01"), Weight=1, bInitial=False, Value=@ClampSpatially.Result, ExecutePin=@RigVMFunction_Sequence.B)
   > &RigUnit_SetTransform(
    Item=(Type=Bone,Name="dyn_collar_front_01")
    Weight=1
    bInitial=False
    Value=&RigVMFunction_MathTransformClampSpatially(
     Axis=X
     DebugColor=(R=1,G=0,B=0,A=1)
     DebugThickness=1
     Maximum=30
     Minimum=10
     Type=Capsule
     bDrawDebug=false
     Value=&RigUnit_GetTransform(Item=(Type=Bone,Name="dyn_collar_front_01"), bInitial=False).Transform
     Space=@n1.Transform
    ) {template="ClampSpatially::Execute(in Value,in Axis,in Type,in Minimum,in Maximum,in Space,in bDrawDebug,in DebugColor,in DebugThickness,out Result)"}.Result
   )
   > &RigUnit_SetTransform(
    Item=(Type=Bone,Name="dyn_collar_rear_01")
    Weight=1
    bInitial=False
    Value=&RigVMFunction_MathTransformClampSpatially(
     Axis=X
     DebugColor=(R=1,G=0,B=0,A=1)
     DebugThickness=1
     Maximum=30
     Minimum=10
     Type=Capsule
     bDrawDebug=false
     Value=&RigUnit_GetTransform(Item=(Type=Bone,Name="dyn_collar_rear_01"), bInitial=False).Transform
     Space=@n1.Transform
    ) {template="ClampSpatially::Execute(in Value,in Axis,in Type,in Minimum,in Maximum,in Space,in bDrawDebug,in DebugColor,in DebugThickness,out Result)"}.Result
   )
   > &RigUnit_SetTransform(
    Item=(Type=Bone,Name="dyn_collar_rear_corner_l_01")
    Weight=1
    bInitial=False
    Value=&RigVMFunction_MathTransformClampSpatially(
     Axis=X
     DebugColor=(R=1,G=0,B=0,A=1)
     DebugThickness=1
     Maximum=30
     Minimum=10
     Type=Capsule
     bDrawDebug=false
     Value=&RigUnit_GetTransform(Item=(Type=Bone,Name="dyn_collar_rear_corner_l_01"), bInitial=False).Transform
     Space=@n1.Transform
    ) {template="ClampSpatially::Execute(in Value,in Axis,in Type,in Minimum,in Maximum,in Space,in bDrawDebug,in DebugColor,in DebugThickness,out Result)"}.Result
   )
   > &RigUnit_SetTransform(
    Item=(Type=Bone,Name="dyn_collar_side_r_01")
    Weight=1
    bInitial=False
    Value=&RigVMFunction_MathTransformClampSpatially(
     Axis=X
     DebugColor=(R=1,G=0,B=0,A=1)
     DebugThickness=1
     Maximum=30
     Minimum=10
     Type=Capsule
     bDrawDebug=false
     Value=&RigUnit_GetTransform(Item=(Type=Bone,Name="dyn_collar_side_r_01"), bInitial=False).Transform
     Space=@n1.Transform
    ) {template="ClampSpatially::Execute(in Value,in Axis,in Type,in Minimum,in Maximum,in Space,in bDrawDebug,in DebugColor,in DebugThickness,out Result)"}.Result
   )
   > &RigUnit_SetTransform(
    Item=(Type=Bone,Name="dyn_collar_rear_corner_r_01")
    Weight=1
    bInitial=False
    Value=&RigVMFunction_MathTransformClampSpatially(
     Axis=X
     DebugColor=(R=1,G=0,B=0,A=1)
     DebugThickness=1
     Maximum=30
     Minimum=10
     Type=Capsule
     bDrawDebug=false
     Value=&RigUnit_GetTransform(Item=(Type=Bone,Name="dyn_collar_rear_corner_r_01"), bInitial=False).Transform
     Space=@n1.Transform
    ) {template="ClampSpatially::Execute(in Value,in Axis,in Type,in Minimum,in Maximum,in Space,in bDrawDebug,in DebugColor,in DebugThickness,out Result)"}.Result
   )
   Spherical Collision = RigVMFunctionReferenceNode(
    Debug=false
    Max Radius=10
    Min Radius=10
    Source=(Type=Bone,Name="dyn_collar_side_l_02")
    Source Offset=(X=0,Y=0,Z=0)
    Target=(Type=Bone,Name="head")
    Target Offset=(X=-2,Y=2,Z=0)
    Weight=1
    ExecuteContext=@RigVMFunction_Sequence.C
   )
   > Spherical Collision_1 = RigVMFunctionReferenceNode(
    Debug=false
    Max Radius=10
    Min Radius=10
    Source=(Type=Bone,Name="dyn_collar_front_02")
    Source Offset=(X=0,Y=0,Z=0)
    Target=(Type=Bone,Name="head")
    Target Offset=(X=-2,Y=1,Z=0)
    Weight=1
   )
   > Spherical Collision_1_1 = RigVMFunctionReferenceNode(
    Debug=false
    Max Radius=10
    Min Radius=10
    Source=(Type=Bone,Name="dyn_collar_side_r_02")
    Source Offset=(X=0,Y=0,Z=0)
    Target=(Type=Bone,Name="head")
    Target Offset=(X=-2,Y=2,Z=0)
    Weight=1
   )
   n2 = &RigUnit_SphericalPoseReader(
    ActiveRegionScaleFactors=(PositiveWidth=1,NegativeWidth=2,PositiveHeight=2,NegativeHeight=2)
    ActiveRegionSize=0.1
    Debug=(bDrawDebug=false,bDraw2D=False,bDrawLocalAxes=False,DebugScale=25,DebugSegments=20,DebugThickness=0.25)
    DriverAxis=(X=1,Y=0,Z=0)
    DriverItem=(Type=Bone,Name="head")
    FalloffRegionScaleFactors=(PositiveWidth=1,NegativeWidth=1,PositiveHeight=1,NegativeHeight=1)
    FalloffSize=0.3
    FlipHeightScaling=False
    FlipWidthScaling=False
    OptionalParentItem=(Type=Bone,Name="spine_05")
    RotationOffset=(X=-90,Y=0,Z=20)
    ExecutePin=@RigVMFunction_Sequence.D
   )
   > &RigUnit_ModifyTransforms(
    ItemToModify="((Item=(Type=Bone,Name=\"dyn_collar_rear_01\"),Transform=(Rotation=(X=0.0116,Y=-0.0864,Z=-0.2552,W=0.9629),Translation=(X=0,Y=1.5824,Z=0.2221),Scale3D=(X=1,Y=1,Z=1))),(Item=(Type=Bone,Name=\"dyn_collar_rear_corner_r_01\"),Transform=(Rotation=(X=0.171,Y=-0.0302 ...(truncated)"
    WeightMinimum=0
    Weight=@n2.OutputParam
   )
   RigVMFunctionReturnNode(ExecuteContext=@RigVMFunction_Sequence.F)
   n3 = &RigUnit_SphericalPoseReader(
    ActiveRegionScaleFactors=(PositiveWidth=1,NegativeWidth=1,PositiveHeight=3,NegativeHeight=1)
    ActiveRegionSize=0.1
    Debug=(bDrawDebug=false,DebugScale=25,DebugSegments=20,DebugThickness=0.25)
    DriverAxis=(X=0,Y=-1,Z=0)
    DriverItem=(Type=Bone,Name="head")
    ExecutePin=()
    FalloffRegionScaleFactors=(PositiveWidth=1,NegativeWidth=0.5,PositiveHeight=2,NegativeHeight=1)
    FalloffSize=0.3
    OptionalParentItem=(Type=Bone,Name="neck_02")
    RotationOffset=(X=-15,Y=0,Z=-10)
   )
   > &RigUnit_ModifyTransforms(
    ItemToModify=((Item=(Type=Bone,Name="dyn_collar_side_l_01"),Transform=(Rotation=(X=0,Y=0,Z=0,W=1),Translation=(X=0,Y=0,Z=0),Scale3D=(X=1,Y=1,Z=1))))
    ItemToModify.0.Transform.Rotation=&RigVMFunction_MathQuaternionFromEuler(Euler=(X=0,Y=0,Z=-65)).Result
    Weight=@n3.OutputParam
   )
   > &RigUnit_ModifyTransforms(
    ItemToModify=((Item=(Type=Bone,Name="dyn_collar_rear_corner_l_01"),Transform=(Rotation=(X=0,Y=0,Z=0,W=1),Translation=(X=0,Y=0,Z=0),Scale3D=(X=1,Y=1,Z=1))))
    ItemToModify.0.Transform.Rotation=&RigVMFunction_MathQuaternionFromEuler(Euler=(X=0,Y=0,Z=-45)).Result
    Weight=@n3.OutputParam
   )
   > &RigUnit_ModifyTransforms(
    ItemToModify=((Item=(Type=Bone,Name="dyn_collar_front_01"),Transform=(Rotation=(X=0,Y=0,Z=0,W=1),Translation=(X=0,Y=0,Z=0),Scale3D=(X=1,Y=1,Z=1))))
    ItemToModify.0.Transform.Rotation=&RigVMFunction_MathQuaternionFromEuler(Euler=(X=0,Y=0,Z=-20)).Result
    Weight=@n3.OutputParam
   )
   > n4 = &RigUnit_SphericalPoseReader(
    ActiveRegionScaleFactors=(PositiveWidth=1,NegativeWidth=1,PositiveHeight=1,NegativeHeight=3)
    ActiveRegionSize=0.1
    Debug=(bDrawDebug=false,DebugScale=25,DebugSegments=20,DebugThickness=0.25)
    DriverAxis=(X=0,Y=-1,Z=0)
    DriverItem=(Type=Bone,Name="head")
    FalloffRegionScaleFactors=(PositiveWidth=1,NegativeWidth=0.5,PositiveHeight=1,NegativeHeight=2)
    FalloffSize=0.3
    OptionalParentItem=(Type=Bone,Name="neck_02")
    RotationOffset=(X=165,Y=0,Z=-10)
   )
   > &RigUnit_ModifyTransforms(
    ItemToModify=((Item=(Type=Bone,Name="dyn_collar_side_r_01"),Transform=(Rotation=(X=0,Y=0,Z=0,W=1),Translation=(X=0,Y=0,Z=0),Scale3D=(X=1,Y=1,Z=1))))
    ItemToModify.0.Transform.Rotation=&RigVMFunction_MathQuaternionFromEuler(Euler=(X=0,Y=0,Z=-65)).Result
    Weight=@n4.OutputParam
   )
   > &RigUnit_ModifyTransforms(
    ItemToModify=((Item=(Type=Bone,Name="dyn_collar_rear_corner_r_01"),Transform=(Rotation=(X=0,Y=0,Z=0,W=1),Translation=(X=0,Y=0,Z=0),Scale3D=(X=1,Y=1,Z=1))))
    ItemToModify.0.Transform.Rotation=&RigVMFunction_MathQuaternionFromEuler(Euler=(X=0,Y=0,Z=-45)).Result
    Weight=@n4.OutputParam
   )
   > &RigUnit_ModifyTransforms(
    ItemToModify=((Item=(Type=Bone,Name="dyn_collar_front_01"),Transform=(Rotation=(X=0,Y=0,Z=0,W=1),Translation=(X=0,Y=0,Z=0),Scale3D=(X=1,Y=1,Z=1))))
    ItemToModify.0.Transform.Rotation=&RigVMFunction_MathQuaternionFromEuler(Euler=(X=0,Y=0,Z=-20)).Result
    Weight=@n4.OutputParam
   )
  ]}
 New Function_ContainedGraph {nodes [
   RigVMFunctionEntryNode()
   > &RigUnit_ParentConstraint(
    AdvancedSettings=(RotationOrderForFilter=XZY)
    Child=(Type=Bone,Name="dyn_body_cloth_side_l_01")
    Parents="((Item=(Type=Bone,Name=\"deform_groin_l\"),Weight=0.0084),(Item=(Type=Bone,Name=\"deform_glute_l\"),Weight=0.0135),(Item=(Type=Bone,Name=\"spine_03\"),Weight=0.0153),(Item=(Type=Bone,Name=\"spine_01\"),Weight=0.024),(Item=(Type=Bone,Name=\"spine_02\"),Weight=0.0279) ...(truncated)"
    Weight=1
    bMaintainOffset=true
   )
   > &RigUnit_ParentConstraint(
    AdvancedSettings=(RotationOrderForFilter=XZY)
    Child=(Type=Bone,Name="dyn_body_cloth_side_r_01")
    Parents="((Item=(Type=Bone,Name=\"deform_groin_r\"),Weight=0.0084),(Item=(Type=Bone,Name=\"deform_glute_r\"),Weight=0.0135),(Item=(Type=Bone,Name=\"spine_03\"),Weight=0.0153),(Item=(Type=Bone,Name=\"spine_01\"),Weight=0.024),(Item=(Type=Bone,Name=\"spine_02\"),Weight=0.0279) ...(truncated)"
    Weight=1
    bMaintainOffset=true
   )
   > &RigUnit_ParentConstraint(
    AdvancedSettings=(RotationOrderForFilter=XZY)
    Child=(Type=Bone,Name="dyn_body_cloth_front_01")
    Parents="((Item=(Type=Bone,Name=\"deform_groin_l\"),Weight=0.0088),(Item=(Type=Bone,Name=\"deform_groin_r\"),Weight=0.0088),(Item=(Type=Bone,Name=\"thigh_twist_01_l\"),Weight=0.013),(Item=(Type=Bone,Name=\"thigh_twist_01_r\"),Weight=0.013),(Item=(Type=Bone,Name=\"spine_03\") ...(truncated)"
    Weight=1
    bMaintainOffset=true
   )
   > RigVMFunctionReturnNode()
  ]}
 New Function_ContainedGraph {nodes [
   &RigUnit_AimConstraintLocalSpaceOffset(
    Child=(Type=Bone,Name="dyn_lowerarm_cloth_l_MSTR")
    ExecutePin=()
    Filter=(bX=true,bY=false,bZ=false)
    Parents=((Item=(Type=Bone,Name="root"),Weight=1))
    Weight=1
    WorldUp=(Target=(X=0,Y=0,Z=1),Kind=Location,Space=(Type=Bone,Name="root"))
   )
   > &RigUnit_AimConstraintLocalSpaceOffset(
    Child=(Type=Bone,Name="dyn_lowerarm_cloth_r_MSTR")
    Filter=(bX=true,bY=false,bZ=false)
    Parents=((Item=(Type=Bone,Name="root"),Weight=1))
    Weight=1
    WorldUp=(Target=(X=0,Y=0,Z=1),Kind=Location,Space=(Type=Bone,Name="root"))
    bMaintainOffset=true
   )
   RigVMFunctionEntryNode()
   > RigVMFunction_DebugTransformMutableNoSpace = &RigVMFunction_DebugTransformMutableNoSpace(
    Color=(R=1,G=0,B=0,A=1)
    DebugDrawSettings=(DepthPriority=SDPG_Foreground,Lifetime=-1)
    Mode=Axes
    Scale=10
    Thickness=0
    Transform=(Rotation=(X=0,Y=0,Z=0,W=1),Translation=(X=0,Y=0,Z=0),Scale3D=(X=1,Y=1,Z=1))
    WorldOffset=(Rotation=(X=0,Y=0,Z=0,W=1),Translation=(X=0,Y=0,Z=0),Scale3D=(X=1,Y=1,Z=1))
    bEnabled=true
   )
   > RigVMFunctionReturnNode()
   RigUnit_GetRelativeTransformForItem = &RigUnit_GetRelativeTransformForItem(Child=(Type=Bone,Name="lowerarm_l"), Parent=(Type=Bone,Name="root"), bChildInitial=False, bParentInitial=False)
   Multiply = /Script/RigVM.RigVMFunction_MathVectorMul(B=(X=1,Y=1,Z=0), A=@RigUnit_GetRelativeTransformForItem.RelativeTransform.Translation) {is=struct RigVMUnitNode, template="Multiply::Execute(in A,in B,out Result)"}
   n1 = &RigUnit_GetRelativeTransformForItem(Child=(Type=Control,Name="lowerarm_l_ctrl"), Parent=(Type=Control,Name="root_ctrl"), bChildInitial=False, bParentInitial=False)
   RigVMFunction_MathVectorAngle = /Script/RigVM.RigVMFunction_MathVectorAngle(A=@Multiply.Result, B=@n1.RelativeTransform.Translation) {is=struct RigVMUnitNode, template="AngleBetween::Execute(in A,in B,out Result)"}
   Degrees = /Script/RigVM.RigVMFunction_MathDoubleDeg(Value=@RigVMFunction_MathVectorAngle.Result) {is=struct RigVMUnitNode, template="Degrees::Execute(in Value,out Result)"}
   &RigUnit_GetRelativeTransformForItem(Child=(Type=Bone,Name="lowerarm_l"), Parent=(Type=Bone,Name="root"), bChildInitial=False, bParentInitial=False)
  ]}
 New Function_ContainedGraph {nodes [
   RigVMFunctionEntryNode()
   > &RigUnit_AimConstraintLocalSpaceOffset(
    Child=(Type=Bone,Name="dyn_bow_l_MSTR")
    Filter=(bX=True,bY=True,bZ=True)
    Parents=((Item=(Type=Bone,Name="dyn_rear_skirt_l_03"),Weight=1))
    Weight=1
    WorldUp=(Target=(X=0,Y=0,Z=1),Kind=Location,Space=(Type=Bone,Name="dyn_rear_skirt_l_03"))
    bMaintainOffset=true
   )
   > &RigUnit_AimConstraintLocalSpaceOffset(
    Child=(Type=Bone,Name="dyn_bow_r_MSTR")
    Filter=(bX=True,bY=True,bZ=True)
    Parents=((Item=(Type=Bone,Name="dyn_rear_skirt_r_03"),Weight=1))
    Weight=1
    WorldUp=(Target=(X=0,Y=0,Z=1),Kind=Location,Space=(Type=Bone,Name="dyn_rear_skirt_r_03"))
    bMaintainOffset=true
   )
   > RigVMFunctionReturnNode()
  ]}
 New Function_ContainedGraph {nodes [
   n1 = RigVMFunctionEntryNode()
   > RigVMFunction_Sequence = RigVMAggregateNode()
   Set_Num_2 = RigVMDispatchNode(
    Num=RigVMDispatchNode(Array=@n1.Base Modify Data) {template="DISPATCH_RigVMDispatch_ArrayGetNum(in Array,out Num)"}.Num
    ExecuteContext=@RigVMFunction_Sequence.A
    Array=RigVMVariableNode() {variable=Mirrored}.Value
   ) {template="DISPATCH_RigVMDispatch_ArraySetNum(io Array,in Num)"}
   > For_Each_1 = RigVMDispatchNode(Array=@n1.Base Modify Data) {template="DISPATCH_RigVMDispatch_ArrayIterator(in Array,out Element,out Index,out Count,out Ratio)"}
   > Get Mirror Transform = RigVMFunctionReferenceNode(transform=@For_Each_1.Element.Transform)
   n2 = RigVMDispatchNode(Index=@For_Each_1.Index, Array=@n1.Mirror Modify Data) {template="DISPATCH_RigVMDispatch_ArrayGetAtIndex(in Array,in Index,out Element)"}
   > Set_At_2 = RigVMDispatchNode(
    Element=(Item=(Type=Bone,Name="None"),Transform=(Rotation=(X=0,Y=0,Z=0,W=1),Translation=(X=0,Y=0,Z=0),Scale3D=(X=1,Y=1,Z=1)))
    Array=RigVMVariableNode() {variable=Mirrored}.Value
    Element.Transform.Rotation=/Script/RigVM.RigVMFunction_MathQuaternionMul(B=@n2.Element.Transform.Rotation, A=@"Get Mirror Transform".mirrorTransform.Rotation) {is=struct RigVMUnitNode, template="Multiply::Execute(in A,in B,out Result)"}.Result
    Element.Item=@n2.Element.Item
    Element.Transform.Scale3D=@"Get Mirror Transform".mirrorTransform.Scale3D
    Index=@For_Each_1.Index
    Element.Transform.Translation=&RigVMFunction_MathVectorAdd(B=@n2.Element.Transform.Translation, A=@"Get Mirror Transform".mirrorTransform.Translation) {template="Add::Execute(in A,in B,out Result)"}.Result
   ) {template="DISPATCH_RigVMDispatch_ArraySetAtIndex(io Array,in Index,in Element)"}
   &RigUnit_ModifyTransforms(WeightMinimum=0, Weight=@n1.Weight, ExecutePin=@RigVMFunction_Sequence.B, ItemToModify=@n1.Base Modify Data)
   > &RigUnit_ModifyTransforms(WeightMinimum=0, Weight=@n1."Mirror Weight ", ItemToModify=RigVMVariableNode() {variable=Mirrored}.Value)
   RigVMFunctionReturnNode(ExecuteContext=@RigVMFunction_Sequence.C)
  ]}
 RigVMFunctionLibrary {nodes [
   Body Cloth Deformation = RigVMCollapseNode(ExecuteContext=())
   Bow Deformation = RigVMCollapseNode(ExecuteContext=())
   Collar Deformation = RigVMCollapseNode(ExecuteContext=())
   Full Leg ROM = RigVMCollapseNode(ExecuteContext=(), Interp Output Speed=0, Result Alpha=1)
   Modify Transforms with Mirror = RigVMCollapseNode(Base Modify Data=(), ExecuteContext=(), Mirror Modify Data=())
   Skirt Deformation = RigVMCollapseNode(ExecuteContext=())
   Sleeve Deformation = RigVMCollapseNode(ExecuteContext=())
  ]}
 RigVMFunction_Sequence_1_1_ContainedGraph {nodes [
   n1 = RigVMFunctionEntryNode()
   n2 = &RigVMFunction_Sequence(ExecuteContext=@n1.ExecuteContext)
   n3 = &RigVMFunction_Sequence(ExecuteContext=@n2.B)
   n4 = &RigVMFunction_Sequence(ExecuteContext=@n3.B)
   RigVMFunctionReturnNode(A=@n2.A, B=@n3.A, C=@n4.A, D=@n4.B)
  ]}
 RigVMFunction_Sequence_1_ContainedGraph {nodes [
   n1 = RigVMFunctionEntryNode()
   RigVMFunction_Sequence = &RigVMFunction_Sequence(ExecuteContext=@n1.ExecuteContext)
   n2 = &RigVMFunction_Sequence(ExecuteContext=@RigVMFunction_Sequence.B)
   n3 = &RigVMFunction_Sequence(ExecuteContext=@n2.B)
   n4 = &RigVMFunction_Sequence(ExecuteContext=@n3.B)
   n5 = &RigVMFunction_Sequence(ExecuteContext=@n4.B)
   RigVMFunctionReturnNode(A=@RigVMFunction_Sequence.A, B=@n2.A, C=@n3.A, D=@n4.A, E=@n5.A, F=@n5.B)
  ]}
 RigVMFunction_Sequence_1_ContainedGraph {nodes [
   n1 = RigVMFunctionEntryNode()
   RigVMFunction_Sequence = &RigVMFunction_Sequence(ExecuteContext=@n1.ExecuteContext)
   n2 = &RigVMFunction_Sequence(ExecuteContext=@RigVMFunction_Sequence.B)
   n3 = &RigVMFunction_Sequence(ExecuteContext=@n2.B)
   n4 = &RigVMFunction_Sequence(ExecuteContext=@n3.B)
   n5 = &RigVMFunction_Sequence(ExecuteContext=@n4.B)
   RigVMFunctionReturnNode(A=@RigVMFunction_Sequence.A, B=@n2.A, C=@n3.A, D=@n4.A, E=@n5.A, F=@n5.B)
  ]}
 RigVMFunction_Sequence_1_ContainedGraph {nodes [
   n1 = RigVMFunctionEntryNode()
   RigVMFunction_Sequence = &RigVMFunction_Sequence(ExecuteContext=@n1.ExecuteContext)
   n2 = &RigVMFunction_Sequence(ExecuteContext=@RigVMFunction_Sequence.B)
   RigVMFunctionReturnNode(A=@RigVMFunction_Sequence.A, B=@n2.A, C=@n2.B)
  ]}
 RigVMFunction_Sequence_2_ContainedGraph {nodes [
   n1 = RigVMFunctionEntryNode()
   RigVMFunction_Sequence = &RigVMFunction_Sequence(ExecuteContext=@n1.ExecuteContext)
   n2 = &RigVMFunction_Sequence(ExecuteContext=@RigVMFunction_Sequence.B)
   n3 = &RigVMFunction_Sequence(ExecuteContext=@n2.B)
   n4 = &RigVMFunction_Sequence(ExecuteContext=@n3.B)
   n5 = &RigVMFunction_Sequence(ExecuteContext=@n4.B)
   RigVMFunctionReturnNode(A=@RigVMFunction_Sequence.A, B=@n2.A, C=@n3.A, D=@n4.A, E=@n5.A, F=@n5.B)
  ]}
 RigVMModel {nodes [
   RigUnit_BeginExecution = &RigUnit_BeginExecution()
   FN_DeformRig_FNC = RigVMFunctionReferenceNode(
    DeformRig_ Curve="(deformRig_l_upperarm_up_alpha_72_#1=False,deformRig_r_upperarm_up_alpha_73_#2=False,deformRig_l_upperarm_fr_alpha_74_#3=False,deformRig_r_upperarm_fr_alpha_75_CB51B8 ...(truncated)"
    DeformRig_Bool="(upperarm_1_#4=True,elbow_3_#5=True,wrist_7_#6=True,groin_8_#7=True,glute_17_#8=False,knee_9_40C408324CA26 ...(truncated)"
    DeformRig_Debug=(upperarm_debug_18_#4=False,wrist_debug_20_#6=False,lat_debug_26_#9=False,groinGlute_debug_27_#7=False)
    DeformRig_Value="(upperarm_up_18_#10=(Rotation=(X=0,Y=0,Z=0,W=1),Translation=(X=0,Y=0,Z=8),Scale3D=(X=0.7,Y=1,Z=1)),upperarm_fwd_fr_19_#11=(Rotation=(X=0,Y=-0.3827,Z=0,W=0.9239),Translation=(X=-8,Y=15,Z=0),Scale3D=( ...(truncated)"
    ExecuteContext=@RigUnit_BeginExecution.ExecutePin
   )
   > Skirt Deformation = RigVMFunctionReferenceNode()
   > Collar Deformation = RigVMFunctionReferenceNode()
   > Body Cloth Deformation = RigVMFunctionReferenceNode()
   Bow Deformation = RigVMFunctionReferenceNode(ExecuteContext=())
   &RigUnit_ModifyTransforms(
    ExecutePin=()
    ItemToModify=((Item=(Type=Bone,Name="dyn_lowerarm_cloth_r_MSTR"),Transform=(Rotation=(X=-0.4483,Y=0.054,Z=0.0496,W=0.8909),Translation=(X=-0.1726,Y=2.3605,Z=-1.9732),Scale3D=(X=1,Y=1,Z=1))))
    Weight=1
    WeightMinimum=0
   )
   Sleeve Deformation = RigVMFunctionReferenceNode(ExecuteContext=())
  ]}
]
hierarchy [
 root {loc=<0 0 0>, rot=<0 0 0 1>}
 pelvis {parent=root, loc=<0 -0.7071 93.1191>, rot=<0.0127 0.707 -0.0127 -0.707>}
 spine_01 {parent=pelvis, loc=<6.7494 -0.6388 0>, rot=<0 0 0.034 -0.9994>}
 spine_02 {parent=spine_01, loc=<7.252 -0.035 0>, rot=<0 0 0.0238 -0.9997>}
 spine_03 {parent=spine_02, loc=<6.0106 0 0>, rot=<0 0 0.0149 -0.9999>}
 spine_04 {parent=spine_03, loc=<8.1023 0 0>, rot=<0 0 -0.0132 -0.9999>}
 spine_05 {parent=spine_04, loc=<10.0142 0 0>, rot=<0 0 -0.0015 -1>}
 attach_backpack {parent=spine_05, loc=<-0.123 -13.473 0>, rot=<-0.0093 -0.707 0.0093 -0.707>}
 neck_01 {parent=spine_05, loc=<11.3138 0.1529 0>, rot=<0 0 -0.2003 -0.9797>}
 neck_02 {parent=neck_01, loc=<4.5627 0 0>, rot=<0 0 0.0504 -0.9987>}
 head {parent=neck_02, loc=<5.3914 0 0>, rot=<0 0 0.115 -0.9934>}
 spine_05_ctrl {controlType=EULER_TRANSFORM, parent=neck_01, type=Control, loc=<0 0 0>, rot=<0 0 0 1>}
 clavicle_l {parent=spine_05, loc=<4.3575 -0.1466 -2.3223>, rot=<0.0465 -0.7289 -0.009 -0.6829>}
 upperarm_l {parent=clavicle_l, loc=<15.5111 0 0>, rot=<0.0226 -0.3234 -0.0088 -0.9459>}
 lowerarm_l {parent=upperarm_l, loc=<26.6514 0 0>, rot=<0.0391 -0.0099 -0.2453 -0.9686>}
 hand_l {parent=lowerarm_l, loc=<22.3491 0 0>, rot=<-0.7081 -0.0507 0.0975 -0.6975>}
 thumb_01_l {parent=hand_l, loc=<1.3118 -0.3191 -2.459>, rot=<0.6546 -0.4042 0.0664 -0.6354>}
 thumb_02_l {parent=thumb_01_l, loc=<5.9336 -0.0448 0.0401>, rot=<-0.0037 0.0213 0.298 -0.9543>}
 thumb_03_l {parent=thumb_02_l, loc=<2.2223 0.3682 -0.0907>, rot=<0.0118 -0.013 0.0066 -0.9998>}
 index_metacarpal_l {parent=hand_l, loc=<1.6535 0.0183 -2.96>, rot=<0.0158 -0.0124 -0.0173 -0.9996>}
 index_01_l {parent=index_metacarpal_l, loc=<8.3008 0.0357 -0.0524>, rot=<-0.0044 0.0485 0.0924 -0.9945>}
 index_02_l {parent=index_01_l, loc=<3.3587 -0.3004 0.0452>, rot=<-0.002 0.0027 0.3225 -0.9466>}
 index_03_l {parent=index_02_l, loc=<2.5578 -0.123 -0.0252>, rot=<-0.002 0.0067 0.2558 -0.9667>}
 middle_metacarpal_l {parent=hand_l, loc=<2.1856 -0.1725 -0.5984>, rot=<0.006 0.0168 -0.0329 -0.9993>}
 middle_01_l {parent=middle_metacarpal_l, loc=<7.3275 -0.2447 0.0129>, rot=<-0.0208 0.0042 0.0858 -0.9961>}
 middle_02_l {parent=middle_01_l, loc=<4.0123 0.1401 0.1175>, rot=<-0.0218 0.0258 0.3512 -0.9357>}
 middle_03_l {parent=middle_02_l, loc=<2.6077 0.0534 0.1373>, rot=<0.0082 0.0188 0.1708 -0.9851>}
 ring_metacarpal_l {parent=hand_l, loc=<2.2074 -0.3625 1.3638>, rot=<-0.0284 0.0722 -0.0458 -0.9959>}
 ring_01_l {parent=ring_metacarpal_l, loc=<7.0295 -0.6197 -0.1446>, rot=<-0.0489 -0.0252 0.1892 -0.9804>}
 ring_02_l {parent=ring_01_l, loc=<3.6174 0.4647 -0.1027>, rot=<0 0 0.2806 -0.9598>}
 ring_03_l {parent=ring_02_l, loc=<2.4152 0.114 0.0199>, rot=<0 0 0.1189 -0.9929>}
 pinky_metacarpal_l {parent=hand_l, loc=<1.7616 -0.9753 3.3814>, rot=<-0.0397 0.0814 -0.069 -0.9935>}
 pinky_01_l {parent=pinky_metacarpal_l, loc=<7.1709 -0.4831 -0.1094>, rot=<-0.0654 -0.0035 0.2242 -0.9723>}
 pinky_02_l {parent=pinky_01_l, loc=<2.8274 0.0893 -0.1043>, rot=<0.0037 0.0011 0.2854 -0.9584>}
 pinky_03_l {parent=pinky_02_l, loc=<2.2046 0.0307 -0.0078>, rot=<-0.0002 0.0098 0.1445 -0.9895>}
 deform_wrist_dn_base_l {parent=hand_l, loc=<0 -2.2342 0>, rot=<0 0 0 -1>}
 deform_wrist_dn_l {parent=deform_wrist_dn_base_l, loc=<0 0 0>, rot=<0 0 0 1>}
 deform_wrist_up_base_l {parent=hand_l, loc=<0 2.2682 0>, rot=<0 0 0 -1>}
 deform_wrist_up_l {parent=deform_wrist_up_base_l, loc=<0 0 0>, rot=<0 0 0 1>}
 lowerarm_twist_02_l {parent=lowerarm_l, loc=<9.6199 0 0>, rot=<0 0 0 -1>}
 lowerarm_twist_01_l {parent=lowerarm_l, loc=<17.8549 0 0>, rot=<0 0 0 -1>}
 deform_elbow_in_base_l {parent=lowerarm_l, loc=<0.2112 5.259 0.8274>, rot=<0 0 0 -1>}
 deform_elbow_in_l {parent=deform_elbow_in_base_l, loc=<0 0 0>, rot=<0 0 0 1>}
 deform_elbow_out_base_l {parent=lowerarm_l, loc=<0 -4 0>, rot=<0 0 0 -1>}
 deform_elbow_out_l {parent=deform_elbow_out_base_l, loc=<0 0 0>, rot=<0 0 0 1>}
 dyn_lowerarm_cloth_l_MSTR {parent=lowerarm_l, loc=<5.362 0 0>, rot=<0 0 0 1>}
 dyn_lowerarm_cloth_up_l {parent=dyn_lowerarm_cloth_l_MSTR, loc=<0 0 0.1>, rot=<0 0.0029 0 -1>}
 dyn_lowerarm_cloth_down_l_01 {parent=dyn_lowerarm_cloth_l_MSTR, loc=<0 0 -0.1>, rot=<0 0.0096 0 -1>}
 dyn_lowerarm_cloth_down_l_02 {parent=dyn_lowerarm_cloth_down_l_01, loc=<0 0 -5.455>, rot=<0 0 0 1>}
 upperarm_twist_01_l {parent=upperarm_l, loc=<0.0041 0 0>, rot=<0 0 0 -1>}
 deform_upperarm_bk_base_l {parent=upperarm_twist_01_l, loc=<1.0813 -7.2193 -1.9453>, rot=<0 0 0 1>}
 deform_upperarm_bk_l {parent=deform_upperarm_bk_base_l, loc=<0 0 0>, rot=<0 0 0 1>}
 deform_upperarm_fr_base_l {parent=upperarm_twist_01_l, loc=<-0.3691 7.0506 -2.765>, rot=<0 0 0 1>}
 deform_upperarm_fr_l {parent=deform_upperarm_fr_base_l, loc=<0 0 0>, rot=<0 0 0 1>}
 deform_upperarm_up_base_l {parent=upperarm_twist_01_l, loc=<-3.2216 -1.1391 3.3311>, rot=<0 0 0 1>}
 deform_upperarm_up_l {parent=deform_upperarm_up_base_l, loc=<0 0 0>, rot=<0 0 0 1>}
 upperarm_twist_02_l {parent=upperarm_l, loc=<13.7941 0 0>, rot=<0 0 0 -1>}
 attach_cape {parent=spine_05, loc=<9.017 -11.714 0>, rot=<0.0616 -0.7044 -0.0616 -0.7044>}
 clavicle_r {parent=spine_05, loc=<4.3575 -0.1466 2.3223>, rot=<0.7289 0.0465 -0.6829 0.009>}
 upperarm_r {parent=clavicle_r, loc=<-15.5111 0 0>, rot=<0.0226 -0.3234 -0.0088 -0.9459>}
 lowerarm_r {parent=upperarm_r, loc=<-26.6514 0 0>, rot=<0.0391 -0.0099 -0.2453 -0.9686>}
 hand_r {parent=lowerarm_r, loc=<-22.3491 0 0>, rot=<-0.7081 -0.0507 0.0975 -0.6975>}
 thumb_01_r {parent=hand_r, loc=<-1.3118 0.3191 2.459>, rot=<0.6546 -0.4042 0.0664 -0.6354>}
 thumb_02_r {parent=thumb_01_r, loc=<-5.9336 0.0448 -0.0401>, rot=<-0.0037 0.0213 0.298 -0.9543>}
 thumb_03_r {parent=thumb_02_r, loc=<-2.2223 -0.3682 0.0907>, rot=<0.0118 -0.013 0.0066 -0.9998>}
 index_metacarpal_r {parent=hand_r, loc=<-1.6535 -0.0183 2.96>, rot=<0.0158 -0.0124 -0.0173 -0.9996>}
 index_01_r {parent=index_metacarpal_r, loc=<-8.3008 -0.0357 0.0524>, rot=<-0.0044 0.0485 0.0924 -0.9945>}
 index_02_r {parent=index_01_r, loc=<-3.3587 0.3004 -0.0452>, rot=<-0.002 0.0027 0.3225 -0.9466>}
 index_03_r {parent=index_02_r, loc=<-2.5578 0.123 0.0252>, rot=<-0.002 0.0067 0.2558 -0.9667>}
 middle_metacarpal_r {parent=hand_r, loc=<-2.1856 0.1725 0.5984>, rot=<0.006 0.0168 -0.0329 -0.9993>}
 middle_01_r {parent=middle_metacarpal_r, loc=<-7.3275 0.2447 -0.0129>, rot=<-0.0208 0.0042 0.0858 -0.9961>}
 middle_02_r {parent=middle_01_r, loc=<-4.0123 -0.1401 -0.1175>, rot=<-0.0218 0.0258 0.3512 -0.9357>}
 middle_03_r {parent=middle_02_r, loc=<-2.6077 -0.0534 -0.1373>, rot=<0.0082 0.0188 0.1708 -0.9851>}
 ring_metacarpal_r {parent=hand_r, loc=<-2.2074 0.3625 -1.3638>, rot=<-0.0284 0.0722 -0.0458 -0.9959>}
 ring_01_r {parent=ring_metacarpal_r, loc=<-7.0295 0.6197 0.1446>, rot=<-0.0489 -0.0252 0.1892 -0.9804>}
 ring_02_r {parent=ring_01_r, loc=<-3.6174 -0.4647 0.1027>, rot=<0 0 0.2806 -0.9598>}
 ring_03_r {parent=ring_02_r, loc=<-2.4152 -0.114 -0.0199>, rot=<0 0 0.1189 -0.9929>}
 pinky_metacarpal_r {parent=hand_r, loc=<-1.7616 0.9753 -3.3814>, rot=<-0.0397 0.0814 -0.069 -0.9935>}
 pinky_01_r {parent=pinky_metacarpal_r, loc=<-7.1709 0.4831 0.1094>, rot=<-0.0654 -0.0035 0.2242 -0.9723>}
 pinky_02_r {parent=pinky_01_r, loc=<-2.8274 -0.0893 0.1043>, rot=<0.0037 0.0011 0.2854 -0.9584>}
 pinky_03_r {parent=pinky_02_r, loc=<-2.2046 -0.0307 0.0078>, rot=<-0.0002 0.0098 0.1445 -0.9895>}
 deform_wrist_dn_base_r {parent=hand_r, loc=<-0.0004 2.2345 -0.0001>, rot=<0 0 0 -1>}
 deform_wrist_dn_r {parent=deform_wrist_dn_base_r, loc=<0 0 0>, rot=<0 0 0 1>}
 deform_wrist_up_base_r {parent=hand_r, loc=<0.0003 -2.2684 0.0001>, rot=<0 0 0 -1>}
 deform_wrist_up_r {parent=deform_wrist_up_base_r, loc=<0 0 0>, rot=<0 0 0 1>}
 lowerarm_twist_02_r {parent=lowerarm_r, loc=<-9.6199 0 0>, rot=<0 0 0 -1>}
 lowerarm_twist_01_r {parent=lowerarm_r, loc=<-17.8549 0 0>, rot=<0 0 0 -1>}
 deform_elbow_in_base_r {parent=lowerarm_r, loc=<-0.2111 -5.2591 -0.8275>, rot=<0 0 0 -1>}
 deform_elbow_in_r {parent=deform_elbow_in_base_r, loc=<-0.0001 0 0.0002>, rot=<0 0 0 1>}
 deform_elbow_out_base_r {parent=lowerarm_r, loc=<0 4 -0.0001>, rot=<0 0 0 -1>}
 deform_elbow_out_r {parent=deform_elbow_out_base_r, loc=<0 0 0>, rot=<0 0 0 1>}
 dyn_lowerarm_cloth_r_MSTR {parent=lowerarm_r, loc=<-5.3628 0.0001 0.0003>, rot=<0 0 0 1>}
 dyn_lowerarm_cloth_up_r {parent=dyn_lowerarm_cloth_r_MSTR, loc=<0 0 -0.1>, rot=<0 0.0029 0 -1>}
 dyn_lowerarm_cloth_down_r_01 {parent=dyn_lowerarm_cloth_r_MSTR, loc=<0 -0.0001 0.1>, rot=<0 0.0096 0 -1>}
 dyn_lowerarm_cloth_down_r_02 {parent=dyn_lowerarm_cloth_down_r_01, loc=<-0.0001 0.0001 5.4551>, rot=<0 0 0 -1>}
 upperarm_twist_01_r {parent=upperarm_r, loc=<-0.0041 0 0>, rot=<0 0 0 -1>}
 deform_upperarm_bk_base_r {parent=upperarm_twist_01_r, loc=<-1.0814 7.2192 1.9455>, rot=<0 0 0 1>}
 deform_upperarm_bk_r {parent=deform_upperarm_bk_base_r, loc=<0 0 0>, rot=<0 0 0 1>}
 deform_upperarm_fr_base_r {parent=upperarm_twist_01_r, loc=<0.3689 -7.0506 2.7652>, rot=<0 0 0 1>}
 deform_upperarm_fr_r {parent=deform_upperarm_fr_base_r, loc=<0 0 0>, rot=<0 0 0 1>}
 deform_upperarm_up_base_r {parent=upperarm_twist_01_r, loc=<3.2214 1.1391 -3.3308>, rot=<0 0 0 1>}
 deform_upperarm_up_r {parent=deform_upperarm_up_base_r, loc=<0 0 0>, rot=<0 0 0 -1>}
 upperarm_twist_02_r {parent=upperarm_r, loc=<-13.7941 0 0>, rot=<0 0 0 -1>}
 dyn_collar_side_l_01 {parent=spine_05, loc=<12.7515 4.1152 -8.5426>, rot=<-0.7642 -0.1738 -0.103 -0.6126>}
 dyn_collar_side_l_02 {parent=dyn_collar_side_l_01, loc=<2.9698 0.0197 0.0049>, rot=<0.005 -0.0028 0.0111 -0.9999>}
 dyn_collar_front_01 {parent=spine_05, loc=<8.7311 10.6351 0>, rot=<0.9797 0.2003 0 0>}
 dyn_collar_front_02 {parent=dyn_collar_front_01, loc=<3.4982 0.0079 0.002>, rot=<0.005 -0.0028 0.0111 -0.9999>}
 dyn_collar_rear_01 {parent=spine_05, loc=<13.8346 -5.4729 0>, rot=<0 0 -0.2003 -0.9797>}
 dyn_collar_rear_02 {parent=dyn_collar_rear_01, loc=<4.2209 -0.0082 -0.0021>, rot=<0.005 -0.0028 0.0111 -0.9999>}
 dyn_collar_rear_corner_l_01 {parent=spine_05, loc=<14.7166 -3.6185 -6.4414>, rot=<-0.2918 0.025 -0.2233 -0.9297>}
 dyn_collar_rear_corner_l_02 {parent=dyn_collar_rear_corner_l_01, loc=<3.0325 0.0183 0.0046>, rot=<0.005 -0.0028 0.0111 -0.9999>}
 dyn_collar_side_r_01 {parent=spine_05, loc=<11.7218 3.676 8.4803>, rot=<0.1738 -0.7642 -0.6126 0.103>}
 dyn_collar_side_r_02 {parent=dyn_collar_side_r_01, loc=<-2.581 -0.0284 -0.0072>, rot=<0.005 -0.0028 0.0111 -0.9999>}
 dyn_collar_rear_corner_r_01 {parent=spine_05, loc=<14.078 -3.903 6.5669>, rot=<0.025 0.2918 0.9297 -0.2233>}
 dyn_collar_rear_corner_r_02 {parent=dyn_collar_rear_corner_r_01, loc=<-3.2622 -0.0129 -0.0033>, rot=<0.005 -0.0028 0.0111 -0.9999>}
 deform_pec_aim_base_l {parent=spine_04, loc=<7.8728 12.6388 -2.4123>, rot=<0.0102 -0.4602 0.3979 -0.7936>}
 deform_pec_aim_l {parent=deform_pec_aim_base_l, loc=<0 0 0>, rot=<0 0 0 1>}
 deform_pec_base_l {parent=deform_pec_aim_l, loc=<10.766 0 0>, rot=<0 0 0 1>}
 deform_pec_l {parent=deform_pec_base_l, loc=<0.055 0 0>, rot=<0 0 0 1>}
 deform_pec_aim_base_r {parent=spine_04, loc=<7.8723 12.6388 2.4123>, rot=<-0.4602 -0.0102 0.7936 0.3979>}
 deform_pec_aim_r {parent=deform_pec_aim_base_r, loc=<0 0 0>, rot=<0 0 0 1>}
 deform_pec_base_r {parent=deform_pec_aim_r, loc=<-10.7662 -0.0006 -0.0007>, rot=<0 0 0 1>}
 deform_pec_r {parent=deform_pec_base_r, loc=<0 0 0>, rot=<0 0 0 1>}
 dyn_body_cloth_side_l_01 {parent=spine_01, loc=<2.751 -0.3075 -7.4348>, rot=<0.0616 -0.6947 0.0451 -0.7153>}
 dyn_body_cloth_side_l_02 {parent=dyn_body_cloth_side_l_01, loc=<7.0565 0 0>, rot=<0 0 0.1496 -0.9888>}
 dyn_body_cloth_side_r_01 {parent=spine_01, loc=<2.7513 -0.3075 7.4348>, rot=<-0.6947 -0.0616 0.7153 0.0451>}
 dyn_body_cloth_side_r_02 {parent=dyn_body_cloth_side_r_01, loc=<-7.0564 0 0>, rot=<0 0 0.1496 -0.9888>}
 dyn_body_cloth_front_01 {parent=spine_01, loc=<2.3695 6.369 0.1205>, rot=<-0.4396 -0.5485 -0.4868 -0.5186>}
 dyn_body_cloth_front_l_02 {parent=dyn_body_cloth_front_01, loc=<7.0565 0 0>, rot=<0 0 0.1496 -0.9888>}
 thigh_l {parent=pelvis, loc=<-2.5693 -0.0109 -7.7501>, rot=<0.0012 0.0572 0.0155 -0.9982>}
 calf_l {parent=thigh_l, loc=<-39.0484 0 0>, rot=<0.0006 -0.0161 -0.0391 -0.9991>}
 foot_l {parent=calf_l, loc=<-42.561 0.1077 0.4292>, rot=<0.0418 -0.0455 0.0343 -0.9975>}
 ball_l {parent=foot_l, loc=<-6.1599 13.8264 0.6806>, rot=<0.0001 -0.0001 -0.7063 -0.7079>}
 calf_twist_02_l {parent=calf_l, loc=<-18.8064 -0.0967 0.0397>, rot=<0.0237 0.0095 -0.0035 -0.9997>}
 calf_twist_01_l {parent=calf_l, loc=<-35.813 -0.0967 0.0397>, rot=<0.0237 0.0095 -0.0035 -0.9997>}
 deform_knee_in_base_l {parent=calf_l, loc=<1.298 -5.2957 -0.0496>, rot=<0 0 0 -1>}
 deform_knee_in_l {parent=deform_knee_in_base_l, loc=<0 0 0>, rot=<0 0 0 1>}
 deform_knee_out_base_l {parent=calf_l, loc=<0 4 0>, rot=<0 0 0 -1>}
 deform_knee_out_l {parent=deform_knee_out_base_l, loc=<0 0 0>, rot=<0 0 0 1>}
 thigh_twist_01_l {parent=thigh_l, loc=<-21.6303 0 0>, rot=<0.0013 0 0 -1>}
 deform_groin_base_l {parent=thigh_l, loc=<-2.1908 8.1142 -1.7442>, rot=<0 0 0 -1>}
 deform_groin_l {parent=deform_groin_base_l, loc=<0 0 0>, rot=<0 0 0 1>}
 thigh_r {parent=pelvis, loc=<-2.5693 -0.0109 7.7501>, rot=<0.0572 -0.0012 0.9982 0.0155>}
 calf_r {parent=thigh_r, loc=<39.0484 0 0>, rot=<0.0006 -0.0161 -0.0391 -0.9991>}
 foot_r {parent=calf_r, loc=<42.561 -0.1077 -0.4292>, rot=<0.0418 -0.0455 0.0343 -0.9975>}
 ball_r {parent=foot_r, loc=<6.1599 -13.8264 -0.6806>, rot=<0.0001 -0.0001 -0.7063 -0.7079>}
 calf_twist_02_r {parent=calf_r, loc=<18.8064 0.0967 -0.0397>, rot=<0.0237 0.0095 -0.0035 -0.9997>}
 calf_twist_01_r {parent=calf_r, loc=<35.813 0.0967 -0.0397>, rot=<0.0237 0.0095 -0.0035 -0.9997>}
 deform_knee_in_base_r {parent=calf_r, loc=<-1.298 5.2957 0.0496>, rot=<0 0 0 -1>}
 deform_knee_in_r {parent=deform_knee_in_base_r, loc=<0 0 0>, rot=<0 0 0 1>}
 deform_knee_out_base_r {parent=calf_r, loc=<0 -4 0>, rot=<0 0 0 -1>}
 deform_knee_out_r {parent=deform_knee_out_base_r, loc=<0 0 0>, rot=<0 0 0 1>}
 thigh_twist_01_r {parent=thigh_r, loc=<21.6303 0 0>, rot=<0.0013 0 0 -1>}
 deform_groin_base_r {parent=thigh_r, loc=<2.1908 -8.1142 1.7442>, rot=<0 0 0 -1>}
 deform_groin_r {parent=deform_groin_base_r, loc=<0 0 0>, rot=<0 0 0 1>}
 deform_glute_base_l {parent=pelvis, loc=<-8.8064 -8.8222 -8.4287>, rot=<0.0275 0.0501 -0.4472 -0.8926>}
 deform_glute_l {parent=deform_glute_base_l, loc=<0 0 0>, rot=<0 0 0 1>}
 deform_glute_base_r {parent=pelvis, loc=<-8.8062 -8.8222 8.4287>, rot=<0.0501 -0.0275 0.8926 -0.4472>}
 deform_glute_r {parent=deform_glute_base_r, loc=<0 0 0>, rot=<0 0 0 1>}
 dyn_bow_loop_MSTR {parent=pelvis, loc=<2.7402 -12.3909 0>, rot=<0.5369 -0.5564 -0.4403 -0.4564>}
 dyn_bow_l_MSTR {parent=dyn_bow_loop_MSTR, loc=<4.8397 4.3107 1.1159>, rot=<-0.3374 -0.383 -0.5161 -0.6879>}
 dyn_bow_l_01 {parent=dyn_bow_l_MSTR, loc=<0 0 0>, rot=<0 0 0 1>}
 dyn_bow_l_02 {parent=dyn_bow_l_01, loc=<10 0 0>, rot=<0.0007 -0.0111 -0.0599 -0.9981>}
 dyn_bow_l_03 {parent=dyn_bow_l_02, loc=<10 0 0>, rot=<0 0 0 1>}
 dyn_bow_l_04 {parent=dyn_bow_l_03, loc=<10 0 0>, rot=<0 0 0 1>}
 dyn_bow_l_05 {parent=dyn_bow_l_04, loc=<10 0 0>, rot=<0 0 0 1>}
 dyn_bow_r_MSTR {parent=dyn_bow_loop_MSTR, loc=<-4.8973 4.3107 -0.8269>, rot=<0.5997 -0.4401 0.5102 -0.4318>}
 dyn_bow_r_01 {parent=dyn_bow_r_MSTR, loc=<0 0 0>, rot=<0 0 0 1>}
 dyn_bow_r_02 {parent=dyn_bow_r_01, loc=<-10 0 0>, rot=<0.0007 -0.0111 -0.0599 -0.9981>}
 dyn_bow_r_03 {parent=dyn_bow_r_02, loc=<-10 0 0>, rot=<0 0 0 1>}
 dyn_bow_r_04 {parent=dyn_bow_r_03, loc=<-10 0 0>, rot=<0 0 0 1>}
 dyn_bow_r_05 {parent=dyn_bow_r_04, loc=<-10 0.0001 0>, rot=<0 0 0 1>}
 dyn_bow_loop_l_01 {parent=dyn_bow_loop_MSTR, loc=<0.9018 0 0.1799>, rot=<-0.7065 -0.0467 -0.0794 -0.7017>}
 dyn_bow_loop_l_02 {parent=dyn_bow_loop_l_01, loc=<7.0565 0 0>, rot=<0 0 0.1496 -0.9888>}
 dyn_bow_loop_r_01 {parent=dyn_bow_loop_MSTR, loc=<-0.9018 0 -0.1799>, rot=<0.679 0.0604 0.1831 -0.7084>}
 dyn_bow_loop_r_02 {parent=dyn_bow_loop_r_01, loc=<-7.0564 0 0>, rot=<0 0 0.1496 -0.9888>}
 dyn_front_skirt_l_MSTR {parent=pelvis, loc=<-7.3501 8.6733 -5.2119>, rot=<-0.0036 0.0516 0.9984 0.0205>}
 dyn_front_skirt_l_01 {parent=dyn_front_skirt_l_MSTR, loc=<0 0 0>, rot=<0 0 0 1>}
 dyn_front_skirt_l_02 {parent=dyn_front_skirt_l_01, loc=<20.5429 0 0>, rot=<0 0 0 1>}
 dyn_front_skirt_l_03 {parent=dyn_front_skirt_l_02, loc=<14.7382 0 0>, rot=<0 0 0 1>}
 dyn_front_skirt_l_04 {parent=dyn_front_skirt_l_03, loc=<13.2954 0 0>, rot=<0 0 0 1>}
 dyn_front_skirt_l_05 {parent=dyn_front_skirt_l_04, loc=<12.098 0 0>, rot=<0 0 0 1>}
 dyn_side_skirt_l_MSTR {parent=pelvis, loc=<-6.1917 -0.9894 -14.8401>, rot=<0.0289 -0.7211 -0.6916 -0.0301>}
 dyn_side_skirt_l_01 {parent=dyn_side_skirt_l_MSTR, loc=<0 0 0>, rot=<0 0 0 1>}
 dyn_side_skirt_l_02 {parent=dyn_side_skirt_l_01, loc=<20.5429 0 0>, rot=<0 0 0 1>}
 dyn_side_skirt_l_03 {parent=dyn_side_skirt_l_02, loc=<14.7382 0 0>, rot=<0 0 0 1>}
 dyn_side_skirt_l_04 {parent=dyn_side_skirt_l_03, loc=<13.2954 0 0>, rot=<0 0 0 1>}
 dyn_side_skirt_l_05 {parent=dyn_side_skirt_l_04, loc=<12.098 0 0>, rot=<0 0 0 1>}
 dyn_rear_skirt_l_MSTR {parent=pelvis, loc=<-5.1552 -10.8873 -4.0696>, rot=<0.0177 -0.9986 -0.0482 -0.0105>}
 dyn_rear_skirt_l_01 {parent=dyn_rear_skirt_l_MSTR, loc=<0 0 0>, rot=<0 0 0 1>}
 dyn_rear_skirt_l_02_MSTR {parent=dyn_rear_skirt_l_01, loc=<20.5429 0 0>, rot=<0 0 0 1>}
 dyn_rear_skirt_l_02 {parent=dyn_rear_skirt_l_02_MSTR, loc=<0 0 0>, rot=<0 0 0 1>}
 dyn_rear_skirt_l_03 {parent=dyn_rear_skirt_l_02, loc=<14.7382 0 0>, rot=<0 0 0 1>}
 dyn_rear_skirt_l_04 {parent=dyn_rear_skirt_l_03, loc=<13.2954 0 0>, rot=<0 0 0 1>}
 dyn_rear_skirt_l_05 {parent=dyn_rear_skirt_l_04, loc=<12.098 0 0>, rot=<0 0 0 1>}
 dyn_rear_skirt_r_MSTR {parent=pelvis, loc=<-5.1552 -10.8873 4.0696>, rot=<0.9986 0.0177 -0.0105 0.0482>}
 dyn_rear_skirt_r_01 {parent=dyn_rear_skirt_r_MSTR, loc=<0 0 0>, rot=<0 0 0 1>}
 dyn_rear_skirt_r_02_MSTR {parent=dyn_rear_skirt_r_01, loc=<-20.5428 0 0>, rot=<0 0 0 1>}
 dyn_rear_skirt_r_02 {parent=dyn_rear_skirt_r_02_MSTR, loc=<0 0 0>, rot=<0 0 0 1>}
 dyn_rear_skirt_r_03 {parent=dyn_rear_skirt_r_02, loc=<-14.7382 0 0.0001>, rot=<0 0 0 1>}
 dyn_rear_skirt_r_04 {parent=dyn_rear_skirt_r_03, loc=<-13.2954 0 0>, rot=<0 0 0 1>}
 dyn_rear_skirt_r_05 {parent=dyn_rear_skirt_r_04, loc=<-12.0979 0 0>, rot=<0 0 0 1>}
 dyn_side_skirt_r_MSTR {parent=pelvis, loc=<-6.1917 -0.9894 14.8401>, rot=<-0.7211 -0.0289 0.0301 -0.6916>}
 dyn_side_skirt_r_01 {parent=dyn_side_skirt_r_MSTR, loc=<0 0 0>, rot=<0 0 0 1>}
 dyn_side_skirt_r_02 {parent=dyn_side_skirt_r_01, loc=<-20.5429 0 0>, rot=<0 0 0 1>}
 dyn_side_skirt_r_03 {parent=dyn_side_skirt_r_02, loc=<-14.7382 0 0>, rot=<0 0 0 1>}
 dyn_side_skirt_r_04 {parent=dyn_side_skirt_r_03, loc=<-13.2955 -0.0001 0>, rot=<0 0 0 1>}
 dyn_side_skirt_r_05 {parent=dyn_side_skirt_r_04, loc=<-12.0979 0 0>, rot=<0 0 0 1>}
 dyn_front_skirt_r_MSTR {parent=pelvis, loc=<-7.3501 8.6733 5.2119>, rot=<-0.0516 -0.0036 0.0205 -0.9984>}
 dyn_front_skirt_r_01 {parent=dyn_front_skirt_r_MSTR, loc=<0 0 0>, rot=<0 0 0 1>}
 dyn_front_skirt_r_02 {parent=dyn_front_skirt_r_01, loc=<-20.5429 0 0>, rot=<0 0 0 1>}
 dyn_front_skirt_r_03 {parent=dyn_front_skirt_r_02, loc=<-14.7382 0 0>, rot=<0 0 0 1>}
 dyn_front_skirt_r_04 {parent=dyn_front_skirt_r_03, loc=<-13.2955 0 -0.0001>, rot=<0 0 0 1>}
 dyn_front_skirt_r_05 {parent=dyn_front_skirt_r_04, loc=<-12.098 0 0>, rot=<0 0 0 1>}
 dyn_front_skirt_corner_l_MSTR {parent=pelvis, loc=<-7.2351 6.6377 -11.9298>, rot=<-0.022 0.3672 0.9294 0.03>}
 dyn_front_skirt_corner_l_01 {parent=dyn_front_skirt_corner_l_MSTR, loc=<0 0 0>, rot=<0 0 0 1>}
 dyn_front_skirt_corner_l_02 {parent=dyn_front_skirt_corner_l_01, loc=<20.5429 0 0>, rot=<0 0 0 1>}
 dyn_front_skirt_corner_l_03 {parent=dyn_front_skirt_corner_l_02, loc=<14.7382 0 0>, rot=<0 0 0 1>}
 dyn_front_skirt_corner_l_04 {parent=dyn_front_skirt_corner_l_03, loc=<13.2954 0 0>, rot=<0 0 0 1>}
 dyn_front_skirt_corner_l_05 {parent=dyn_front_skirt_corner_l_04, loc=<12.098 0 0>, rot=<0 0 0 1>}
 dyn_rear_skirt_corner_l_MSTR {parent=pelvis, loc=<-5.021 -7.9583 -11.7894>, rot=<0.0247 -0.9718 -0.2332 -0.0233>}
 dyn_rear_skirt_corner_l_01 {parent=dyn_rear_skirt_corner_l_MSTR, loc=<0 0 0>, rot=<0 0 0 1>}
 dyn_rear_skirt_corner_l_02_MSTR {parent=dyn_rear_skirt_corner_l_01, loc=<20.5429 0 0>, rot=<0 0 0 1>}
 dyn_rear_skirt_corner_l_02 {parent=dyn_rear_skirt_corner_l_02_MSTR, loc=<0 0 0>, rot=<0 0 0 1>}
 dyn_rear_skirt_corner_l_03 {parent=dyn_rear_skirt_corner_l_02, loc=<14.7382 0 0>, rot=<0 0 0 1>}
 dyn_rear_skirt_corner_l_04 {parent=dyn_rear_skirt_corner_l_03, loc=<13.2954 0 0>, rot=<0 0 0 1>}
 dyn_rear_skirt_corner_l_05 {parent=dyn_rear_skirt_corner_l_04, loc=<12.098 0 0>, rot=<0 0 0 1>}
 dyn_front_skirt_corner_r_MSTR {parent=pelvis, loc=<-7.2351 6.6377 11.9298>, rot=<-0.3672 -0.022 0.03 -0.9294>}
 dyn_front_skirt_corner_r_01 {parent=dyn_front_skirt_corner_r_MSTR, loc=<0 0 0>, rot=<0 0 0 1>}
 dyn_front_skirt_corner_r_02 {parent=dyn_front_skirt_corner_r_01, loc=<-20.5429 0 -0.0001>, rot=<0 0 0 1>}
 dyn_front_skirt_corner_r_03 {parent=dyn_front_skirt_corner_r_02, loc=<-14.7382 0 0>, rot=<0 0 0 1>}
 dyn_front_skirt_corner_r_04 {parent=dyn_front_skirt_corner_r_03, loc=<-13.2955 0 0>, rot=<0 0 0 1>}
 dyn_front_skirt_corner_r_05 {parent=dyn_front_skirt_corner_r_04, loc=<-12.098 0 0>, rot=<0 0 0 1>}
 dyn_rear_skirt_corner_r_MSTR {parent=pelvis, loc=<-5.021 -7.9583 11.7894>, rot=<0.9718 0.0247 -0.0233 0.2332>}
 dyn_rear_skirt_corner_r_01 {parent=dyn_rear_skirt_corner_r_MSTR, loc=<0 0 0>, rot=<0 0 0 -1>}
 dyn_rear_skirt_corner_r_02_MSTR {parent=dyn_rear_skirt_corner_r_01, loc=<-20.5429 0.0001 -0.0001>, rot=<0 0 0 -1>}
 dyn_rear_skirt_corner_r_02 {parent=dyn_rear_skirt_corner_r_02_MSTR, loc=<0 0 0>, rot=<0 0 0 1>}
 dyn_rear_skirt_corner_r_03 {parent=dyn_rear_skirt_corner_r_02, loc=<-14.7382 -0.0001 0.0001>, rot=<0 0 0 -1>}
 dyn_rear_skirt_corner_r_04 {parent=dyn_rear_skirt_corner_r_03, loc=<-13.2955 0 0>, rot=<0 0 0 -1>}
 dyn_rear_skirt_corner_r_05 {parent=dyn_rear_skirt_corner_r_04, loc=<-12.098 0.0001 0>, rot=<0 0 0 -1>}
 dyn_root_offset {parent=root, loc=<0 0 0>, rot=<0 0 0 1>}
]
variables [
 IsBaseballSliding {type=bool, default=False}
 IsCrouching {type=bool, default=False}
 IsKneeSliding {type=bool, default=False}
 MoveGroundCollision {type=bool, default=False}
]
