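# Physics asset for the male medium PigeonChart body cosmetic, providing secondary motion for a bow and rear skirt.
# Capsule chains hang from the pelvis: two bow tails and paired loops via master bodies, plus four skirt strips (left, right and two corners).
# Bow segments have limited swing and twist below free master joints, while every skirt joint rotates freely.
# A huge unconstrained sphere below dyn_root_offset likely acts as a ground collider.
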PhysicsAsset /BRCosmetics/Characters/Player/Male/Medium/Bodies/M_MED_PigeonChart/Meshes/M_MED_PigeonChart_Physics
bodies {
 dyn_bow_l_01 [{shapeType=Capsule, length=5, radius=4, shapeName=dyn_bow_l_01_capsule, center=<1.8184 0.8081 0.2591>, rotation=<rot 90 0 0>}]
 dyn_bow_l_02 [{shapeType=Capsule, length=5, radius=4, shapeName=dyn_bow_l_02_capsule, center=<1.9107 -1.176 0.2572>, rotation=<rot 90 0 0>}]
 dyn_bow_l_03 [{shapeType=Capsule, length=9.9, radius=3.6, shapeName=dyn_bow_l_03_capsule, center=<1.9107 -1.176 0.2572>, rotation=<rot 90 0 0>}]
 dyn_bow_l_04 [{shapeType=Capsule, length=9.9, radius=3.6, shapeName=dyn_bow_l_04_capsule, center=<1.9107 -1.176 0.2572>, rotation=<rot 90 0 0>}]
 dyn_bow_l_05 [{shapeType=Capsule, length=9.9, radius=3.6, shapeName=dyn_bow_l_05_capsule, center=<2 -1 0>, rotation=<rot 90 0 0>}]
 dyn_bow_l_MSTR [{shapeType=Capsule, length=0.505, radius=0.505, shapeName=dyn_bow_l_MSTR_capsule, rotation=<rot 0 -90 0>}]
 dyn_bow_loop_MSTR [{shapeType=Capsule, length=0.505, radius=0.505, shapeName=dyn_bow_loop_MSTR_capsule, rotation=<rot 0 -90 0>}]
 dyn_bow_loop_l_01 [{shapeType=Capsule, length=5, radius=4, shapeName=dyn_bow_loop_l_01_capsule, center=<2 0 0>, rotation=<rot 90 0 0>}]
 dyn_bow_loop_l_02 [{shapeType=Capsule, length=5, radius=4, shapeName=dyn_bow_loop_l_02_capsule, center=<2 0 0>, rotation=<rot 90 0 0>}]
 dyn_bow_loop_r_01 [{shapeType=Capsule, length=5, radius=4, shapeName=dyn_bow_loop_r_01_capsule, center=<-2 0 0>, rotation=<rot 90 0 0>}]
 dyn_bow_loop_r_02 [{shapeType=Capsule, length=5, radius=4, shapeName=dyn_bow_loop_r_02_capsule, center=<-2 0 0>, rotation=<rot 90 0 0>}]
 dyn_bow_r_01 [{shapeType=Capsule, length=5, radius=4, shapeName=dyn_bow_r_01_capsule, center=<-1.8184 -0.8081 -0.2591>, rotation=<rot 90 0 0>}]
 dyn_bow_r_02 [{shapeType=Capsule, length=5, radius=4, shapeName=dyn_bow_r_02_capsule, center=<-1.9107 1.176 -0.2572>, rotation=<rot 90 0 0>}]
 dyn_bow_r_03 [{shapeType=Capsule, length=9.9, radius=3.6, shapeName=dyn_bow_r_03_capsule, center=<-1.9107 1.176 -0.2572>, rotation=<rot 90 0 0>}]
 dyn_bow_r_04 [{shapeType=Capsule, length=9.9, radius=3.6, shapeName=dyn_bow_r_04_capsule, center=<-1.9107 1.176 -0.2572>, rotation=<rot 90 0 0>}]
 dyn_bow_r_05 [{shapeType=Capsule, length=14.2, radius=3.6, shapeName=dyn_bow_r_05_capsule, center=<-3.4 1 0>, rotation=<rot 90 -180 -180>}]
 dyn_bow_r_MSTR [{shapeType=Capsule, length=0.505, radius=0.505, shapeName=dyn_bow_r_MSTR_capsule, rotation=<rot 0 -90 0>}]
 dyn_rear_skirt_corner_l_01 [{
   shapeType=Capsule
   length=10
   radius=7
   shapeName=dyn_rear_skirt_corner_l_01_capsule
   center=<5.1901 5.8347 0.0888>
   rotation=<rot 0.0001 0.0001 -14.0625>
  }]
 dyn_rear_skirt_corner_l_02 [{
   shapeType=Capsule
   length=10
   radius=7
   shapeName=dyn_rear_skirt_corner_l_02_capsule
   center=<0 3.4266 0.0888>
   rotation=<rot 0.0001 0.0001 -14.0625>
  }]
 dyn_rear_skirt_corner_l_03 [{
   shapeType=Capsule
   length=10
   radius=8
   shapeName=dyn_rear_skirt_corner_l_03_capsule
   center=<-0.3684 5.2578 2.4165>
   rotation=<rot 0.0001 0.0001 -8.4375>
  }]
 dyn_rear_skirt_corner_l_04 [{
   shapeType=Capsule
   length=10
   radius=8
   shapeName=dyn_rear_skirt_corner_l_04_capsule
   center=<-0.2609 3.9492 1.7449>
   rotation=<rot 0.0001 0.0001 -8.4375>
  }]
 dyn_rear_skirt_corner_l_05 [{shapeType=Capsule, length=2.4454, radius=11.8, shapeName=dyn_rear_skirt_corner_l_05_capsule, center=<3.3431 6.6473 0>}]
 dyn_rear_skirt_corner_r_01 [{
   shapeType=Capsule
   length=10
   radius=7
   shapeName=dyn_rear_skirt_corner_l_01_capsule
   center=<-5.1901 -5.8347 -0.0888>
   rotation=<rot 0.0001 0.0001 165.9375>
  }]
 dyn_rear_skirt_corner_r_02 [{
   shapeType=Capsule
   length=10
   radius=7
   shapeName=dyn_rear_skirt_corner_l_02_capsule
   center=<0 -3.4266 -0.0888>
   rotation=<rot 0.0001 0.0001 165.9375>
  }]
 dyn_rear_skirt_corner_r_03 [{
   shapeType=Capsule
   length=10
   radius=6
   shapeName=dyn_rear_skirt_corner_l_03_capsule
   center=<0.3684 -4.4846 -2.3019>
   rotation=<rot 0 0 -180>
  }]
 dyn_rear_skirt_corner_r_04 [{
   shapeType=Capsule
   length=10
   radius=6
   shapeName=dyn_rear_skirt_corner_l_04_capsule
   center=<0.2609 -3.1761 -1.6302>
   rotation=<rot 0 0 -180>
  }]
 dyn_rear_skirt_corner_r_05 [{
   shapeType=Capsule
   length=2.4454
   radius=11.8
   shapeName=dyn_rear_skirt_corner_l_05_capsule
   center=<-3.3431 -6.6473 0>
   rotation=<rot 0 0 -180>
  }]
 dyn_rear_skirt_l_01 [{shapeType=Capsule, length=10, radius=7, shapeName=dyn_rear_skirt_l_01_capsule}]
 dyn_rear_skirt_l_02 [{shapeType=Capsule, length=8, radius=8, shapeName=dyn_rear_skirt_l_02_capsule, center=<-0.1392 7.9789 0.1932>}]
 dyn_rear_skirt_l_03 [{shapeType=Capsule, length=22.2, radius=8, shapeName=dyn_rear_skirt_l_03_capsule, center=<-0.1441 8.0484 0.2>}]
 dyn_rear_skirt_l_04 [{shapeType=Capsule, length=22.2, radius=8, shapeName=dyn_rear_skirt_l_04_capsule, center=<-0.1441 8.0484 0.2>}]
 dyn_rear_skirt_l_05 [{shapeType=Capsule, length=16.0563, radius=11.8, shapeName=dyn_rear_skirt_l_05_capsule, center=<3.3431 6.6473 0>}]
 dyn_rear_skirt_r_01 [{shapeType=Capsule, length=10, radius=7, shapeName=dyn_rear_skirt_r_01_capsule}]
 dyn_rear_skirt_r_02 [{shapeType=Capsule, length=8, radius=8, shapeName=dyn_rear_skirt_l_02_capsule, center=<0.1392 -7.9789 -0.1932>, rotation=<rot 0 0 -180>}]
 dyn_rear_skirt_r_03 [{shapeType=Capsule, length=22.2, radius=8, shapeName=dyn_rear_skirt_l_03_capsule, center=<0.1441 -8.0484 -0.2>, rotation=<rot 0 0 -180>}]
 dyn_rear_skirt_r_04 [{shapeType=Capsule, length=22.2, radius=8, shapeName=dyn_rear_skirt_l_04_capsule, center=<0.1441 -8.0484 -0.2>, rotation=<rot 0 0 -180>}]
 dyn_rear_skirt_r_05 [{shapeType=Capsule, length=16.0563, radius=11.8, shapeName=dyn_rear_skirt_l_05_capsule, center=<-3.3431 -6.6473 0>, rotation=<rot 0 0 -180>}]
 dyn_root_offset [{shapeType=Sphere, radius=300, shapeName=dyn_root_offset_sphere, center=<0 0 -300>}]
 pelvis [{shapeType=Capsule, length=14, radius=12}]
}
constraints [
 {
  bone1Name=dyn_bow_l_MSTR
  bone2Name=pelvis
  swing1LimitDegrees=45
  swing1Motion=Free
  swing2LimitDegrees=45
  swing2Motion=Free
  twistLimitDegrees=45
  twistMotion=Free
 }
 {bone1Name=dyn_bow_l_01, bone2Name=dyn_bow_l_MSTR, swing1LimitDegrees=6, swing2LimitDegrees=4, twistLimitDegrees=4}
 {bone1Name=dyn_bow_l_02, bone2Name=dyn_bow_l_01, swing1LimitDegrees=32, swing2LimitDegrees=24, twistLimitDegrees=24}
 {bone1Name=dyn_bow_l_03, bone2Name=dyn_bow_l_02, swing1LimitDegrees=32, swing2LimitDegrees=24, twistLimitDegrees=24}
 {bone1Name=dyn_bow_l_04, bone2Name=dyn_bow_l_03, swing1LimitDegrees=32, swing2LimitDegrees=24, twistLimitDegrees=24}
 {bone1Name=dyn_bow_l_05, bone2Name=dyn_bow_l_04, swing1LimitDegrees=32, swing2LimitDegrees=24, twistLimitDegrees=24}
 {
  bone1Name=dyn_bow_loop_MSTR
  bone2Name=pelvis
  swing1LimitDegrees=45
  swing1Motion=Free
  swing2LimitDegrees=45
  swing2Motion=Free
  twistLimitDegrees=45
  twistMotion=Free
 }
 {bone1Name=dyn_bow_loop_l_01, bone2Name=dyn_bow_loop_MSTR, swing1LimitDegrees=8, swing2LimitDegrees=24, twistLimitDegrees=8}
 {bone1Name=dyn_bow_loop_l_02, bone2Name=dyn_bow_loop_l_01, swing1LimitDegrees=12, swing2LimitDegrees=32, twistLimitDegrees=8}
 {bone1Name=dyn_bow_loop_r_01, bone2Name=dyn_bow_loop_MSTR, swing1LimitDegrees=8, swing2LimitDegrees=24, twistLimitDegrees=8}
 {bone1Name=dyn_bow_loop_r_02, bone2Name=dyn_bow_loop_r_01, swing1LimitDegrees=12, swing2LimitDegrees=32, twistLimitDegrees=8}
 {
  bone1Name=dyn_bow_r_MSTR
  bone2Name=pelvis
  swing1LimitDegrees=45
  swing1Motion=Free
  swing2LimitDegrees=45
  swing2Motion=Free
  twistLimitDegrees=45
  twistMotion=Free
 }
 {bone1Name=dyn_bow_r_01, bone2Name=dyn_bow_r_MSTR, swing1LimitDegrees=6, swing2LimitDegrees=4, twistLimitDegrees=4}
 {bone1Name=dyn_bow_r_02, bone2Name=dyn_bow_r_01, swing1LimitDegrees=32, swing2LimitDegrees=24, twistLimitDegrees=24}
 {bone1Name=dyn_bow_r_03, bone2Name=dyn_bow_r_02, swing1LimitDegrees=32, swing2LimitDegrees=24, twistLimitDegrees=24}
 {bone1Name=dyn_bow_r_04, bone2Name=dyn_bow_r_03, swing1LimitDegrees=32, swing2LimitDegrees=24, twistLimitDegrees=24}
 {bone1Name=dyn_bow_r_05, bone2Name=dyn_bow_r_04, swing1LimitDegrees=32, swing2LimitDegrees=24, twistLimitDegrees=24}
 {
  bone1Name=dyn_rear_skirt_l_02
  bone2Name=dyn_rear_skirt_l_01
  swing1LimitDegrees=45
  swing1Motion=Free
  swing2LimitDegrees=45
  swing2Motion=Free
  twistLimitDegrees=45
  twistMotion=Free
 }
 {
  bone1Name=dyn_rear_skirt_l_03
  bone2Name=dyn_rear_skirt_l_02
  swing1LimitDegrees=45
  swing1Motion=Free
  swing2LimitDegrees=45
  swing2Motion=Free
  twistLimitDegrees=45
  twistMotion=Free
 }
 {
  bone1Name=dyn_rear_skirt_l_04
  bone2Name=dyn_rear_skirt_l_03
  swing1LimitDegrees=45
  swing1Motion=Free
  swing2LimitDegrees=45
  swing2Motion=Free
  twistLimitDegrees=45
  twistMotion=Free
 }
 {
  bone1Name=dyn_rear_skirt_r_02
  bone2Name=dyn_rear_skirt_r_01
  swing1LimitDegrees=45
  swing1Motion=Free
  swing2LimitDegrees=45
  swing2Motion=Free
  twistLimitDegrees=45
  twistMotion=Free
 }
 {
  bone1Name=dyn_rear_skirt_r_03
  bone2Name=dyn_rear_skirt_r_02
  swing1LimitDegrees=45
  swing1Motion=Free
  swing2LimitDegrees=45
  swing2Motion=Free
  twistLimitDegrees=45
  twistMotion=Free
 }
 {
  bone1Name=dyn_rear_skirt_r_04
  bone2Name=dyn_rear_skirt_r_03
  swing1LimitDegrees=45
  swing1Motion=Free
  swing2LimitDegrees=45
  swing2Motion=Free
  twistLimitDegrees=45
  twistMotion=Free
 }
 {
  bone1Name=dyn_rear_skirt_corner_l_03
  bone2Name=dyn_rear_skirt_corner_l_02
  swing1LimitDegrees=45
  swing1Motion=Free
  swing2LimitDegrees=45
  swing2Motion=Free
  twistLimitDegrees=45
  twistMotion=Free
 }
 {
  bone1Name=dyn_rear_skirt_corner_l_04
  bone2Name=dyn_rear_skirt_corner_l_03
  swing1LimitDegrees=45
  swing1Motion=Free
  swing2LimitDegrees=45
  swing2Motion=Free
  twistLimitDegrees=45
  twistMotion=Free
 }
 {
  bone1Name=dyn_rear_skirt_corner_r_03
  bone2Name=dyn_rear_skirt_corner_r_02
  swing1LimitDegrees=45
  swing1Motion=Free
  swing2LimitDegrees=45
  swing2Motion=Free
  twistLimitDegrees=45
  twistMotion=Free
 }
 {
  bone1Name=dyn_rear_skirt_corner_r_04
  bone2Name=dyn_rear_skirt_corner_r_03
  swing1LimitDegrees=45
  swing1Motion=Free
  swing2LimitDegrees=45
  swing2Motion=Free
  twistLimitDegrees=45
  twistMotion=Free
 }
 {
  bone1Name=dyn_rear_skirt_corner_l_01
  bone2Name=pelvis
  swing1LimitDegrees=45
  swing1Motion=Free
  swing2LimitDegrees=45
  swing2Motion=Free
  twistLimitDegrees=45
  twistMotion=Free
 }
 {
  bone1Name=dyn_rear_skirt_corner_l_02
  bone2Name=dyn_rear_skirt_corner_l_01
  swing1LimitDegrees=45
  swing1Motion=Free
  swing2LimitDegrees=45
  swing2Motion=Free
  twistLimitDegrees=45
  twistMotion=Free
 }
 {
  bone1Name=dyn_rear_skirt_l_01
  bone2Name=pelvis
  swing1LimitDegrees=45
  swing1Motion=Free
  swing2LimitDegrees=45
  swing2Motion=Free
  twistLimitDegrees=45
  twistMotion=Free
 }
 {
  bone1Name=dyn_rear_skirt_r_01
  bone2Name=pelvis
  swing1LimitDegrees=45
  swing1Motion=Free
  swing2LimitDegrees=45
  swing2Motion=Free
  twistLimitDegrees=45
  twistMotion=Free
 }
 {
  bone1Name=dyn_rear_skirt_corner_r_01
  bone2Name=pelvis
  swing1LimitDegrees=45
  swing1Motion=Free
  swing2LimitDegrees=45
  swing2Motion=Free
  twistLimitDegrees=45
  twistMotion=Free
 }
 {
  bone1Name=dyn_rear_skirt_corner_r_02
  bone2Name=dyn_rear_skirt_corner_r_01
  swing1LimitDegrees=45
  swing1Motion=Free
  swing2LimitDegrees=45
  swing2Motion=Free
  twistLimitDegrees=45
  twistMotion=Free
 }
 {
  bone1Name=dyn_rear_skirt_corner_l_05
  bone2Name=dyn_rear_skirt_corner_l_04
  swing1LimitDegrees=45
  swing1Motion=Free
  swing2LimitDegrees=45
  swing2Motion=Free
  twistLimitDegrees=45
  twistMotion=Free
 }
 {
  bone1Name=dyn_rear_skirt_l_05
  bone2Name=dyn_rear_skirt_l_04
  swing1LimitDegrees=45
  swing1Motion=Free
  swing2LimitDegrees=45
  swing2Motion=Free
  twistLimitDegrees=45
  twistMotion=Free
 }
 {
  bone1Name=dyn_rear_skirt_corner_r_05
  bone2Name=dyn_rear_skirt_corner_r_04
  swing1LimitDegrees=45
  swing1Motion=Free
  swing2LimitDegrees=45
  swing2Motion=Free
  twistLimitDegrees=45
  twistMotion=Free
 }
 {
  bone1Name=dyn_rear_skirt_r_05
  bone2Name=dyn_rear_skirt_r_04
  swing1LimitDegrees=45
  swing1Motion=Free
  swing2LimitDegrees=45
  swing2Motion=Free
  twistLimitDegrees=45
  twistMotion=Free
 }
]
