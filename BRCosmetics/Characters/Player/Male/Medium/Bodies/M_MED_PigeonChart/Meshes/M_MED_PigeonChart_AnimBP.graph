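# Post-process style Animation Blueprint for the PigeonChart male medium body cosmetic, adding secondary cloth motion to the pose copied from the parent mesh.
# A Control Rig reacts to slide and crouch states, feeding RigidBody skirt simulation (base bone switched during emotes), additive bone offsets and a second RigidBody pass.
# bRigidBodyEnabled selects that or an AnimDynamics skirt fallback, then AnimDynamics moves body cloth, collar and sleeves.
# Custom idles layer DefaultSlot onto forearm cloth.
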AnimBlueprint /BRCosmetics/Characters/Player/Male/Medium/Bodies/M_MED_PigeonChart/Meshes/M_MED_PigeonChart_AnimBP
target_skeleton=&/M_MED_PigeonChart_Skeleton
graphs [
 AnimGraph {nodes [
   n1 = Layered blend per bone(
    BasePose=Use cached pose 'Post Secondary Anim Dynamics'()
    BlendPoses_0=Slot 'DefaultSlot'()
    BlendWeights_0=Get Custom Idle Is Playing() -> (Custom Idle Is Playing:float)
   ) {layerSetup [{branchFilters [{boneName=dyn_lowerarm_cloth_r_MSTR}, {boneName=dyn_lowerarm_cloth_l_MSTR}]}]}
   Output Pose(Result=@n1) {layerGroup=None}
   n2 = Control Rig!(
    Source=Copy Pose From Mesh() {bUseAttachedParent=true}
    Alpha
    IsKneeSliding:bool=Get bIsKneeSliding()
    IsBaseballSliding:bool=Get bIsBaseballSliding()
    IsCrouching:bool=Get bIsCrouching()
    MoveGroundCollision:bool=Get MoveGroundCollision() -> (MoveGroundCollision:bool)
   ) -> (Pose) {
    bTransferInputCurves=false
    alphaBoolBlend {blendOption=Linear}
    controlRigAssetReference {blueprintRigClass=&/M_MED_PigeonChart_CtrlRig.M_MED_PigeonChart_CtrlRig_C}
   }
   Save cached pose 'Post Deformation Rig'(Pose=@n2) {cacheName=Post Deformation Rig}
   n3 = RigidBody!(
    ComponentPose=Local To Component(LocalPose=Use cached pose 'Post Deformation Rig'())
    Alpha
    OverrideWorldGravity:Vector="0,0,-980"
    SimSpaceSettings:Sim Space Settings Structure=(WorldAlpha=0.5,VelocityScaleZ=0.5,DampingAlpha=1,MaxLinearVelocity=400,MaxAngularVelocity=10,MaxLinearAcceleration=400,MaxAngularAcceleration=10,ExternalLinearDragV=(X=0.5,Y=0.5,Z=0.5),ExternalLinearVelocity=(X=0,Y=0,Z=0),ExternalAngularVelocity=(X=0,Y=0,Z=0))
   ) -> (Pose) {
    bOverrideWorldGravity=true
    overridePhysicsAsset=&/M_MED_PigeonChart_Skirt_Physics
    simulationSpace=BaseBoneSpace
    alphaBoolBlend {blendOption=Linear}
    baseBoneRef {boneName=pelvis}
    overrideWorldGravity {z=-980}
    simSpaceSettings {
     maxAngularAcceleration=10
     maxAngularVelocity=10
     maxLinearAcceleration=400
     maxLinearVelocity=400
     velocityScaleZ=0.5
     worldAlpha=0.5
     externalLinearDragV=<0.5 0.5 0.5>
    }
   }
   n4 = Component To Local(ComponentPose=@n3)
   n5 = RigidBody(
    ComponentPose=Local To Component(LocalPose=Use cached pose 'Post Deformation Rig'())
    OverrideWorldGravity="0,0,-980"
    SimSpaceSettings=(WorldAlpha=0.5,VelocityScaleZ=0.5,DampingAlpha=1,MaxLinearVelocity=400,MaxAngularVelocity=10,MaxLinearAcceleration=400,MaxAngularAcceleration=10,ExternalLinearDragV=(X=0.5,Y=0.5,Z=0.5),ExternalLinearVelocity=(X=0,Y=0,Z=0),ExternalAngularVelocity=(X=0,Y=0,Z=0))
   ) {
    bOverrideWorldGravity=true
    overridePhysicsAsset=&M_MED_PigeonChart_Skirt_Physics
    simulationSpace=BaseBoneSpace
    alphaBoolBlend {blendOption=Linear}
    baseBoneRef {boneName=root}
    overrideWorldGravity {z=-980}
    simSpaceSettings {
     maxAngularAcceleration=10
     maxAngularVelocity=10
     maxLinearAcceleration=400
     maxLinearVelocity=400
     velocityScaleZ=0.5
     worldAlpha=0.5
     externalLinearDragV=<0.5 0.5 0.5>
    }
   }
   n6 = Component To Local(ComponentPose=@n5)
   n7 = Get bIsPlayingEmote()
   n8 = "Transform (Modify) Bone - Bone: dyn_front_skirt_l_MSTR"(
    ComponentPose=Local To Component(LocalPose=Blend Poses by bool(BlendPose_0=@n4, BlendPose_1=@n6, bActiveValue=@n7) {blendType=Linear})
    Alpha=1
    Translation="0,-1,0"
    Rotation="0,-2.5,0"
    Scale="1,1,1"
   ) {
    rotationMode=BMM_Additive
    rotationSpace=BCS_BoneSpace
    translationMode=BMM_Additive
    translationSpace=BCS_BoneSpace
    alphaBoolBlend {blendOption=Linear}
    boneToModify {boneName=dyn_front_skirt_l_MSTR}
    rotation {yaw=-2.5}
    translation {y=-1}
   }
   n9 = "Transform (Modify) Bone - Bone: dyn_front_skirt_corner_l_MSTR"(ComponentPose=@n8, Alpha=1, Translation="0,-1,0", Rotation="0,-2.5,0", Scale="1,1,1") {
    rotationMode=BMM_Additive
    rotationSpace=BCS_BoneSpace
    translationMode=BMM_Additive
    translationSpace=BCS_BoneSpace
    alphaBoolBlend {blendOption=Linear}
    boneToModify {boneName=dyn_front_skirt_corner_l_MSTR}
    rotation {yaw=-2.5}
    translation {y=-1}
   }
   n10 = "Transform (Modify) Bone - Bone: dyn_side_skirt_l_MSTR"(ComponentPose=@n9, Alpha=1, Translation="0,-1,0", Rotation="0,-2.5,0", Scale="1,1,1") {
    rotationMode=BMM_Additive
    rotationSpace=BCS_BoneSpace
    translationMode=BMM_Additive
    translationSpace=BCS_BoneSpace
    alphaBoolBlend {blendOption=Linear}
    boneToModify {boneName=dyn_side_skirt_l_MSTR}
    rotation {yaw=-2.5}
    translation {y=-1}
   }
   n11 = "Transform (Modify) Bone - Bone: dyn_rear_skirt_corner_l_MSTR"(ComponentPose=@n10, Alpha=0.5, Translation="0,-1,0", Rotation="0,-2.5,0", Scale="1,1,1") {
    rotationMode=BMM_Additive
    rotationSpace=BCS_BoneSpace
    translationMode=BMM_Additive
    translationSpace=BCS_BoneSpace
    alphaBoolBlend {blendOption=Linear}
    boneToModify {boneName=dyn_rear_skirt_corner_l_MSTR}
    rotation {yaw=-2.5}
    translation {y=-1}
   }
   n12 = "Transform (Modify) Bone - Bone: dyn_rear_skirt_l_MSTR"(ComponentPose=@n11, Alpha=0.5, Translation="0,-1,0", Rotation="0,-2.5,0", Scale="1,1,1") {
    rotationMode=BMM_Additive
    rotationSpace=BCS_BoneSpace
    translationMode=BMM_Additive
    translationSpace=BCS_BoneSpace
    alphaBoolBlend {blendOption=Linear}
    boneToModify {boneName=dyn_rear_skirt_l_MSTR}
    rotation {yaw=-2.5}
    translation {y=-1}
   }
   n13 = "Transform (Modify) Bone - Bone: dyn_rear_skirt_r_MSTR"(ComponentPose=@n12, Alpha=0.5, Translation="0,1,0", Rotation="0,-2.5,0", Scale="1,1,1") {
    rotationMode=BMM_Additive
    rotationSpace=BCS_BoneSpace
    translationMode=BMM_Additive
    translationSpace=BCS_BoneSpace
    alphaBoolBlend {blendOption=Linear}
    boneToModify {boneName=dyn_rear_skirt_r_MSTR}
    rotation {yaw=-2.5}
    translation {y=1}
   }
   n14 = "Transform (Modify) Bone - Bone: dyn_rear_skirt_corner_r_MSTR"(ComponentPose=@n13, Alpha=0.5, Translation="0,1,0", Rotation="0,-2.5,0", Scale="1,1,1") {
    rotationMode=BMM_Additive
    rotationSpace=BCS_BoneSpace
    translationMode=BMM_Additive
    translationSpace=BCS_BoneSpace
    alphaBoolBlend {blendOption=Linear}
    boneToModify {boneName=dyn_rear_skirt_corner_r_MSTR}
    rotation {yaw=-2.5}
    translation {y=1}
   }
   n15 = "Transform (Modify) Bone - Bone: dyn_side_skirt_r_MSTR"(ComponentPose=@n14, Alpha=1, Translation="0,1,0", Rotation="0,-2.5,0", Scale="1,1,1") {
    rotationMode=BMM_Additive
    rotationSpace=BCS_BoneSpace
    translationMode=BMM_Additive
    translationSpace=BCS_BoneSpace
    alphaBoolBlend {blendOption=Linear}
    boneToModify {boneName=dyn_side_skirt_r_MSTR}
    rotation {yaw=-2.5}
    translation {y=1}
   }
   n16 = "Transform (Modify) Bone - Bone: dyn_front_skirt_r_MSTR"(ComponentPose=@n15, Alpha=1, Translation="0,1,0", Rotation="0,-2.5,0", Scale="1,1,1") {
    rotationMode=BMM_Additive
    rotationSpace=BCS_BoneSpace
    translationMode=BMM_Additive
    translationSpace=BCS_BoneSpace
    alphaBoolBlend {blendOption=Linear}
    boneToModify {boneName=dyn_front_skirt_r_MSTR}
    rotation {yaw=-2.5}
    translation {y=1}
   }
   n17 = "Transform (Modify) Bone - Bone: dyn_front_skirt_corner_r_MSTR"(ComponentPose=@n16, Alpha=1, Translation="0,1,0", Rotation="0,-2.5,0", Scale="1,1,1") {
    rotationMode=BMM_Additive
    rotationSpace=BCS_BoneSpace
    translationMode=BMM_Additive
    translationSpace=BCS_BoneSpace
    alphaBoolBlend {blendOption=Linear}
    boneToModify {boneName=dyn_front_skirt_corner_r_MSTR}
    rotation {yaw=-2.5}
    translation {y=1}
   }
   n18 = "Transform (Modify) Bone - Bone: dyn_bow_l_MSTR"(ComponentPose=@n17, Alpha=1, Translation="0,0,0", Rotation="0,10,0", Scale="1,1,1") {
    rotationMode=BMM_Additive
    rotationSpace=BCS_BoneSpace
    translationMode=BMM_Additive
    translationSpace=BCS_BoneSpace
    alphaBoolBlend {blendOption=Linear}
    boneToModify {boneName=dyn_bow_l_MSTR}
    rotation {yaw=10}
   }
   n19 = "Transform (Modify) Bone - Bone: dyn_bow_r_MSTR"(ComponentPose=@n18, Alpha=1, Translation="0,0,0", Rotation="0,10,0", Scale="1,1,1") {
    rotationMode=BMM_Additive
    rotationSpace=BCS_BoneSpace
    translationMode=BMM_Additive
    translationSpace=BCS_BoneSpace
    alphaBoolBlend {blendOption=Linear}
    boneToModify {boneName=dyn_bow_r_MSTR}
    rotation {yaw=10}
   }
   n20 = RigidBody!(ComponentPose=@n19, Alpha, OverrideWorldGravity="0,0,-980", ExternalForce:Vector) -> (Pose) {
    bOverrideWorldGravity=true
    overridePhysicsAsset=&/M_MED_PigeonChart_Physics
    simulationSpace=BaseBoneSpace
    alphaBoolBlend {blendOption=Linear}
    baseBoneRef {boneName=pelvis}
    overrideWorldGravity {z=-980}
    simSpaceSettings {
     maxAngularAcceleration=25
     maxAngularVelocity=25
     maxLinearAcceleration=500
     maxLinearVelocity=500
     worldAlpha=1
     externalLinearDragV=<1 1 1>
    }
   }
   Save cached pose 'Post RBAN'(Pose=Component To Local(ComponentPose=@n20)) {cacheName=Post RBAN}
   n21 = Use cached pose 'Post RBAN'()
   n22 = Use cached pose 'Post Anim Dynamic Skirt'()
   n23 = Get bRigidBodyEnabled()
   n24 = Anim Dynamics - Bone: dyn_body_cloth_side_l_01!(
    SimulationSpace:AnimPhysSimSpaceType Enum=Component
    ComponentPose=Local To Component(LocalPose=Blend Poses by bool(BlendPose_0=@n21, BlendPose_1=@n22, bActiveValue=@n23) {blendType=Linear})
    Alpha=1
    ExternalForce="0,0,0"
   ) -> (Pose) {
    angularDampingOverride=1
    angularSpringConstant=7500
    bAngularSpring=true
    bLinearSpring=true
    bOverrideAngularDamping=true
    bOverrideLinearDamping=true
    bUseGravityOverride=true
    gravityScale=0.35
    linearDampingOverride=0.95
    linearSpringConstant=1
    simulationSpace=RootRelative
    alphaBoolBlend {blendOption=Linear}
    boundBone {boneName=dyn_body_cloth_side_l_01}
    componentAppliedLinearAccClamp=<600 600 600>
    componentLinearAccScale=<1 1 1>
    componentLinearVelScale=<1 1 1>
    gravityOverride {z=-980}
    physicsBodyDefinitions [{
      boundBone {boneName=dyn_body_cloth_side_l_01}
      boxExtents=<20 30 30>
      constraintSetup {
       angularLimitsMax=<15 15 15>
       angularLimitsMin=<-15 -15 -15>
       angularTarget=<1 0 0>
       linearAxesMax=<0.5 0.5 0.5>
       linearAxesMin=<-0.5 -0.5 -0.5>
      }
      localJointOffset=<4 0 0>
     }]
   }
   n25 = Anim Dynamics - Bone: dyn_body_cloth_front_01!(SimulationSpace:AnimPhysSimSpaceType Enum, ComponentPose=@n24, Alpha=1, ExternalForce="0,0,0") -> (Pose) {
    angularDampingOverride=1
    angularSpringConstant=7500
    bAngularSpring=true
    bLinearSpring=true
    bOverrideAngularDamping=true
    bOverrideLinearDamping=true
    bUseGravityOverride=true
    gravityScale=0.35
    linearDampingOverride=0.95
    linearSpringConstant=1
    simulationSpace=RootRelative
    alphaBoolBlend {blendOption=Linear}
    boundBone {boneName=dyn_body_cloth_front_01}
    componentAppliedLinearAccClamp=<600 600 600>
    componentLinearAccScale=<1 1 1>
    componentLinearVelScale=<1 1 1>
    gravityOverride {z=-980}
    physicsBodyDefinitions [{
      boundBone {boneName=dyn_body_cloth_front_01}
      boxExtents=<20 30 30>
      constraintSetup {
       angularLimitsMax=<15 15 15>
       angularLimitsMin=<-15 -15 -15>
       angularTarget=<1 0 0>
       linearAxesMax=<0.5 0.5 0.5>
       linearAxesMin=<-0.5 -0.5 -0.5>
      }
      localJointOffset=<4 0 0>
     }]
   }
   n26 = Anim Dynamics - Bone: dyn_body_cloth_side_r_01!(SimulationSpace:AnimPhysSimSpaceType Enum, ComponentPose=@n25, Alpha=1, ExternalForce="0,0,0") -> (Pose) {
    angularDampingOverride=1
    angularSpringConstant=7500
    bAngularSpring=true
    bLinearSpring=true
    bOverrideAngularDamping=true
    bOverrideLinearDamping=true
    bUseGravityOverride=true
    gravityScale=0.35
    linearDampingOverride=0.95
    linearSpringConstant=1
    simulationSpace=RootRelative
    alphaBoolBlend {blendOption=Linear}
    boundBone {boneName=dyn_body_cloth_side_r_01}
    componentAppliedLinearAccClamp=<600 600 600>
    componentLinearAccScale=<1 1 1>
    componentLinearVelScale=<1 1 1>
    gravityOverride {z=-980}
    physicsBodyDefinitions [{
      boundBone {boneName=dyn_body_cloth_side_r_01}
      boxExtents=<20 30 30>
      constraintSetup {
       angularLimitsMax=<15 15 15>
       angularLimitsMin=<-15 -15 -15>
       angularTarget=<1 0 0>
       linearAxesMax=<0.5 0.5 0.5>
       linearAxesMin=<-0.5 -0.5 -0.5>
      }
      localJointOffset=<-4 0 0>
     }]
   }
   n27 = Anim Dynamics - Bone: dyn_collar_side_l_01!(SimulationSpace:AnimPhysSimSpaceType Enum=BoneRelative, ComponentPose=@n26, Alpha=1, ExternalForce="0,0,0") -> (Pose) {
    angularDampingOverride=1
    angularSpringConstant=7500
    bAngularSpring=true
    bOverrideAngularDamping=true
    bUseGravityOverride=true
    gravityScale=0.35
    simulationSpace=BoneRelative
    alphaBoolBlend {blendOption=Linear}
    boundBone {boneName=dyn_collar_side_l_01}
    componentAppliedLinearAccClamp=<1000 1000 1000>
    componentLinearAccScale=<1 1 1>
    componentLinearVelScale=<1 1 1>
    gravityOverride {z=-980}
    physicsBodyDefinitions [{
      boundBone {boneName=dyn_collar_side_l_01}
      boxExtents=<25 35 35>
      constraintSetup {angularLimitsMax=<5 5 0>, angularLimitsMin=<-5 -5 -7>, angularTarget=<1 0 0>}
      localJointOffset=<4 0 0>
     }]
    relativeSpaceBone {boneName=neck_01}
   }
   n28 = Anim Dynamics - Bone: dyn_collar_side_l_02!(SimulationSpace:AnimPhysSimSpaceType Enum=BoneRelative, ComponentPose=@n27, Alpha=1, ExternalForce="0,0,0") -> (Pose) {
    angularDampingOverride=1
    angularSpringConstant=7500
    bAngularSpring=true
    bOverrideAngularDamping=true
    bUseGravityOverride=true
    gravityScale=0.35
    simulationSpace=BoneRelative
    alphaBoolBlend {blendOption=Linear}
    boundBone {boneName=dyn_collar_side_l_02}
    componentAppliedLinearAccClamp=<1000 1000 1000>
    componentLinearAccScale=<1 1 1>
    componentLinearVelScale=<1 1 1>
    gravityOverride {z=-980}
    physicsBodyDefinitions [{
      boundBone {boneName=dyn_collar_side_l_02}
      boxExtents=<25 35 35>
      constraintSetup {angularLimitsMax=<5 5 0>, angularLimitsMin=<-5 -5 -7>, angularTarget=<1 0 0>}
      localJointOffset=<4 0 0>
     }]
    relativeSpaceBone {boneName=neck_01}
   }
   n29 = Anim Dynamics - Bone: dyn_collar_front_01!(SimulationSpace:AnimPhysSimSpaceType Enum=BoneRelative, ComponentPose=@n28, Alpha=1, ExternalForce="0,0,0") -> (Pose) {
    angularDampingOverride=1
    angularSpringConstant=7500
    bAngularSpring=true
    bOverrideAngularDamping=true
    bUseGravityOverride=true
    gravityScale=0.35
    simulationSpace=BoneRelative
    alphaBoolBlend {blendOption=Linear}
    boundBone {boneName=dyn_collar_front_01}
    componentAppliedLinearAccClamp=<1000 1000 1000>
    componentLinearAccScale=<1 1 1>
    componentLinearVelScale=<1 1 1>
    gravityOverride {z=-980}
    physicsBodyDefinitions [{
      boundBone {boneName=dyn_collar_front_01}
      boxExtents=<25 35 35>
      constraintSetup {angularLimitsMax=<5 5 0>, angularLimitsMin=<-5 -5 -7>, angularTarget=<1 0 0>}
      localJointOffset=<4 0 0>
     }]
    relativeSpaceBone {boneName=neck_01}
   }
   n30 = Anim Dynamics - Bone: dyn_collar_front_02!(SimulationSpace:AnimPhysSimSpaceType Enum=BoneRelative, ComponentPose=@n29, Alpha=1, ExternalForce="0,0,0") -> (Pose) {
    angularDampingOverride=1
    angularSpringConstant=7500
    bAngularSpring=true
    bOverrideAngularDamping=true
    bUseGravityOverride=true
    gravityScale=0.35
    simulationSpace=BoneRelative
    alphaBoolBlend {blendOption=Linear}
    boundBone {boneName=dyn_collar_front_02}
    componentAppliedLinearAccClamp=<1000 1000 1000>
    componentLinearAccScale=<1 1 1>
    componentLinearVelScale=<1 1 1>
    gravityOverride {z=-980}
    physicsBodyDefinitions [{
      boundBone {boneName=dyn_collar_front_02}
      boxExtents=<25 35 35>
      constraintSetup {angularLimitsMax=<5 5 0>, angularLimitsMin=<-5 -5 -7>, angularTarget=<1 0 0>}
      localJointOffset=<4 0 0>
     }]
    relativeSpaceBone {boneName=neck_01}
   }
   n31 = Anim Dynamics - Bone: dyn_collar_rear_01!(SimulationSpace:AnimPhysSimSpaceType Enum=BoneRelative, ComponentPose=@n30, Alpha=1, ExternalForce="0,0,0") -> (Pose) {
    angularDampingOverride=1
    angularSpringConstant=7500
    bAngularSpring=true
    bOverrideAngularDamping=true
    bUseGravityOverride=true
    gravityScale=0.35
    simulationSpace=BoneRelative
    alphaBoolBlend {blendOption=Linear}
    boundBone {boneName=dyn_collar_rear_01}
    componentAppliedLinearAccClamp=<1000 1000 1000>
    componentLinearAccScale=<1 1 1>
    componentLinearVelScale=<1 1 1>
    gravityOverride {z=-980}
    physicsBodyDefinitions [{
      boundBone {boneName=dyn_collar_rear_01}
      boxExtents=<25 35 35>
      constraintSetup {angularLimitsMax=<5 5 0>, angularLimitsMin=<-5 -5 -7>, angularTarget=<1 0 0>}
      localJointOffset=<4 0 0>
     }]
    relativeSpaceBone {boneName=neck_01}
   }
   n32 = Anim Dynamics - Bone: dyn_collar_rear_02!(SimulationSpace:AnimPhysSimSpaceType Enum=BoneRelative, ComponentPose=@n31, Alpha=1, ExternalForce="0,0,0") -> (Pose) {
    angularDampingOverride=1
    angularSpringConstant=7500
    bAngularSpring=true
    bOverrideAngularDamping=true
    bUseGravityOverride=true
    gravityScale=0.35
    simulationSpace=BoneRelative
    alphaBoolBlend {blendOption=Linear}
    boundBone {boneName=dyn_collar_rear_02}
    componentAppliedLinearAccClamp=<1000 1000 1000>
    componentLinearAccScale=<1 1 1>
    componentLinearVelScale=<1 1 1>
    gravityOverride {z=-980}
    physicsBodyDefinitions [{
      boundBone {boneName=dyn_collar_rear_02}
      boxExtents=<25 35 35>
      constraintSetup {angularLimitsMax=<5 5 0>, angularLimitsMin=<-5 -5 -7>, angularTarget=<1 0 0>}
      localJointOffset=<4 0 0>
     }]
    relativeSpaceBone {boneName=neck_01}
   }
   n33 = Anim Dynamics - Bone: dyn_collar_rear_corner_l_01!(SimulationSpace:AnimPhysSimSpaceType Enum=BoneRelative, ComponentPose=@n32, Alpha=1, ExternalForce="0,0,0") -> (Pose) {
    angularDampingOverride=1
    angularSpringConstant=7500
    bAngularSpring=true
    bOverrideAngularDamping=true
    bUseGravityOverride=true
    gravityScale=0.35
    simulationSpace=BoneRelative
    alphaBoolBlend {blendOption=Linear}
    boundBone {boneName=dyn_collar_rear_corner_l_01}
    componentAppliedLinearAccClamp=<1000 1000 1000>
    componentLinearAccScale=<1 1 1>
    componentLinearVelScale=<1 1 1>
    gravityOverride {z=-980}
    physicsBodyDefinitions [{
      boundBone {boneName=dyn_collar_rear_corner_l_01}
      boxExtents=<25 35 35>
      constraintSetup {angularLimitsMax=<5 5 0>, angularLimitsMin=<-5 -5 -7>, angularTarget=<1 0 0>}
      localJointOffset=<4 0 0>
     }]
    relativeSpaceBone {boneName=neck_01}
   }
   n34 = Anim Dynamics - Bone: dyn_collar_rear_corner_l_02!(SimulationSpace:AnimPhysSimSpaceType Enum=BoneRelative, ComponentPose=@n33, Alpha=1, ExternalForce="0,0,0") -> (Pose) {
    angularDampingOverride=1
    angularSpringConstant=7500
    bAngularSpring=true
    bOverrideAngularDamping=true
    bUseGravityOverride=true
    gravityScale=0.35
    simulationSpace=BoneRelative
    alphaBoolBlend {blendOption=Linear}
    boundBone {boneName=dyn_collar_rear_corner_l_02}
    componentAppliedLinearAccClamp=<1000 1000 1000>
    componentLinearAccScale=<1 1 1>
    componentLinearVelScale=<1 1 1>
    gravityOverride {z=-980}
    physicsBodyDefinitions [{
      boundBone {boneName=dyn_collar_rear_corner_l_02}
      boxExtents=<25 35 35>
      constraintSetup {angularLimitsMax=<5 5 0>, angularLimitsMin=<-5 -5 -7>, angularTarget=<1 0 0>}
      localJointOffset=<4 0 0>
     }]
    relativeSpaceBone {boneName=neck_01}
   }
   n35 = Anim Dynamics - Bone: dyn_collar_side_r_01!(SimulationSpace:AnimPhysSimSpaceType Enum=BoneRelative, ComponentPose=@n34, Alpha=1, ExternalForce="0,0,0") -> (Pose) {
    angularDampingOverride=1
    angularSpringConstant=7500
    bAngularSpring=true
    bOverrideAngularDamping=true
    bUseGravityOverride=true
    gravityScale=0.35
    simulationSpace=BoneRelative
    alphaBoolBlend {blendOption=Linear}
    boundBone {boneName=dyn_collar_side_r_01}
    componentAppliedLinearAccClamp=<1000 1000 1000>
    componentLinearAccScale=<1 1 1>
    componentLinearVelScale=<1 1 1>
    gravityOverride {z=-980}
    physicsBodyDefinitions [{
      boundBone {boneName=dyn_collar_side_r_01}
      boxExtents=<25 35 35>
      constraintSetup {angularLimitsMax=<5 5 0>, angularLimitsMin=<-5 -5 -7>, angularTarget=<1 0 0>}
      localJointOffset=<-4 0 0>
     }]
    relativeSpaceBone {boneName=neck_01}
   }
   n36 = Anim Dynamics - Bone: dyn_collar_side_r_02!(SimulationSpace:AnimPhysSimSpaceType Enum=BoneRelative, ComponentPose=@n35, Alpha=1, ExternalForce="0,0,0") -> (Pose) {
    angularDampingOverride=1
    angularSpringConstant=7500
    bAngularSpring=true
    bOverrideAngularDamping=true
    bUseGravityOverride=true
    gravityScale=0.35
    simulationSpace=BoneRelative
    alphaBoolBlend {blendOption=Linear}
    boundBone {boneName=dyn_collar_side_r_02}
    componentAppliedLinearAccClamp=<1000 1000 1000>
    componentLinearAccScale=<1 1 1>
    componentLinearVelScale=<1 1 1>
    gravityOverride {z=-980}
    physicsBodyDefinitions [{
      boundBone {boneName=dyn_collar_side_r_02}
      boxExtents=<25 35 35>
      constraintSetup {angularLimitsMax=<5 5 0>, angularLimitsMin=<-5 -5 -7>, angularTarget=<1 0 0>}
      localJointOffset=<-4 0 0>
     }]
    relativeSpaceBone {boneName=neck_01}
   }
   n37 = Anim Dynamics - Bone: dyn_collar_rear_corner_r_01!(SimulationSpace:AnimPhysSimSpaceType Enum=BoneRelative, ComponentPose=@n36, Alpha=1, ExternalForce="0,0,0") -> (Pose) {
    angularDampingOverride=1
    angularSpringConstant=7500
    bAngularSpring=true
    bOverrideAngularDamping=true
    bUseGravityOverride=true
    gravityScale=0.35
    simulationSpace=BoneRelative
    alphaBoolBlend {blendOption=Linear}
    boundBone {boneName=dyn_collar_rear_corner_r_01}
    componentAppliedLinearAccClamp=<1000 1000 1000>
    componentLinearAccScale=<1 1 1>
    componentLinearVelScale=<1 1 1>
    gravityOverride {z=-980}
    physicsBodyDefinitions [{
      boundBone {boneName=dyn_collar_rear_corner_r_01}
      boxExtents=<25 35 35>
      constraintSetup {angularLimitsMax=<5 5 0>, angularLimitsMin=<-5 -5 -7>, angularTarget=<1 0 0>}
      localJointOffset=<-4 0 0>
     }]
    relativeSpaceBone {boneName=neck_01}
   }
   n38 = Anim Dynamics - Bone: dyn_collar_rear_corner_r_02!(SimulationSpace:AnimPhysSimSpaceType Enum=BoneRelative, ComponentPose=@n37, Alpha=1, ExternalForce="0,0,0") -> (Pose) {
    angularDampingOverride=1
    angularSpringConstant=7500
    bAngularSpring=true
    bOverrideAngularDamping=true
    bUseGravityOverride=true
    gravityScale=0.35
    simulationSpace=BoneRelative
    alphaBoolBlend {blendOption=Linear}
    boundBone {boneName=dyn_collar_rear_corner_r_02}
    componentAppliedLinearAccClamp=<1000 1000 1000>
    componentLinearAccScale=<1 1 1>
    componentLinearVelScale=<1 1 1>
    gravityOverride {z=-980}
    physicsBodyDefinitions [{
      boundBone {boneName=dyn_collar_rear_corner_r_02}
      boxExtents=<25 35 35>
      constraintSetup {angularLimitsMax=<5 5 0>, angularLimitsMin=<-5 -5 -7>, angularTarget=<1 0 0>}
      localJointOffset=<-4 0 0>
     }]
    relativeSpaceBone {boneName=neck_01}
   }
   n39 = Anim Dynamics - Bone: dyn_lowerarm_cloth_up_l!(SimulationSpace:AnimPhysSimSpaceType Enum, ComponentPose=@n38, Alpha=1, ExternalForce="0,0,0") -> (Pose) {
    angularDampingOverride=1
    angularSpringConstant=8000
    bAngularSpring=true
    bOverrideAngularDamping=true
    bUseGravityOverride=true
    gravityScale=0.35
    simulationSpace=RootRelative
    alphaBoolBlend {blendOption=Linear}
    boundBone {boneName=dyn_lowerarm_cloth_up_l}
    componentAppliedLinearAccClamp=<1000 1000 1000>
    componentLinearAccScale=<0.3 0.3 0.3>
    componentLinearVelScale=<0.3 0.3 0.3>
    gravityOverride {z=-400}
    physicsBodyDefinitions [{
      boundBone {boneName=dyn_lowerarm_cloth_up_l}
      boxExtents=<20 30 30>
      constraintSetup {angularLimitsMax=<5 0 5>, angularLimitsMin=<-5 -5 -5>, angularTarget=<1 0 0>}
      localJointOffset=<4 0 0>
     }]
   }
   n40 = Anim Dynamics - Bone: dyn_lowerarm_cloth_down_l_01!(SimulationSpace:AnimPhysSimSpaceType Enum, ComponentPose=@n39, Alpha=1, ExternalForce="0,0,0") -> (Pose) {
    angularDampingOverride=1
    angularSpringConstant=8000
    bAngularSpring=true
    bOverrideAngularDamping=true
    bUseGravityOverride=true
    gravityScale=0.35
    simulationSpace=RootRelative
    alphaBoolBlend {blendOption=Linear}
    boundBone {boneName=dyn_lowerarm_cloth_down_l_01}
    componentAppliedLinearAccClamp=<1000 1000 1000>
    componentLinearAccScale=<0.3 0.3 0.3>
    componentLinearVelScale=<0.3 0.3 0.3>
    gravityOverride {z=-400}
    physicsBodyDefinitions [{
      boundBone {boneName=dyn_lowerarm_cloth_down_l_01}
      boxExtents=<20 30 30>
      constraintSetup {angularLimitsMax=<4 4 4>, angularLimitsMin=<-4 -4 -4>, angularTarget=<1 0 0>}
      localJointOffset=<4 0 0>
     }]
   }
   n41 = Anim Dynamics - Bone: dyn_lowerarm_cloth_down_l_02!(SimulationSpace:AnimPhysSimSpaceType Enum, ComponentPose=@n40, Alpha=1, ExternalForce="0,0,0") -> (Pose) {
    angularDampingOverride=1
    angularSpringConstant=8000
    bAngularSpring=true
    bOverrideAngularDamping=true
    bUseGravityOverride=true
    gravityScale=0.35
    simulationSpace=RootRelative
    alphaBoolBlend {blendOption=Linear}
    boundBone {boneName=dyn_lowerarm_cloth_down_l_02}
    componentAppliedLinearAccClamp=<1000 1000 1000>
    componentLinearAccScale=<0.3 0.3 0.3>
    componentLinearVelScale=<0.3 0.3 0.3>
    gravityOverride {z=-400}
    physicsBodyDefinitions [{
      boundBone {boneName=dyn_lowerarm_cloth_down_l_02}
      boxExtents=<20 30 30>
      constraintSetup {angularLimitsMax=<4 4 4>, angularLimitsMin=<-4 -4 -4>, angularTarget=<1 0 0>}
      localJointOffset=<4 0 0>
     }]
   }
   n42 = Anim Dynamics - Bone: dyn_lowerarm_cloth_up_r!(SimulationSpace:AnimPhysSimSpaceType Enum, ComponentPose=@n41, Alpha=1, ExternalForce="0,0,0") -> (Pose) {
    angularDampingOverride=1
    angularSpringConstant=8000
    bAngularSpring=true
    bOverrideAngularDamping=true
    bUseGravityOverride=true
    gravityScale=0.35
    simulationSpace=RootRelative
    alphaBoolBlend {blendOption=Linear}
    boundBone {boneName=dyn_lowerarm_cloth_up_r}
    componentAppliedLinearAccClamp=<1000 1000 1000>
    componentLinearAccScale=<0.3 0.3 0.3>
    componentLinearVelScale=<0.3 0.3 0.3>
    gravityOverride {z=-400}
    physicsBodyDefinitions [{
      boundBone {boneName=dyn_lowerarm_cloth_up_r}
      boxExtents=<20 30 30>
      constraintSetup {angularLimitsMax=<5 0 5>, angularLimitsMin=<-5 -5 -5>, angularTarget=<1 0 0>}
      localJointOffset=<-4 0 0>
     }]
   }
   n43 = Anim Dynamics - Bone: dyn_lowerarm_cloth_down_r_01!(SimulationSpace:AnimPhysSimSpaceType Enum, ComponentPose=@n42, Alpha=1, ExternalForce="0,0,0") -> (Pose) {
    angularDampingOverride=1
    angularSpringConstant=8000
    bAngularSpring=true
    bOverrideAngularDamping=true
    bUseGravityOverride=true
    gravityScale=0.35
    simulationSpace=RootRelative
    alphaBoolBlend {blendOption=Linear}
    boundBone {boneName=dyn_lowerarm_cloth_down_r_01}
    componentAppliedLinearAccClamp=<1000 1000 1000>
    componentLinearAccScale=<0.3 0.3 0.3>
    componentLinearVelScale=<0.3 0.3 0.3>
    gravityOverride {z=-400}
    physicsBodyDefinitions [{
      boundBone {boneName=dyn_lowerarm_cloth_down_r_01}
      boxExtents=<20 30 30>
      constraintSetup {angularLimitsMax=<4 4 4>, angularLimitsMin=<-4 -4 -4>, angularTarget=<1 0 0>}
      localJointOffset=<-4 0 0>
     }]
   }
   n44 = Anim Dynamics - Bone: dyn_lowerarm_cloth_down_r_02!(SimulationSpace:AnimPhysSimSpaceType Enum, ComponentPose=@n43, Alpha=1, ExternalForce="0,0,0") -> (Pose) {
    angularDampingOverride=1
    angularSpringConstant=8000
    bAngularSpring=true
    bOverrideAngularDamping=true
    bUseGravityOverride=true
    gravityScale=0.35
    simulationSpace=RootRelative
    alphaBoolBlend {blendOption=Linear}
    boundBone {boneName=dyn_lowerarm_cloth_down_r_02}
    componentAppliedLinearAccClamp=<1000 1000 1000>
    componentLinearAccScale=<0.3 0.3 0.3>
    componentLinearVelScale=<0.3 0.3 0.3>
    gravityOverride {z=-400}
    physicsBodyDefinitions [{
      boundBone {boneName=dyn_lowerarm_cloth_down_r_02}
      boxExtents=<20 30 30>
      constraintSetup {angularLimitsMax=<4 4 4>, angularLimitsMin=<-4 -4 -4>, angularTarget=<1 0 0>}
      localJointOffset=<-4 0 0>
     }]
   }
   Save cached pose 'Post Secondary Anim Dynamics'(Pose=Component To Local(ComponentPose=@n44)) {cacheName=Post Secondary Anim Dynamics}
   n45 = Anim Dynamics - Bone: dyn_front_skirt_corner_l_02(ComponentPose=Local To Component(LocalPose=Use cached pose 'Post Deformation Rig'()), Alpha=1, ExternalForce="0,0,0") {
    angularDampingOverride=1
    angularSpringConstant=4000
    bAngularSpring=true
    bOverrideAngularDamping=true
    bUseGravityOverride=true
    gravityScale=0.35
    simulationSpace=RootRelative
    alphaBoolBlend {blendOption=Linear}
    boundBone {boneName=dyn_front_skirt_corner_l_02}
    componentAppliedLinearAccClamp=<1000 1000 1000>
    componentLinearAccScale=<1 1 1>
    componentLinearVelScale=<1 1 1>
    gravityOverride {z=-2000}
    physicsBodyDefinitions [{
      boundBone {boneName=dyn_front_skirt_corner_l_02}
      boxExtents=<30 40 40>
      constraintSetup {angularLimitsMax=<20 20 2>, angularLimitsMin=<-20 -20 -45>, angularTarget=<1 0 0>}
      localJointOffset=<7 0 0>
     }]
   }
   n46 = Anim Dynamics - Bone: dyn_front_skirt_corner_l_03(ComponentPose=@n45, Alpha=1, ExternalForce="0,0,0") {
    angularDampingOverride=1
    angularSpringConstant=4000
    bAngularSpring=true
    bOverrideAngularDamping=true
    bUseGravityOverride=true
    gravityScale=0.35
    simulationSpace=RootRelative
    alphaBoolBlend {blendOption=Linear}
    boundBone {boneName=dyn_front_skirt_corner_l_03}
    componentAppliedLinearAccClamp=<1000 1000 1000>
    componentLinearAccScale=<1 1 1>
    componentLinearVelScale=<1 1 1>
    gravityOverride {z=-2000}
    physicsBodyDefinitions [{
      boundBone {boneName=dyn_front_skirt_corner_l_03}
      boxExtents=<30 40 40>
      constraintSetup {angularLimitsMax=<20 20 2>, angularLimitsMin=<-20 -20 -45>, angularTarget=<1 0 0>}
      localJointOffset=<7 0 0>
     }]
   }
   n47 = Anim Dynamics - Bone: dyn_front_skirt_corner_l_04(ComponentPose=@n46, Alpha=1, ExternalForce="0,0,0") {
    angularDampingOverride=1
    angularSpringConstant=4000
    bAngularSpring=true
    bOverrideAngularDamping=true
    bUseGravityOverride=true
    gravityScale=0.35
    simulationSpace=RootRelative
    alphaBoolBlend {blendOption=Linear}
    boundBone {boneName=dyn_front_skirt_corner_l_04}
    componentAppliedLinearAccClamp=<1000 1000 1000>
    componentLinearAccScale=<1 1 1>
    componentLinearVelScale=<1 1 1>
    gravityOverride {z=-2000}
    physicsBodyDefinitions [{
      boundBone {boneName=dyn_front_skirt_corner_l_04}
      boxExtents=<30 40 40>
      constraintSetup {angularLimitsMax=<20 20 2>, angularLimitsMin=<-20 -20 -45>, angularTarget=<1 0 0>}
      localJointOffset=<7 0 0>
     }]
   }
   n48 = Anim Dynamics - Bone: dyn_front_skirt_corner_l_05(ComponentPose=@n47, Alpha=1, ExternalForce="0,0,0") {
    angularDampingOverride=1
    angularSpringConstant=4000
    bAngularSpring=true
    bOverrideAngularDamping=true
    bUseGravityOverride=true
    gravityScale=0.35
    simulationSpace=RootRelative
    alphaBoolBlend {blendOption=Linear}
    boundBone {boneName=dyn_front_skirt_corner_l_05}
    componentAppliedLinearAccClamp=<1000 1000 1000>
    componentLinearAccScale=<1 1 1>
    componentLinearVelScale=<1 1 1>
    gravityOverride {z=-2000}
    physicsBodyDefinitions [{
      boundBone {boneName=dyn_front_skirt_corner_l_05}
      boxExtents=<30 40 40>
      constraintSetup {angularLimitsMax=<20 20 2>, angularLimitsMin=<-20 -20 -45>, angularTarget=<1 0 0>}
      localJointOffset=<7 0 0>
     }]
   }
   n49 = Anim Dynamics - Bone: dyn_rear_skirt_l_02(ComponentPose=@n48, Alpha=1, ExternalForce="0,0,0") {
    angularDampingOverride=1
    angularSpringConstant=4000
    bAngularSpring=true
    bOverrideAngularDamping=true
    bUseGravityOverride=true
    gravityScale=0.35
    simulationSpace=RootRelative
    alphaBoolBlend {blendOption=Linear}
    boundBone {boneName=dyn_rear_skirt_l_02}
    componentAppliedLinearAccClamp=<1000 1000 1000>
    componentLinearAccScale=<1 1 1>
    componentLinearVelScale=<1 1 1>
    gravityOverride {z=-2000}
    physicsBodyDefinitions [{
      boundBone {boneName=dyn_rear_skirt_l_02}
      boxExtents=<30 40 40>
      constraintSetup {angularLimitsMax=<20 20 2>, angularLimitsMin=<-20 -20 -45>, angularTarget=<1 0 0>}
      localJointOffset=<7 0 0>
     }]
   }
   n50 = Anim Dynamics - Bone: dyn_rear_skirt_l_03(ComponentPose=@n49, Alpha=1, ExternalForce="0,0,0") {
    angularDampingOverride=1
    angularSpringConstant=4000
    bAngularSpring=true
    bOverrideAngularDamping=true
    bUseGravityOverride=true
    gravityScale=0.35
    simulationSpace=RootRelative
    alphaBoolBlend {blendOption=Linear}
    boundBone {boneName=dyn_rear_skirt_l_03}
    componentAppliedLinearAccClamp=<1000 1000 1000>
    componentLinearAccScale=<1 1 1>
    componentLinearVelScale=<1 1 1>
    gravityOverride {z=-2000}
    physicsBodyDefinitions [{
      boundBone {boneName=dyn_rear_skirt_l_03}
      boxExtents=<30 40 40>
      constraintSetup {angularLimitsMax=<20 20 2>, angularLimitsMin=<-20 -20 -45>, angularTarget=<1 0 0>}
      localJointOffset=<7 0 0>
     }]
   }
   n51 = Anim Dynamics - Bone: dyn_rear_skirt_l_04(ComponentPose=@n50, Alpha=1, ExternalForce="0,0,0") {
    angularDampingOverride=1
    angularSpringConstant=4000
    bAngularSpring=true
    bOverrideAngularDamping=true
    bUseGravityOverride=true
    gravityScale=0.35
    simulationSpace=RootRelative
    alphaBoolBlend {blendOption=Linear}
    boundBone {boneName=dyn_rear_skirt_l_04}
    componentAppliedLinearAccClamp=<1000 1000 1000>
    componentLinearAccScale=<1 1 1>
    componentLinearVelScale=<1 1 1>
    gravityOverride {z=-2000}
    physicsBodyDefinitions [{
      boundBone {boneName=dyn_rear_skirt_l_04}
      boxExtents=<30 40 40>
      constraintSetup {angularLimitsMax=<20 20 2>, angularLimitsMin=<-20 -20 -45>, angularTarget=<1 0 0>}
      localJointOffset=<7 0 0>
     }]
   }
   n52 = Anim Dynamics - Bone: dyn_rear_skirt_l_05(ComponentPose=@n51, Alpha=1, ExternalForce="0,0,0") {
    angularDampingOverride=1
    angularSpringConstant=4000
    bAngularSpring=true
    bOverrideAngularDamping=true
    bUseGravityOverride=true
    gravityScale=0.35
    simulationSpace=RootRelative
    alphaBoolBlend {blendOption=Linear}
    boundBone {boneName=dyn_rear_skirt_l_05}
    componentAppliedLinearAccClamp=<1000 1000 1000>
    componentLinearAccScale=<1 1 1>
    componentLinearVelScale=<1 1 1>
    gravityOverride {z=-2000}
    physicsBodyDefinitions [{
      boundBone {boneName=dyn_rear_skirt_l_05}
      boxExtents=<30 40 40>
      constraintSetup {angularLimitsMax=<20 20 2>, angularLimitsMin=<-20 -20 -45>, angularTarget=<1 0 0>}
      localJointOffset=<7 0 0>
     }]
   }
   n53 = Anim Dynamics - Bone: dyn_rear_skirt_corner_l_02(ComponentPose=@n52, Alpha=1, ExternalForce="0,0,0") {
    angularDampingOverride=1
    angularSpringConstant=4000
    bAngularSpring=true
    bOverrideAngularDamping=true
    bUseGravityOverride=true
    gravityScale=0.35
    simulationSpace=RootRelative
    alphaBoolBlend {blendOption=Linear}
    boundBone {boneName=dyn_rear_skirt_corner_l_02}
    componentAppliedLinearAccClamp=<1000 1000 1000>
    componentLinearAccScale=<1 1 1>
    componentLinearVelScale=<1 1 1>
    gravityOverride {z=-2000}
    physicsBodyDefinitions [{
      boundBone {boneName=dyn_rear_skirt_corner_l_02}
      boxExtents=<30 40 40>
      constraintSetup {angularLimitsMax=<20 20 2>, angularLimitsMin=<-20 -20 -45>, angularTarget=<1 0 0>}
      localJointOffset=<7 0 0>
     }]
   }
   n54 = Anim Dynamics - Bone: dyn_rear_skirt_corner_l_03(ComponentPose=@n53, Alpha=1, ExternalForce="0,0,0") {
    angularDampingOverride=1
    angularSpringConstant=4000
    bAngularSpring=true
    bOverrideAngularDamping=true
    bUseGravityOverride=true
    gravityScale=0.35
    simulationSpace=RootRelative
    alphaBoolBlend {blendOption=Linear}
    boundBone {boneName=dyn_rear_skirt_corner_l_03}
    componentAppliedLinearAccClamp=<1000 1000 1000>
    componentLinearAccScale=<1 1 1>
    componentLinearVelScale=<1 1 1>
    gravityOverride {z=-2000}
    physicsBodyDefinitions [{
      boundBone {boneName=dyn_rear_skirt_corner_l_03}
      boxExtents=<30 40 40>
      constraintSetup {angularLimitsMax=<20 20 2>, angularLimitsMin=<-20 -20 -45>, angularTarget=<1 0 0>}
      localJointOffset=<7 0 0>
     }]
   }
   n55 = Anim Dynamics - Bone: dyn_rear_skirt_corner_l_04(ComponentPose=@n54, Alpha=1, ExternalForce="0,0,0") {
    angularDampingOverride=1
    angularSpringConstant=4000
    bAngularSpring=true
    bOverrideAngularDamping=true
    bUseGravityOverride=true
    gravityScale=0.35
    simulationSpace=RootRelative
    alphaBoolBlend {blendOption=Linear}
    boundBone {boneName=dyn_rear_skirt_corner_l_04}
    componentAppliedLinearAccClamp=<1000 1000 1000>
    componentLinearAccScale=<1 1 1>
    componentLinearVelScale=<1 1 1>
    gravityOverride {z=-2000}
    physicsBodyDefinitions [{
      boundBone {boneName=dyn_rear_skirt_corner_l_04}
      boxExtents=<30 40 40>
      constraintSetup {angularLimitsMax=<20 20 2>, angularLimitsMin=<-20 -20 -45>, angularTarget=<1 0 0>}
      localJointOffset=<7 0 0>
     }]
   }
   n56 = Anim Dynamics - Bone: dyn_rear_skirt_corner_l_05(ComponentPose=@n55, Alpha=1, ExternalForce="0,0,0") {
    angularDampingOverride=1
    angularSpringConstant=4000
    bAngularSpring=true
    bOverrideAngularDamping=true
    bUseGravityOverride=true
    gravityScale=0.35
    simulationSpace=RootRelative
    alphaBoolBlend {blendOption=Linear}
    boundBone {boneName=dyn_rear_skirt_corner_l_05}
    componentAppliedLinearAccClamp=<1000 1000 1000>
    componentLinearAccScale=<1 1 1>
    componentLinearVelScale=<1 1 1>
    gravityOverride {z=-2000}
    physicsBodyDefinitions [{
      boundBone {boneName=dyn_rear_skirt_corner_l_05}
      boxExtents=<30 40 40>
      constraintSetup {angularLimitsMax=<20 20 2>, angularLimitsMin=<-20 -20 -45>, angularTarget=<1 0 0>}
      localJointOffset=<7 0 0>
     }]
   }
   n57 = Anim Dynamics - Bone: dyn_side_skirt_l_02(ComponentPose=@n56, Alpha=1, ExternalForce="0,0,0") {
    angularDampingOverride=1
    angularSpringConstant=4000
    bAngularSpring=true
    bOverrideAngularDamping=true
    bUseGravityOverride=true
    gravityScale=0.35
    simulationSpace=RootRelative
    alphaBoolBlend {blendOption=Linear}
    boundBone {boneName=dyn_side_skirt_l_02}
    componentAppliedLinearAccClamp=<1000 1000 1000>
    componentLinearAccScale=<1 1 1>
    componentLinearVelScale=<1 1 1>
    gravityOverride {z=-2000}
    physicsBodyDefinitions [{
      boundBone {boneName=dyn_side_skirt_l_02}
      boxExtents=<30 40 40>
      constraintSetup {angularLimitsMax=<20 20 2>, angularLimitsMin=<-20 -20 -45>, angularTarget=<1 0 0>}
      localJointOffset=<7 0 0>
     }]
   }
   n58 = Anim Dynamics - Bone: dyn_side_skirt_l_03(ComponentPose=@n57, Alpha=1, ExternalForce="0,0,0") {
    angularDampingOverride=1
    angularSpringConstant=4000
    bAngularSpring=true
    bOverrideAngularDamping=true
    bUseGravityOverride=true
    gravityScale=0.35
    simulationSpace=RootRelative
    alphaBoolBlend {blendOption=Linear}
    boundBone {boneName=dyn_side_skirt_l_03}
    componentAppliedLinearAccClamp=<1000 1000 1000>
    componentLinearAccScale=<1 1 1>
    componentLinearVelScale=<1 1 1>
    gravityOverride {z=-2000}
    physicsBodyDefinitions [{
      boundBone {boneName=dyn_side_skirt_l_03}
      boxExtents=<30 40 40>
      constraintSetup {angularLimitsMax=<20 20 2>, angularLimitsMin=<-20 -20 -45>, angularTarget=<1 0 0>}
      localJointOffset=<7 0 0>
     }]
   }
   n59 = Anim Dynamics - Bone: dyn_side_skirt_l_04(ComponentPose=@n58, Alpha=1, ExternalForce="0,0,0") {
    angularDampingOverride=1
    angularSpringConstant=4000
    bAngularSpring=true
    bOverrideAngularDamping=true
    bUseGravityOverride=true
    gravityScale=0.35
    simulationSpace=RootRelative
    alphaBoolBlend {blendOption=Linear}
    boundBone {boneName=dyn_side_skirt_l_04}
    componentAppliedLinearAccClamp=<1000 1000 1000>
    componentLinearAccScale=<1 1 1>
    componentLinearVelScale=<1 1 1>
    gravityOverride {z=-2000}
    physicsBodyDefinitions [{
      boundBone {boneName=dyn_side_skirt_l_04}
      boxExtents=<30 40 40>
      constraintSetup {angularLimitsMax=<20 20 2>, angularLimitsMin=<-20 -20 -45>, angularTarget=<1 0 0>}
      localJointOffset=<7 0 0>
     }]
   }
   n60 = Anim Dynamics - Bone: dyn_side_skirt_l_05(ComponentPose=@n59, Alpha=1, ExternalForce="0,0,0") {
    angularDampingOverride=1
    angularSpringConstant=4000
    bAngularSpring=true
    bOverrideAngularDamping=true
    bUseGravityOverride=true
    gravityScale=0.35
    simulationSpace=RootRelative
    alphaBoolBlend {blendOption=Linear}
    boundBone {boneName=dyn_side_skirt_l_05}
    componentAppliedLinearAccClamp=<1000 1000 1000>
    componentLinearAccScale=<1 1 1>
    componentLinearVelScale=<1 1 1>
    gravityOverride {z=-2000}
    physicsBodyDefinitions [{
      boundBone {boneName=dyn_side_skirt_l_05}
      boxExtents=<30 40 40>
      constraintSetup {angularLimitsMax=<20 20 2>, angularLimitsMin=<-20 -20 -45>, angularTarget=<1 0 0>}
      localJointOffset=<7 0 0>
     }]
   }
   n61 = Anim Dynamics - Bone: dyn_side_skirt_r_02(ComponentPose=@n60, Alpha=1, ExternalForce="0,0,0") {
    angularDampingOverride=1
    angularSpringConstant=4000
    bAngularSpring=true
    bOverrideAngularDamping=true
    bUseGravityOverride=true
    gravityScale=0.35
    simulationSpace=RootRelative
    alphaBoolBlend {blendOption=Linear}
    boundBone {boneName=dyn_side_skirt_r_02}
    componentAppliedLinearAccClamp=<1000 1000 1000>
    componentLinearAccScale=<1 1 1>
    componentLinearVelScale=<1 1 1>
    gravityOverride {z=-2000}
    physicsBodyDefinitions [{
      boundBone {boneName=dyn_side_skirt_r_02}
      boxExtents=<30 40 40>
      constraintSetup {angularLimitsMax=<20 20 2>, angularLimitsMin=<-20 -20 -45>, angularTarget=<1 0 0>}
      localJointOffset=<-7 0 0>
     }]
   }
   n62 = Anim Dynamics - Bone: dyn_side_skirt_r_03(ComponentPose=@n61, Alpha=1, ExternalForce="0,0,0") {
    angularDampingOverride=1
    angularSpringConstant=4000
    bAngularSpring=true
    bOverrideAngularDamping=true
    bUseGravityOverride=true
    gravityScale=0.35
    simulationSpace=RootRelative
    alphaBoolBlend {blendOption=Linear}
    boundBone {boneName=dyn_side_skirt_r_03}
    componentAppliedLinearAccClamp=<1000 1000 1000>
    componentLinearAccScale=<1 1 1>
    componentLinearVelScale=<1 1 1>
    gravityOverride {z=-2000}
    physicsBodyDefinitions [{
      boundBone {boneName=dyn_side_skirt_r_03}
      boxExtents=<30 40 40>
      constraintSetup {angularLimitsMax=<20 20 2>, angularLimitsMin=<-20 -20 -45>, angularTarget=<1 0 0>}
      localJointOffset=<-7 0 0>
     }]
   }
   n63 = Anim Dynamics - Bone: dyn_side_skirt_r_04(ComponentPose=@n62, Alpha=1, ExternalForce="0,0,0") {
    angularDampingOverride=1
    angularSpringConstant=4000
    bAngularSpring=true
    bOverrideAngularDamping=true
    bUseGravityOverride=true
    gravityScale=0.35
    simulationSpace=RootRelative
    alphaBoolBlend {blendOption=Linear}
    boundBone {boneName=dyn_side_skirt_r_04}
    componentAppliedLinearAccClamp=<1000 1000 1000>
    componentLinearAccScale=<1 1 1>
    componentLinearVelScale=<1 1 1>
    gravityOverride {z=-2000}
    physicsBodyDefinitions [{
      boundBone {boneName=dyn_side_skirt_r_04}
      boxExtents=<30 40 40>
      constraintSetup {angularLimitsMax=<20 20 2>, angularLimitsMin=<-20 -20 -45>, angularTarget=<1 0 0>}
      localJointOffset=<-7 0 0>
     }]
   }
   n64 = Anim Dynamics - Bone: dyn_side_skirt_r_05(ComponentPose=@n63, Alpha=1, ExternalForce="0,0,0") {
    angularDampingOverride=1
    angularSpringConstant=4000
    bAngularSpring=true
    bOverrideAngularDamping=true
    bUseGravityOverride=true
    gravityScale=0.35
    simulationSpace=RootRelative
    alphaBoolBlend {blendOption=Linear}
    boundBone {boneName=dyn_side_skirt_r_05}
    componentAppliedLinearAccClamp=<1000 1000 1000>
    componentLinearAccScale=<1 1 1>
    componentLinearVelScale=<1 1 1>
    gravityOverride {z=-2000}
    physicsBodyDefinitions [{
      boundBone {boneName=dyn_side_skirt_r_05}
      boxExtents=<30 40 40>
      constraintSetup {angularLimitsMax=<20 20 2>, angularLimitsMin=<-20 -20 -45>, angularTarget=<1 0 0>}
      localJointOffset=<-7 0 0>
     }]
   }
   n65 = Anim Dynamics - Bone: dyn_front_skirt_r_02(ComponentPose=@n64, Alpha=1, ExternalForce="0,0,0") {
    angularDampingOverride=1
    angularSpringConstant=4000
    bAngularSpring=true
    bOverrideAngularDamping=true
    bUseGravityOverride=true
    gravityScale=0.35
    simulationSpace=RootRelative
    alphaBoolBlend {blendOption=Linear}
    boundBone {boneName=dyn_front_skirt_r_02}
    componentAppliedLinearAccClamp=<1000 1000 1000>
    componentLinearAccScale=<1 1 1>
    componentLinearVelScale=<1 1 1>
    gravityOverride {z=-2000}
    physicsBodyDefinitions [{
      boundBone {boneName=dyn_front_skirt_r_02}
      boxExtents=<30 40 40>
      constraintSetup {angularLimitsMax=<20 20 2>, angularLimitsMin=<-20 -20 -45>, angularTarget=<1 0 0>}
      localJointOffset=<-7 0 0>
     }]
   }
   n66 = Anim Dynamics - Bone: dyn_front_skirt_r_03(ComponentPose=@n65, Alpha=1, ExternalForce="0,0,0") {
    angularDampingOverride=1
    angularSpringConstant=4000
    bAngularSpring=true
    bOverrideAngularDamping=true
    bUseGravityOverride=true
    gravityScale=0.35
    simulationSpace=RootRelative
    alphaBoolBlend {blendOption=Linear}
    boundBone {boneName=dyn_front_skirt_r_03}
    componentAppliedLinearAccClamp=<1000 1000 1000>
    componentLinearAccScale=<1 1 1>
    componentLinearVelScale=<1 1 1>
    gravityOverride {z=-2000}
    physicsBodyDefinitions [{
      boundBone {boneName=dyn_front_skirt_r_03}
      boxExtents=<30 40 40>
      constraintSetup {angularLimitsMax=<20 20 2>, angularLimitsMin=<-20 -20 -45>, angularTarget=<1 0 0>}
      localJointOffset=<-7 0 0>
     }]
   }
   n67 = Anim Dynamics - Bone: dyn_front_skirt_r_04(ComponentPose=@n66, Alpha=1, ExternalForce="0,0,0") {
    angularDampingOverride=1
    angularSpringConstant=4000
    bAngularSpring=true
    bOverrideAngularDamping=true
    bUseGravityOverride=true
    gravityScale=0.35
    simulationSpace=RootRelative
    alphaBoolBlend {blendOption=Linear}
    boundBone {boneName=dyn_front_skirt_r_04}
    componentAppliedLinearAccClamp=<1000 1000 1000>
    componentLinearAccScale=<1 1 1>
    componentLinearVelScale=<1 1 1>
    gravityOverride {z=-2000}
    physicsBodyDefinitions [{
      boundBone {boneName=dyn_front_skirt_r_04}
      boxExtents=<30 40 40>
      constraintSetup {angularLimitsMax=<20 20 2>, angularLimitsMin=<-20 -20 -45>, angularTarget=<1 0 0>}
      localJointOffset=<-7 0 0>
     }]
   }
   n68 = Anim Dynamics - Bone: dyn_front_skirt_r_05(ComponentPose=@n67, Alpha=1, ExternalForce="0,0,0") {
    angularDampingOverride=1
    angularSpringConstant=4000
    bAngularSpring=true
    bOverrideAngularDamping=true
    bUseGravityOverride=true
    gravityScale=0.35
    simulationSpace=RootRelative
    alphaBoolBlend {blendOption=Linear}
    boundBone {boneName=dyn_front_skirt_r_05}
    componentAppliedLinearAccClamp=<1000 1000 1000>
    componentLinearAccScale=<1 1 1>
    componentLinearVelScale=<1 1 1>
    gravityOverride {z=-2000}
    physicsBodyDefinitions [{
      boundBone {boneName=dyn_front_skirt_r_05}
      boxExtents=<30 40 40>
      constraintSetup {angularLimitsMax=<20 20 2>, angularLimitsMin=<-20 -20 -45>, angularTarget=<1 0 0>}
      localJointOffset=<-7 0 0>
     }]
   }
   n69 = Anim Dynamics - Bone: dyn_front_skirt_corner_r_02(ComponentPose=@n68, Alpha=1, ExternalForce="0,0,0") {
    angularDampingOverride=1
    angularSpringConstant=4000
    bAngularSpring=true
    bOverrideAngularDamping=true
    bUseGravityOverride=true
    gravityScale=0.35
    simulationSpace=RootRelative
    alphaBoolBlend {blendOption=Linear}
    boundBone {boneName=dyn_front_skirt_corner_r_02}
    componentAppliedLinearAccClamp=<1000 1000 1000>
    componentLinearAccScale=<1 1 1>
    componentLinearVelScale=<1 1 1>
    gravityOverride {z=-2000}
    physicsBodyDefinitions [{
      boundBone {boneName=dyn_front_skirt_corner_r_02}
      boxExtents=<30 40 40>
      constraintSetup {angularLimitsMax=<20 20 2>, angularLimitsMin=<-20 -20 -45>, angularTarget=<1 0 0>}
      localJointOffset=<-7 0 0>
     }]
   }
   n70 = Anim Dynamics - Bone: dyn_front_skirt_corner_r_03(ComponentPose=@n69, Alpha=1, ExternalForce="0,0,0") {
    angularDampingOverride=1
    angularSpringConstant=4000
    bAngularSpring=true
    bOverrideAngularDamping=true
    bUseGravityOverride=true
    gravityScale=0.35
    simulationSpace=RootRelative
    alphaBoolBlend {blendOption=Linear}
    boundBone {boneName=dyn_front_skirt_corner_r_03}
    componentAppliedLinearAccClamp=<1000 1000 1000>
    componentLinearAccScale=<1 1 1>
    componentLinearVelScale=<1 1 1>
    gravityOverride {z=-2000}
    physicsBodyDefinitions [{
      boundBone {boneName=dyn_front_skirt_corner_r_03}
      boxExtents=<30 40 40>
      constraintSetup {angularLimitsMax=<20 20 2>, angularLimitsMin=<-20 -20 -45>, angularTarget=<1 0 0>}
      localJointOffset=<-7 0 0>
     }]
   }
   n71 = Anim Dynamics - Bone: dyn_front_skirt_corner_r_04(ComponentPose=@n70, Alpha=1, ExternalForce="0,0,0") {
    angularDampingOverride=1
    angularSpringConstant=4000
    bAngularSpring=true
    bOverrideAngularDamping=true
    bUseGravityOverride=true
    gravityScale=0.35
    simulationSpace=RootRelative
    alphaBoolBlend {blendOption=Linear}
    boundBone {boneName=dyn_front_skirt_corner_r_04}
    componentAppliedLinearAccClamp=<1000 1000 1000>
    componentLinearAccScale=<1 1 1>
    componentLinearVelScale=<1 1 1>
    gravityOverride {z=-2000}
    physicsBodyDefinitions [{
      boundBone {boneName=dyn_front_skirt_corner_r_04}
      boxExtents=<30 40 40>
      constraintSetup {angularLimitsMax=<20 20 2>, angularLimitsMin=<-20 -20 -45>, angularTarget=<1 0 0>}
      localJointOffset=<-7 0 0>
     }]
   }
   n72 = Anim Dynamics - Bone: dyn_front_skirt_corner_r_05(ComponentPose=@n71, Alpha=1, ExternalForce="0,0,0") {
    angularDampingOverride=1
    angularSpringConstant=4000
    bAngularSpring=true
    bOverrideAngularDamping=true
    bUseGravityOverride=true
    gravityScale=0.35
    simulationSpace=RootRelative
    alphaBoolBlend {blendOption=Linear}
    boundBone {boneName=dyn_front_skirt_corner_r_05}
    componentAppliedLinearAccClamp=<1000 1000 1000>
    componentLinearAccScale=<1 1 1>
    componentLinearVelScale=<1 1 1>
    gravityOverride {z=-2000}
    physicsBodyDefinitions [{
      boundBone {boneName=dyn_front_skirt_corner_r_05}
      boxExtents=<30 40 40>
      constraintSetup {angularLimitsMax=<20 20 2>, angularLimitsMin=<-20 -20 -45>, angularTarget=<1 0 0>}
      localJointOffset=<-7 0 0>
     }]
   }
   n73 = Anim Dynamics - Bone: dyn_rear_skirt_r_02(ComponentPose=@n72, Alpha=1, ExternalForce="0,0,0") {
    angularDampingOverride=1
    angularSpringConstant=4000
    bAngularSpring=true
    bOverrideAngularDamping=true
    bUseGravityOverride=true
    gravityScale=0.35
    simulationSpace=RootRelative
    alphaBoolBlend {blendOption=Linear}
    boundBone {boneName=dyn_rear_skirt_r_02}
    componentAppliedLinearAccClamp=<1000 1000 1000>
    componentLinearAccScale=<1 1 1>
    componentLinearVelScale=<1 1 1>
    gravityOverride {z=-2000}
    physicsBodyDefinitions [{
      boundBone {boneName=dyn_rear_skirt_r_02}
      boxExtents=<30 40 40>
      constraintSetup {angularLimitsMax=<20 20 2>, angularLimitsMin=<-20 -20 -45>, angularTarget=<1 0 0>}
      localJointOffset=<-7 0 0>
     }]
   }
   n74 = Anim Dynamics - Bone: dyn_rear_skirt_r_03(ComponentPose=@n73, Alpha=1, ExternalForce="0,0,0") {
    angularDampingOverride=1
    angularSpringConstant=4000
    bAngularSpring=true
    bOverrideAngularDamping=true
    bUseGravityOverride=true
    gravityScale=0.35
    simulationSpace=RootRelative
    alphaBoolBlend {blendOption=Linear}
    boundBone {boneName=dyn_rear_skirt_r_03}
    componentAppliedLinearAccClamp=<1000 1000 1000>
    componentLinearAccScale=<1 1 1>
    componentLinearVelScale=<1 1 1>
    gravityOverride {z=-2000}
    physicsBodyDefinitions [{
      boundBone {boneName=dyn_rear_skirt_r_03}
      boxExtents=<30 40 40>
      constraintSetup {angularLimitsMax=<20 20 2>, angularLimitsMin=<-20 -20 -45>, angularTarget=<1 0 0>}
      localJointOffset=<-7 0 0>
     }]
   }
   n75 = Anim Dynamics - Bone: dyn_rear_skirt_r_04(ComponentPose=@n74, Alpha=1, ExternalForce="0,0,0") {
    angularDampingOverride=1
    angularSpringConstant=4000
    bAngularSpring=true
    bOverrideAngularDamping=true
    bUseGravityOverride=true
    gravityScale=0.35
    simulationSpace=RootRelative
    alphaBoolBlend {blendOption=Linear}
    boundBone {boneName=dyn_rear_skirt_r_04}
    componentAppliedLinearAccClamp=<1000 1000 1000>
    componentLinearAccScale=<1 1 1>
    componentLinearVelScale=<1 1 1>
    gravityOverride {z=-2000}
    physicsBodyDefinitions [{
      boundBone {boneName=dyn_rear_skirt_r_04}
      boxExtents=<30 40 40>
      constraintSetup {angularLimitsMax=<20 20 2>, angularLimitsMin=<-20 -20 -45>, angularTarget=<1 0 0>}
      localJointOffset=<-7 0 0>
     }]
   }
   n76 = Anim Dynamics - Bone: dyn_rear_skirt_r_05(ComponentPose=@n75, Alpha=1, ExternalForce="0,0,0") {
    angularDampingOverride=1
    angularSpringConstant=4000
    bAngularSpring=true
    bOverrideAngularDamping=true
    bUseGravityOverride=true
    gravityScale=0.35
    simulationSpace=RootRelative
    alphaBoolBlend {blendOption=Linear}
    boundBone {boneName=dyn_rear_skirt_r_05}
    componentAppliedLinearAccClamp=<1000 1000 1000>
    componentLinearAccScale=<1 1 1>
    componentLinearVelScale=<1 1 1>
    gravityOverride {z=-2000}
    physicsBodyDefinitions [{
      boundBone {boneName=dyn_rear_skirt_r_05}
      boxExtents=<30 40 40>
      constraintSetup {angularLimitsMax=<20 20 2>, angularLimitsMin=<-20 -20 -45>, angularTarget=<1 0 0>}
      localJointOffset=<-7 0 0>
     }]
   }
   n77 = Anim Dynamics - Bone: dyn_rear_skirt_corner_r_02(ComponentPose=@n76, Alpha=1, ExternalForce="0,0,0") {
    angularDampingOverride=1
    angularSpringConstant=4000
    bAngularSpring=true
    bOverrideAngularDamping=true
    bUseGravityOverride=true
    gravityScale=0.35
    simulationSpace=RootRelative
    alphaBoolBlend {blendOption=Linear}
    boundBone {boneName=dyn_rear_skirt_corner_r_02}
    componentAppliedLinearAccClamp=<1000 1000 1000>
    componentLinearAccScale=<1 1 1>
    componentLinearVelScale=<1 1 1>
    gravityOverride {z=-2000}
    physicsBodyDefinitions [{
      boundBone {boneName=dyn_rear_skirt_corner_r_02}
      boxExtents=<30 40 40>
      constraintSetup {angularLimitsMax=<20 20 2>, angularLimitsMin=<-20 -20 -45>, angularTarget=<1 0 0>}
      localJointOffset=<-7 0 0>
     }]
   }
   n78 = Anim Dynamics - Bone: dyn_rear_skirt_corner_r_03(ComponentPose=@n77, Alpha=1, ExternalForce="0,0,0") {
    angularDampingOverride=1
    angularSpringConstant=4000
    bAngularSpring=true
    bOverrideAngularDamping=true
    bUseGravityOverride=true
    gravityScale=0.35
    simulationSpace=RootRelative
    alphaBoolBlend {blendOption=Linear}
    boundBone {boneName=dyn_rear_skirt_corner_r_03}
    componentAppliedLinearAccClamp=<1000 1000 1000>
    componentLinearAccScale=<1 1 1>
    componentLinearVelScale=<1 1 1>
    gravityOverride {z=-2000}
    physicsBodyDefinitions [{
      boundBone {boneName=dyn_rear_skirt_corner_r_03}
      boxExtents=<30 40 40>
      constraintSetup {angularLimitsMax=<20 20 2>, angularLimitsMin=<-20 -20 -45>, angularTarget=<1 0 0>}
      localJointOffset=<-7 0 0>
     }]
   }
   n79 = Anim Dynamics - Bone: dyn_rear_skirt_corner_r_04(ComponentPose=@n78, Alpha=1, ExternalForce="0,0,0") {
    angularDampingOverride=1
    angularSpringConstant=4000
    bAngularSpring=true
    bOverrideAngularDamping=true
    bUseGravityOverride=true
    gravityScale=0.35
    simulationSpace=RootRelative
    alphaBoolBlend {blendOption=Linear}
    boundBone {boneName=dyn_rear_skirt_corner_r_04}
    componentAppliedLinearAccClamp=<1000 1000 1000>
    componentLinearAccScale=<1 1 1>
    componentLinearVelScale=<1 1 1>
    gravityOverride {z=-2000}
    physicsBodyDefinitions [{
      boundBone {boneName=dyn_rear_skirt_corner_r_04}
      boxExtents=<30 40 40>
      constraintSetup {angularLimitsMax=<20 20 2>, angularLimitsMin=<-20 -20 -45>, angularTarget=<1 0 0>}
      localJointOffset=<-7 0 0>
     }]
   }
   n80 = Anim Dynamics - Bone: dyn_rear_skirt_corner_r_05(ComponentPose=@n79, Alpha=1, ExternalForce="0,0,0") {
    angularDampingOverride=1
    angularSpringConstant=4000
    bAngularSpring=true
    bOverrideAngularDamping=true
    bUseGravityOverride=true
    gravityScale=0.35
    simulationSpace=RootRelative
    alphaBoolBlend {blendOption=Linear}
    boundBone {boneName=dyn_rear_skirt_corner_r_05}
    componentAppliedLinearAccClamp=<1000 1000 1000>
    componentLinearAccScale=<1 1 1>
    componentLinearVelScale=<1 1 1>
    gravityOverride {z=-2000}
    physicsBodyDefinitions [{
      boundBone {boneName=dyn_rear_skirt_corner_r_05}
      boxExtents=<30 40 40>
      constraintSetup {angularLimitsMax=<20 20 2>, angularLimitsMin=<-20 -20 -45>, angularTarget=<1 0 0>}
      localJointOffset=<-7 0 0>
     }]
   }
   n81 = Look At - Bone: dyn_bow_l_01(ComponentPose=@n80, Alpha=1) {
    alphaBoolBlend {blendOption=Linear}
    boneToModify {boneName=dyn_bow_l_01}
    lookAtLocation=<10 -8.7215 0>
    lookAtTarget {boneReference {boneName=dyn_rear_skirt_l_04}}
    lookAt_Axis {axis=<1 0>}
    lookUp_Axis {axis {x=0, z=1}}
   }
   n82 = Look At - Bone: dyn_bow_r_01(ComponentPose=@n81, Alpha=1) {
    alphaBoolBlend {blendOption=Linear}
    boneToModify {boneName=dyn_bow_r_01}
    lookAtLocation=<6.8985 5.2119 -3.7254>
    lookAtTarget {boneReference {boneName=dyn_rear_skirt_r_04}}
    lookAt_Axis {axis=<-1 0>}
    lookUp_Axis {axis {x=0, z=-1}}
   }
   n83 = Look At - Bone: dyn_bow_l_03(ComponentPose=@n82, Alpha=1) {
    alphaBoolBlend {blendOption=Linear}
    boneToModify {boneName=dyn_bow_l_03}
    lookAtLocation=<7.7428 -8.7215 0>
    lookAtTarget {boneReference {boneName=dyn_rear_skirt_l_05}}
    lookAt_Axis {axis=<1 0>}
    lookUp_Axis {axis {x=0, z=1}}
   }
   n84 = Look At - Bone: dyn_bow_r_03(ComponentPose=@n83, Alpha=1) {
    alphaBoolBlend {blendOption=Linear}
    boneToModify {boneName=dyn_bow_r_03}
    lookAtLocation=<7.5897 5.7504 -3.7254>
    lookAtTarget {boneReference {boneName=dyn_rear_skirt_r_05}}
    lookAt_Axis {axis=<-1 0>}
    lookUp_Axis {axis {x=0, z=-1}}
   }
   n85 = Anim Dynamics - Bone: dyn_bow_loop_l_01(ComponentPose=@n84, Alpha=1, ExternalForce="0,0,0") {
    angularDampingOverride=1
    angularSpringConstant=20000
    bAngularSpring=true
    bOverrideAngularDamping=true
    bUseGravityOverride=true
    gravityScale=0.35
    simulationSpace=RootRelative
    alphaBoolBlend {blendOption=Linear}
    boundBone {boneName=dyn_bow_loop_l_01}
    componentAppliedLinearAccClamp=<1000 1000 1000>
    componentLinearAccScale=<1 1 1>
    componentLinearVelScale=<1 1 1>
    gravityOverride {z=-980}
    physicsBodyDefinitions [{
      boundBone {boneName=dyn_bow_loop_l_01}
      boxExtents=<20 30 30>
      constraintSetup {angularLimitsMax=<15 15 15>, angularLimitsMin=<-15 -15 -15>, angularTarget=<1 0 0>}
      localJointOffset=<4 0 0>
     }]
   }
   n86 = Anim Dynamics - Bone: dyn_bow_loop_l_02(ComponentPose=@n85, Alpha=1, ExternalForce="0,0,0") {
    angularDampingOverride=1
    angularSpringConstant=20000
    bAngularSpring=true
    bOverrideAngularDamping=true
    bUseGravityOverride=true
    gravityScale=0.35
    simulationSpace=RootRelative
    alphaBoolBlend {blendOption=Linear}
    boundBone {boneName=dyn_bow_loop_l_02}
    componentAppliedLinearAccClamp=<1000 1000 1000>
    componentLinearAccScale=<1 1 1>
    componentLinearVelScale=<1 1 1>
    gravityOverride {z=-980}
    physicsBodyDefinitions [{
      boundBone {boneName=dyn_bow_loop_l_02}
      boxExtents=<20 30 30>
      constraintSetup {angularLimitsMax=<15 15 15>, angularLimitsMin=<-15 -15 -15>, angularTarget=<1 0 0>}
      localJointOffset=<4 0 0>
     }]
   }
   n87 = Anim Dynamics - Bone: dyn_bow_loop_r_01(ComponentPose=@n86, Alpha=1, ExternalForce="0,0,0") {
    angularDampingOverride=1
    angularSpringConstant=20000
    bAngularSpring=true
    bOverrideAngularDamping=true
    bUseGravityOverride=true
    gravityScale=0.35
    simulationSpace=RootRelative
    alphaBoolBlend {blendOption=Linear}
    boundBone {boneName=dyn_bow_loop_r_01}
    componentAppliedLinearAccClamp=<1000 1000 1000>
    componentLinearAccScale=<1 1 1>
    componentLinearVelScale=<1 1 1>
    gravityOverride {z=-980}
    physicsBodyDefinitions [{
      boundBone {boneName=dyn_bow_loop_r_01}
      boxExtents=<20 30 30>
      constraintSetup {angularLimitsMax=<15 15 15>, angularLimitsMin=<-15 -15 -15>, angularTarget=<1 0 0>}
      localJointOffset=<-4 0 0>
     }]
   }
   n88 = Anim Dynamics - Bone: dyn_bow_loop_r_02(ComponentPose=@n87, Alpha=1, ExternalForce="0,0,0") {
    angularDampingOverride=1
    angularSpringConstant=20000
    bAngularSpring=true
    bOverrideAngularDamping=true
    bUseGravityOverride=true
    gravityScale=0.35
    simulationSpace=RootRelative
    alphaBoolBlend {blendOption=Linear}
    boundBone {boneName=dyn_bow_loop_r_02}
    componentAppliedLinearAccClamp=<1000 1000 1000>
    componentLinearAccScale=<1 1 1>
    componentLinearVelScale=<1 1 1>
    gravityOverride {z=-980}
    physicsBodyDefinitions [{
      boundBone {boneName=dyn_bow_loop_r_02}
      boxExtents=<20 30 30>
      constraintSetup {angularLimitsMax=<15 15 15>, angularLimitsMin=<-15 -15 -15>, angularTarget=<1 0 0>}
      localJointOffset=<-4 0 0>
     }]
   }
   n89 = Anim Dynamics - Bone: dyn_bow_l_01(ComponentPose=@n88, Alpha=1, ExternalForce="0,0,0") {
    angularDampingOverride=1
    angularSpringConstant=20000
    bAngularSpring=true
    bOverrideAngularDamping=true
    bUseGravityOverride=true
    gravityScale=0.35
    simulationSpace=RootRelative
    alphaBoolBlend {blendOption=Linear}
    boundBone {boneName=dyn_bow_l_01}
    componentAppliedLinearAccClamp=<1000 1000 1000>
    componentLinearAccScale=<1 1 1>
    componentLinearVelScale=<1 1 1>
    gravityOverride {z=-980}
    physicsBodyDefinitions [{
      boundBone {boneName=dyn_bow_l_01}
      boxExtents=<20 30 30>
      constraintSetup {angularLimitsMax=<15 15 0>, angularLimitsMin=<-15 -15 -15>, angularTarget=<1 0 0>}
      localJointOffset=<4 0 0>
     }]
   }
   n90 = Anim Dynamics - Bone: dyn_bow_l_02(ComponentPose=@n89, Alpha=1, ExternalForce="0,0,0") {
    angularDampingOverride=1
    angularSpringConstant=20000
    bAngularSpring=true
    bOverrideAngularDamping=true
    bUseGravityOverride=true
    gravityScale=0.35
    simulationSpace=RootRelative
    alphaBoolBlend {blendOption=Linear}
    boundBone {boneName=dyn_bow_l_02}
    componentAppliedLinearAccClamp=<1000 1000 1000>
    componentLinearAccScale=<1 1 1>
    componentLinearVelScale=<1 1 1>
    gravityOverride {z=-980}
    physicsBodyDefinitions [{
      boundBone {boneName=dyn_bow_l_02}
      boxExtents=<20 30 30>
      constraintSetup {angularLimitsMax=<15 15 0>, angularLimitsMin=<-15 -15 -15>, angularTarget=<1 0 0>}
      localJointOffset=<4 0 0>
     }]
   }
   n91 = Anim Dynamics - Bone: dyn_bow_l_03(ComponentPose=@n90, Alpha=1, ExternalForce="0,0,0") {
    angularDampingOverride=1
    angularSpringConstant=20000
    bAngularSpring=true
    bOverrideAngularDamping=true
    bUseGravityOverride=true
    gravityScale=0.35
    simulationSpace=RootRelative
    alphaBoolBlend {blendOption=Linear}
    boundBone {boneName=dyn_bow_l_03}
    componentAppliedLinearAccClamp=<1000 1000 1000>
    componentLinearAccScale=<1 1 1>
    componentLinearVelScale=<1 1 1>
    gravityOverride {z=-980}
    physicsBodyDefinitions [{
      boundBone {boneName=dyn_bow_l_03}
      boxExtents=<20 30 30>
      constraintSetup {angularLimitsMax=<15 15 0>, angularLimitsMin=<-15 -15 -15>, angularTarget=<1 0 0>}
      localJointOffset=<4 0 0>
     }]
   }
   n92 = Anim Dynamics - Bone: dyn_bow_l_04(ComponentPose=@n91, Alpha=1, ExternalForce="0,0,0") {
    angularDampingOverride=1
    angularSpringConstant=20000
    bAngularSpring=true
    bOverrideAngularDamping=true
    bUseGravityOverride=true
    gravityScale=0.35
    simulationSpace=RootRelative
    alphaBoolBlend {blendOption=Linear}
    boundBone {boneName=dyn_bow_l_04}
    componentAppliedLinearAccClamp=<1000 1000 1000>
    componentLinearAccScale=<1 1 1>
    componentLinearVelScale=<1 1 1>
    gravityOverride {z=-980}
    physicsBodyDefinitions [{
      boundBone {boneName=dyn_bow_l_04}
      boxExtents=<20 30 30>
      constraintSetup {angularLimitsMax=<15 15 0>, angularLimitsMin=<-15 -15 -15>, angularTarget=<1 0 0>}
      localJointOffset=<4 0 0>
     }]
   }
   n93 = Anim Dynamics - Bone: dyn_bow_l_05(ComponentPose=@n92, Alpha=1, ExternalForce="0,0,0") {
    angularDampingOverride=1
    angularSpringConstant=20000
    bAngularSpring=true
    bOverrideAngularDamping=true
    bUseGravityOverride=true
    gravityScale=0.35
    simulationSpace=RootRelative
    alphaBoolBlend {blendOption=Linear}
    boundBone {boneName=dyn_bow_l_05}
    componentAppliedLinearAccClamp=<1000 1000 1000>
    componentLinearAccScale=<1 1 1>
    componentLinearVelScale=<1 1 1>
    gravityOverride {z=-980}
    physicsBodyDefinitions [{
      boundBone {boneName=dyn_bow_l_05}
      boxExtents=<20 30 30>
      constraintSetup {angularLimitsMax=<15 15 0>, angularLimitsMin=<-15 -15 -15>, angularTarget=<1 0 0>}
      localJointOffset=<4 0 0>
     }]
   }
   n94 = Anim Dynamics - Bone: dyn_bow_r_01(ComponentPose=@n93, Alpha=1, ExternalForce="0,0,0") {
    angularDampingOverride=1
    angularSpringConstant=20000
    bAngularSpring=true
    bOverrideAngularDamping=true
    bUseGravityOverride=true
    gravityScale=0.35
    simulationSpace=RootRelative
    alphaBoolBlend {blendOption=Linear}
    boundBone {boneName=dyn_bow_r_01}
    componentAppliedLinearAccClamp=<1000 1000 1000>
    componentLinearAccScale=<1 1 1>
    componentLinearVelScale=<1 1 1>
    gravityOverride {z=-980}
    physicsBodyDefinitions [{
      boundBone {boneName=dyn_bow_r_01}
      boxExtents=<20 30 30>
      constraintSetup {angularLimitsMax=<15 15 0>, angularLimitsMin=<-15 -15 -15>, angularTarget=<1 0 0>}
      localJointOffset=<-4 0 0>
     }]
   }
   n95 = Anim Dynamics - Bone: dyn_bow_r_02(ComponentPose=@n94, Alpha=1, ExternalForce="0,0,0") {
    angularDampingOverride=1
    angularSpringConstant=20000
    bAngularSpring=true
    bOverrideAngularDamping=true
    bUseGravityOverride=true
    gravityScale=0.35
    simulationSpace=RootRelative
    alphaBoolBlend {blendOption=Linear}
    boundBone {boneName=dyn_bow_r_02}
    componentAppliedLinearAccClamp=<1000 1000 1000>
    componentLinearAccScale=<1 1 1>
    componentLinearVelScale=<1 1 1>
    gravityOverride {z=-980}
    physicsBodyDefinitions [{
      boundBone {boneName=dyn_bow_r_02}
      boxExtents=<20 30 30>
      constraintSetup {angularLimitsMax=<15 15 0>, angularLimitsMin=<-15 -15 -15>, angularTarget=<1 0 0>}
      localJointOffset=<-4 0 0>
     }]
   }
   n96 = Anim Dynamics - Bone: dyn_bow_r_03(ComponentPose=@n95, Alpha=1, ExternalForce="0,0,0") {
    angularDampingOverride=1
    angularSpringConstant=20000
    bAngularSpring=true
    bOverrideAngularDamping=true
    bUseGravityOverride=true
    gravityScale=0.35
    simulationSpace=RootRelative
    alphaBoolBlend {blendOption=Linear}
    boundBone {boneName=dyn_bow_r_03}
    componentAppliedLinearAccClamp=<1000 1000 1000>
    componentLinearAccScale=<1 1 1>
    componentLinearVelScale=<1 1 1>
    gravityOverride {z=-980}
    physicsBodyDefinitions [{
      boundBone {boneName=dyn_bow_r_03}
      boxExtents=<20 30 30>
      constraintSetup {angularLimitsMax=<15 15 0>, angularLimitsMin=<-15 -15 -15>, angularTarget=<1 0 0>}
      localJointOffset=<-4 0 0>
     }]
   }
   n97 = Anim Dynamics - Bone: dyn_bow_r_04(ComponentPose=@n96, Alpha=1, ExternalForce="0,0,0") {
    angularDampingOverride=1
    angularSpringConstant=20000
    bAngularSpring=true
    bOverrideAngularDamping=true
    bUseGravityOverride=true
    gravityScale=0.35
    simulationSpace=RootRelative
    alphaBoolBlend {blendOption=Linear}
    boundBone {boneName=dyn_bow_r_04}
    componentAppliedLinearAccClamp=<1000 1000 1000>
    componentLinearAccScale=<1 1 1>
    componentLinearVelScale=<1 1 1>
    gravityOverride {z=-980}
    physicsBodyDefinitions [{
      boundBone {boneName=dyn_bow_r_04}
      boxExtents=<20 30 30>
      constraintSetup {angularLimitsMax=<15 15 0>, angularLimitsMin=<-15 -15 -15>, angularTarget=<1 0 0>}
      localJointOffset=<-4 0 0>
     }]
   }
   n98 = Anim Dynamics - Bone: dyn_bow_r_05(ComponentPose=@n97, Alpha=1, ExternalForce="0,0,0") {
    angularDampingOverride=1
    angularSpringConstant=20000
    bAngularSpring=true
    bOverrideAngularDamping=true
    bUseGravityOverride=true
    gravityScale=0.35
    simulationSpace=RootRelative
    alphaBoolBlend {blendOption=Linear}
    boundBone {boneName=dyn_bow_r_05}
    componentAppliedLinearAccClamp=<1000 1000 1000>
    componentLinearAccScale=<1 1 1>
    componentLinearVelScale=<1 1 1>
    gravityOverride {z=-980}
    physicsBodyDefinitions [{
      boundBone {boneName=dyn_bow_r_05}
      boxExtents=<20 30 30>
      constraintSetup {angularLimitsMax=<15 15 0>, angularLimitsMin=<-15 -15 -15>, angularTarget=<1 0 0>}
      localJointOffset=<-4 0 0>
     }]
   }
   Save cached pose 'Post Anim Dynamic Skirt'(Pose=Component To Local(ComponentPose=@n98)) {cacheName=Post Anim Dynamic Skirt}
  ]}
 EventGraph {nodes [
   Event BlueprintUpdateAnimation()
   > n1 = Sequence!(execute) -> (then_0, then_1, then_2:Exec)
   > n2 = Gravity Override State Function DEPRECATED(
    CharacterPartAnimInstance=Self-Reference()
    GameTimeInSeconds=GetGameTimeInSeconds()
    GravityOverrideParams=Get Gravity Override Params()
    RBAN Params=Get RBAN Params()
   )
   > Set Gravity Override(Gravity Override=@n2)
   > Set RBAN Settings Out(RBAN Settings Out:RBANSetting Struct Structure=@n2.RBAN SettingsOut) -> (Output_Get:RBANSetting Struct Structure)
   Set MoveGroundCollision(execute=@n1.then_1, MoveGroundCollision:bool=OR Boolean(A=Get bIsPlayingEmote(), B=Get bIsOnGrindRail())) -> (Output_Get:bool)
   n3 = Branch(execute=@n1.then_2, Condition=Get bIsInFrontEnd())
   > Set Simulation Space(Simulation Space:AnimPhysSimSpaceType Enum=World) -> (Output_Get:AnimPhysSimSpaceType Enum)
   > Set Alpha(Alpha:double=0.5) -> (Output_Get:double)
   > Set Custom Idle Is Playing(Custom Idle Is Playing:float=GetCurveValue(self=GetOwnerFrontEndAnimInstance(), CurveName=Frontend_Idle_Bypass)) -> (Output_Get:float)
   Set Simulation Space(execute=@n3.else, Simulation Space=Component)
   > Set Alpha(Alpha=1)
  ]}
]
variables [
 Alpha {type=number}
 Custom Idle Is Playing {type=number}
 Gravity Override {type=Vector}
 Gravity Override Params {
  type=GravityOverrideParamsStruct
  default {
   onGround_Moving {gravityOverride=<0 -400 -700>}
   skydive_DivingDown {windAmplitude=<50 600 400>, windFrequency=<3 24 20>}
   skydive_FloatingDown {gravityOverride=<0 -700 400>, windAmplitude=<100 -100 150>, windFrequency=<5 12 16>}
   skydive_Parachuting {windAmplitude=<100 -300 200>}
   tactical_Sprint {gravityOverride=<0 -600 -500>}
  }
 }
 MoveGroundCollision {type=boolean}
 RBAN Params {
  type=RBANParamsStruct
  default {
   dBNO {simSpaceSettings {
     maxAngularAcceleration=10
     maxAngularVelocity=10
     maxLinearAcceleration=400
     maxLinearVelocity=400
     velocityScaleZ=0.5
     worldAlpha=0.5
     externalLinearDragV=<0.5 0.5 0.5>
    }}
   dBNO_Carried {simSpaceSettings {
     maxAngularAcceleration=10
     maxAngularVelocity=10
     maxLinearAcceleration=400
     maxLinearVelocity=400
     velocityScaleZ=0.5
     worldAlpha=0.5
     externalLinearDragV=<0.5 0.5 0.5>
    }}
   emoteOrMelee {simSpaceSettings {
     maxAngularAcceleration=5
     maxAngularVelocity=5
     maxLinearAcceleration=100
     maxLinearVelocity=100
     velocityScaleZ=0.25
     worldAlpha=0.25
     externalLinearDragV=<0 0 0>
    }}
   front_End {simSpaceSettings {
     maxAngularAcceleration=8
     maxAngularVelocity=8
     maxLinearAcceleration=300
     maxLinearVelocity=300
     velocityScaleZ=0.2
     worldAlpha=0.3
     externalLinearDragV=<0.5 0.5 0.5>
    }}
   onGround_Crouch_Moving {simSpaceSettings {
     maxAngularAcceleration=10
     maxAngularVelocity=10
     maxLinearAcceleration=400
     maxLinearVelocity=400
     velocityScaleZ=0.5
     worldAlpha=0.5
     externalLinearDragV=<0.5 0.5 0.5>
    }}
   onGround_Falling {simSpaceSettings {
     maxAngularAcceleration=10
     maxAngularVelocity=10
     maxLinearAcceleration=400
     maxLinearVelocity=400
     velocityScaleZ=0.5
     worldAlpha=0.5
     externalLinearDragV=<0.5 0.5 0.5>
    }}
   onGround_Moving {
    componentLinearAccScale=<0 0 0>
    componentLinearVelScale=<0 0 0>
    simSpaceSettings {
     maxAngularAcceleration=10
     maxAngularVelocity=10
     maxLinearAcceleration=300
     maxLinearVelocity=300
     velocityScaleZ=0.5
     worldAlpha=0.5
     externalLinearDragV=<1 1 1>
    }
   }
   onGround_Standing {
    componentLinearAccScale=<0 0 0>
    componentLinearVelScale=<0 0 0>
    simSpaceSettings {
     maxAngularAcceleration=10
     maxAngularVelocity=10
     maxLinearAcceleration=400
     maxLinearVelocity=400
     velocityScaleZ=0.5
     worldAlpha=0.5
     externalLinearDragV=<0.5 0.5 0.5>
    }
   }
   rail_Grinding {simSpaceSettings {
     maxAngularAcceleration=10
     maxAngularVelocity=10
     maxLinearAcceleration=400
     maxLinearVelocity=400
     velocityScaleZ=0.5
     worldAlpha=0.5
     externalLinearDragV=<0.5 0.5 0.5>
    }}
   riding_Moving {simSpaceSettings {
     maxAngularAcceleration=10
     maxAngularVelocity=10
     maxLinearAcceleration=400
     maxLinearVelocity=400
     velocityScaleZ=0.5
     worldAlpha=0.5
     externalLinearDragV=<0.5 0.5 0.5>
    }}
   riding_Moving_Targeting {simSpaceSettings {
     maxAngularAcceleration=10
     maxAngularVelocity=10
     maxLinearAcceleration=400
     maxLinearVelocity=400
     velocityScaleZ=0.5
     worldAlpha=0.5
     externalLinearDragV=<0.5 0.5 0.5>
    }}
   riding_Standing {simSpaceSettings {
     maxAngularAcceleration=10
     maxAngularVelocity=10
     maxLinearAcceleration=400
     maxLinearVelocity=400
     velocityScaleZ=0.5
     worldAlpha=0.5
     externalLinearDragV=<0.5 0.5 0.5>
    }}
   skydive_DivingDown {
    componentLinearAccScale=<0 0 0>
    componentLinearVelScale=<0 0 0>
    simSpaceSettings {
     maxAngularAcceleration=10
     maxAngularVelocity=10
     maxLinearAcceleration=400
     maxLinearVelocity=400
     velocityScaleZ=0.5
     worldAlpha=0.5
     externalLinearDragV=<0.5 0.5 0.5>
    }
   }
   skydive_FloatingDown {
    componentLinearAccScale=<0 0 0>
    componentLinearVelScale=<0 0 0>
    simSpaceSettings {
     maxAngularAcceleration=10
     maxAngularVelocity=10
     maxLinearAcceleration=400
     maxLinearVelocity=400
     velocityScaleZ=0.1
     worldAlpha=0.1
     externalLinearDragV=<0.5 0.5 0.5>
    }
   }
   skydive_GoingUp {simSpaceSettings {
     maxAngularAcceleration=10
     maxAngularVelocity=10
     maxLinearAcceleration=400
     maxLinearVelocity=400
     velocityScaleZ=0.5
     worldAlpha=0.5
     externalLinearDragV=<0.5 0.5 0.5>
    }}
   skydive_Parachuting {
    componentLinearAccScale=<0 0 0>
    componentLinearVelScale=<0 0 0>
    simSpaceSettings {
     maxAngularAcceleration=10
     maxAngularVelocity=10
     maxLinearAcceleration=400
     maxLinearVelocity=400
     velocityScaleZ=0.1
     worldAlpha=0.1
     externalLinearDragV=<0.5 0.5 0.5>
    }
   }
   skydive_Surfing {simSpaceSettings {
     maxAngularAcceleration=10
     maxAngularVelocity=10
     maxLinearAcceleration=400
     maxLinearVelocity=400
     velocityScaleZ=0.5
     worldAlpha=0.5
     externalLinearDragV=<0.5 0.5 0.5>
    }}
   swimming {simSpaceSettings {
     maxAngularAcceleration=10
     maxAngularVelocity=10
     maxLinearAcceleration=400
     maxLinearVelocity=400
     velocityScaleZ=0.5
     worldAlpha=0.5
     externalLinearDragV=<0.5 0.5 0.5>
    }}
   tactical_Sprint {
    componentLinearAccScale=<0 0 0>
    componentLinearVelScale=<0 0 0>
    simSpaceSettings {
     maxAngularAcceleration=5
     maxAngularVelocity=5
     maxLinearAcceleration=200
     maxLinearVelocity=200
     velocityScaleZ=0.25
     worldAlpha=0.25
     externalLinearDragV=<0.5 0.5 0.5>
    }
   }
   ziplining {simSpaceSettings {
     maxAngularAcceleration=10
     maxAngularVelocity=10
     maxLinearAcceleration=400
     maxLinearVelocity=400
     velocityScaleZ=0.5
     worldAlpha=0.5
     externalLinearDragV=<0.5 0.5 0.5>
    }}
  }
 }
 RBAN Settings Out {
  type=RBANSettingStruct
  default {
   componentLinearAccScale=<0 0 0>
   componentLinearVelScale=<0 0 0>
   simSpaceSettings {
    dampingAlpha=1
    maxAngularAcceleration=1000
    maxAngularVelocity=3
    maxLinearAcceleration=300
    maxLinearVelocity=700
    velocityScaleZ=0
    worldAlpha=1
    externalAngularVelocity=<0 0 0>
    externalLinearDragV=<0 0.5 0>
    externalLinearVelocity=<0 0 0>
   }
  }
 }
 Simulation Space {type=AnimPhysSimSpaceType, default=RootRelative}
]
