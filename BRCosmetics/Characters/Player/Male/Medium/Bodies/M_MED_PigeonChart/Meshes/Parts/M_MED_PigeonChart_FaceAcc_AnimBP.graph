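# Animation Blueprint for the PigeonChart male medium character's face accessory mesh, adding procedural secondary motion to its dyn_hair bones.
# It copies the attached parent mesh's pose, converts it to component space and passes it through a long series of Anim Dynamics nodes.
# Each node simulates one bone as a single body with shared spring, damping, gravity and angular-limit settings, varying only box extents and joint offsets.
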AnimBlueprint /BRCosmetics/Characters/Player/Male/Medium/Bodies/M_MED_PigeonChart/Meshes/Parts/M_MED_PigeonChart_FaceAcc_AnimBP
target_skeleton=&/M_MED_PigeonChart_FaceAcc_Skeleton
graphs [
 AnimGraph {nodes [
   n1 = Anim Dynamics - Bone: dyn_hair_01!(
    SimulationSpace:AnimPhysSimSpaceType Enum=Component
    ComponentPose=Local To Component(LocalPose=Copy Pose From Mesh() {bUseAttachedParent=true})
    Alpha=1
    ExternalForce
   ) -> (Pose) {
    angularDampingOverride=0.99
    angularSpringConstant=7500
    bAngularSpring=true
    bOverrideAngularDamping=true
    bUseGravityOverride=true
    bUsePlanarLimit=false
    gravityScale=0.35
    alphaBoolBlend {blendOption=Linear}
    boundBone {boneName=dyn_hair_01}
    componentAppliedLinearAccClamp=<1000 1000 1000>
    componentLinearAccScale=<0.5 0.5 0.5>
    componentLinearVelScale=<0.5 0.5 0.5>
    gravityOverride {z=-200}
    physicsBodyDefinitions [{
      boundBone {boneName=dyn_hair_01}
      boxExtents=<25 30 30>
      constraintSetup {angularLimitsMax=<20 30 20>, angularLimitsMin=<-20 -30 -20>, angularTarget=<1 0 0>}
      localJointOffset=<7 0 0>
     }]
   }
   n2 = Anim Dynamics - Bone: dyn_hair_02!(SimulationSpace:AnimPhysSimSpaceType Enum=Component, ComponentPose=@n1, Alpha=1, ExternalForce) -> (Pose) {
    angularDampingOverride=0.99
    angularSpringConstant=7500
    bAngularSpring=true
    bOverrideAngularDamping=true
    bUseGravityOverride=true
    bUsePlanarLimit=false
    gravityScale=0.35
    alphaBoolBlend {blendOption=Linear}
    boundBone {boneName=dyn_hair_02}
    componentAppliedLinearAccClamp=<1000 1000 1000>
    componentLinearAccScale=<0.5 0.5 0.5>
    componentLinearVelScale=<0.5 0.5 0.5>
    gravityOverride {z=-200}
    physicsBodyDefinitions [{
      boundBone {boneName=dyn_hair_02}
      boxExtents=<25 30 30>
      constraintSetup {angularLimitsMax=<20 30 20>, angularLimitsMin=<-20 -30 -20>, angularTarget=<1 0 0>}
      localJointOffset=<7 4 0>
     }]
   }
   n3 = Anim Dynamics - Bone: dyn_hair_03!(SimulationSpace:AnimPhysSimSpaceType Enum=Component, ComponentPose=@n2, Alpha=1, ExternalForce="0,0,0") -> (Pose) {
    angularDampingOverride=0.99
    angularSpringConstant=7500
    bAngularSpring=true
    bOverrideAngularDamping=true
    bUseGravityOverride=true
    bUsePlanarLimit=false
    gravityScale=0.35
    alphaBoolBlend {blendOption=Linear}
    boundBone {boneName=dyn_hair_03}
    componentAppliedLinearAccClamp=<1000 1000 1000>
    componentLinearAccScale=<0.5 0.5 0.5>
    componentLinearVelScale=<0.5 0.5 0.5>
    gravityOverride {z=-200}
    physicsBodyDefinitions [{
      boundBone {boneName=dyn_hair_03}
      boxExtents=<25 30 30>
      constraintSetup {angularLimitsMax=<20 30 20>, angularLimitsMin=<-20 -30 -20>, angularTarget=<1 0 0>}
      localJointOffset=<7 -4 0>
     }]
   }
   n4 = Anim Dynamics - Bone: dyn_hair_04!(SimulationSpace:AnimPhysSimSpaceType Enum=Component, ComponentPose=@n3, Alpha=1, ExternalForce="0,0,0") -> (Pose) {
    angularDampingOverride=0.99
    angularSpringConstant=7500
    bAngularSpring=true
    bOverrideAngularDamping=true
    bUseGravityOverride=true
    bUsePlanarLimit=false
    gravityScale=0.35
    alphaBoolBlend {blendOption=Linear}
    boundBone {boneName=dyn_hair_04}
    componentAppliedLinearAccClamp=<1000 1000 1000>
    componentLinearAccScale=<0.5 0.5 0.5>
    componentLinearVelScale=<0.5 0.5 0.5>
    gravityOverride {z=-200}
    physicsBodyDefinitions [{
      boundBone {boneName=dyn_hair_04}
      boxExtents=<25 30 30>
      constraintSetup {angularLimitsMax=<20 30 20>, angularLimitsMin=<-20 -30 -20>, angularTarget=<1 0 0>}
      localJointOffset=<7 0 4>
     }]
   }
   n5 = Anim Dynamics - Bone: dyn_hair_05!(SimulationSpace:AnimPhysSimSpaceType Enum=Component, ComponentPose=@n4, Alpha=1, ExternalForce="0,0,0") -> (Pose) {
    angularDampingOverride=0.99
    angularSpringConstant=7500
    bAngularSpring=true
    bOverrideAngularDamping=true
    bUseGravityOverride=true
    bUsePlanarLimit=false
    gravityScale=0.35
    alphaBoolBlend {blendOption=Linear}
    boundBone {boneName=dyn_hair_05}
    componentAppliedLinearAccClamp=<1000 1000 1000>
    componentLinearAccScale=<0.5 0.5 0.5>
    componentLinearVelScale=<0.5 0.5 0.5>
    gravityOverride {z=-200}
    physicsBodyDefinitions [{
      boundBone {boneName=dyn_hair_05}
      boxExtents=<25 30 30>
      constraintSetup {angularLimitsMax=<20 30 20>, angularLimitsMin=<-20 -30 -20>, angularTarget=<1 0 0>}
      localJointOffset=<7 0 -4>
     }]
   }
   n6 = Anim Dynamics - Bone: dyn_hair_06!(SimulationSpace:AnimPhysSimSpaceType Enum=Component, ComponentPose=@n5, Alpha=1, ExternalForce="0,0,0") -> (Pose) {
    angularDampingOverride=0.99
    angularSpringConstant=7500
    bAngularSpring=true
    bOverrideAngularDamping=true
    bUseGravityOverride=true
    bUsePlanarLimit=false
    gravityScale=0.35
    alphaBoolBlend {blendOption=Linear}
    boundBone {boneName=dyn_hair_06}
    componentAppliedLinearAccClamp=<1000 1000 1000>
    componentLinearAccScale=<0.5 0.5 0.5>
    componentLinearVelScale=<0.5 0.5 0.5>
    gravityOverride {z=-200}
    physicsBodyDefinitions [{
      boundBone {boneName=dyn_hair_06}
      boxExtents=<25 30 10>
      constraintSetup {angularLimitsMax=<20 30 20>, angularLimitsMin=<-20 -30 -20>, angularTarget=<1 0 0>}
      localJointOffset=<3 0 0>
     }]
   }
   n7 = Anim Dynamics - Bone: dyn_hair_07!(SimulationSpace:AnimPhysSimSpaceType Enum=Component, ComponentPose=@n6, Alpha=1, ExternalForce="0,0,0") -> (Pose) {
    angularDampingOverride=0.99
    angularSpringConstant=7500
    bAngularSpring=true
    bOverrideAngularDamping=true
    bUseGravityOverride=true
    bUsePlanarLimit=false
    gravityScale=0.35
    alphaBoolBlend {blendOption=Linear}
    boundBone {boneName=dyn_hair_07}
    componentAppliedLinearAccClamp=<1000 1000 1000>
    componentLinearAccScale=<0.5 0.5 0.5>
    componentLinearVelScale=<0.5 0.5 0.5>
    gravityOverride {z=-200}
    physicsBodyDefinitions [{
      boundBone {boneName=dyn_hair_07}
      boxExtents=<25 10 30>
      constraintSetup {angularLimitsMax=<20 30 20>, angularLimitsMin=<-20 -30 -20>, angularTarget=<1 0 0>}
      localJointOffset=<7 0 0>
     }]
   }
   n8 = Anim Dynamics - Bone: dyn_hair_08!(SimulationSpace:AnimPhysSimSpaceType Enum=Component, ComponentPose=@n7, Alpha=1, ExternalForce="0,0,0") -> (Pose) {
    angularDampingOverride=0.99
    angularSpringConstant=7500
    bAngularSpring=true
    bOverrideAngularDamping=true
    bUseGravityOverride=true
    bUsePlanarLimit=false
    gravityScale=0.35
    alphaBoolBlend {blendOption=Linear}
    boundBone {boneName=dyn_hair_08}
    componentAppliedLinearAccClamp=<1000 1000 1000>
    componentLinearAccScale=<0.5 0.5 0.5>
    componentLinearVelScale=<0.5 0.5 0.5>
    gravityOverride {z=-200}
    physicsBodyDefinitions [{
      boundBone {boneName=dyn_hair_08}
      boxExtents=<25 30 30>
      constraintSetup {angularLimitsMax=<20 30 20>, angularLimitsMin=<-20 -30 -20>, angularTarget=<1 0 0>}
      localJointOffset=<3 0 0>
     }]
   }
   n9 = Anim Dynamics - Bone: dyn_hair_09!(SimulationSpace:AnimPhysSimSpaceType Enum=Component, ComponentPose=@n8, Alpha=1, ExternalForce="0,0,0") -> (Pose) {
    angularDampingOverride=0.99
    angularSpringConstant=7500
    bAngularSpring=true
    bOverrideAngularDamping=true
    bUseGravityOverride=true
    bUsePlanarLimit=false
    gravityScale=0.35
    alphaBoolBlend {blendOption=Linear}
    boundBone {boneName=dyn_hair_09}
    componentAppliedLinearAccClamp=<1000 1000 1000>
    componentLinearAccScale=<0.5 0.5 0.5>
    componentLinearVelScale=<0.5 0.5 0.5>
    gravityOverride {z=-200}
    physicsBodyDefinitions [{
      boundBone {boneName=dyn_hair_09}
      boxExtents=<25 30 30>
      constraintSetup {angularLimitsMax=<20 30 20>, angularLimitsMin=<-20 -30 -20>, angularTarget=<1 0 0>}
      localJointOffset=<7 4 -4>
     }]
   }
   n10 = Anim Dynamics - Bone: dyn_hair_43!(SimulationSpace:AnimPhysSimSpaceType Enum=Component, ComponentPose=@n9, Alpha=1, ExternalForce="0,0,0") -> (Pose) {
    angularDampingOverride=0.99
    angularSpringConstant=7500
    bAngularSpring=true
    bOverrideAngularDamping=true
    bUseGravityOverride=true
    bUsePlanarLimit=false
    gravityScale=0.35
    alphaBoolBlend {blendOption=Linear}
    boundBone {boneName=dyn_hair_43}
    componentAppliedLinearAccClamp=<1000 1000 1000>
    componentLinearAccScale=<0.5 0.5 0.5>
    componentLinearVelScale=<0.5 0.5 0.5>
    gravityOverride {z=-200}
    physicsBodyDefinitions [{
      boundBone {boneName=dyn_hair_43}
      boxExtents=<25 30 30>
      constraintSetup {angularLimitsMax=<20 30 20>, angularLimitsMin=<-20 -30 -20>, angularTarget=<1 0 0>}
      localJointOffset=<7 -4 4>
     }]
   }
   n11 = Anim Dynamics - Bone: dyn_hair_44!(SimulationSpace:AnimPhysSimSpaceType Enum=Component, ComponentPose=@n10, Alpha=1, ExternalForce="0,0,0") -> (Pose) {
    angularDampingOverride=0.99
    angularSpringConstant=7500
    bAngularSpring=true
    bOverrideAngularDamping=true
    bUseGravityOverride=true
    bUsePlanarLimit=false
    gravityScale=0.35
    alphaBoolBlend {blendOption=Linear}
    boundBone {boneName=dyn_hair_44}
    componentAppliedLinearAccClamp=<1000 1000 1000>
    componentLinearAccScale=<0.5 0.5 0.5>
    componentLinearVelScale=<0.5 0.5 0.5>
    gravityOverride {z=-200}
    physicsBodyDefinitions [{
      boundBone {boneName=dyn_hair_44}
      boxExtents=<25 30 30>
      constraintSetup {angularLimitsMax=<20 30 20>, angularLimitsMin=<-20 -30 -20>, angularTarget=<1 0 0>}
      localJointOffset=<7 0 0>
     }]
   }
   n12 = Anim Dynamics - Bone: dyn_hair_45!(SimulationSpace:AnimPhysSimSpaceType Enum=Component, ComponentPose=@n11, Alpha=1, ExternalForce="0,0,0") -> (Pose) {
    angularDampingOverride=0.99
    angularSpringConstant=7500
    bAngularSpring=true
    bOverrideAngularDamping=true
    bUseGravityOverride=true
    bUsePlanarLimit=false
    gravityScale=0.35
    alphaBoolBlend {blendOption=Linear}
    boundBone {boneName=dyn_hair_45}
    componentAppliedLinearAccClamp=<1000 1000 1000>
    componentLinearAccScale=<0.5 0.5 0.5>
    componentLinearVelScale=<0.5 0.5 0.5>
    gravityOverride {z=-200}
    physicsBodyDefinitions [{
      boundBone {boneName=dyn_hair_45}
      boxExtents=<25 30 30>
      constraintSetup {angularLimitsMax=<20 30 20>, angularLimitsMin=<-20 -30 -20>, angularTarget=<1 0 0>}
      localJointOffset=<7 4 0>
     }]
   }
   n13 = Anim Dynamics - Bone: dyn_hair_46!(SimulationSpace:AnimPhysSimSpaceType Enum=Component, ComponentPose=@n12, Alpha=1, ExternalForce="0,0,0") -> (Pose) {
    angularDampingOverride=0.99
    angularSpringConstant=7500
    bAngularSpring=true
    bOverrideAngularDamping=true
    bUseGravityOverride=true
    bUsePlanarLimit=false
    gravityScale=0.35
    alphaBoolBlend {blendOption=Linear}
    boundBone {boneName=dyn_hair_46}
    componentAppliedLinearAccClamp=<1000 1000 1000>
    componentLinearAccScale=<0.5 0.5 0.5>
    componentLinearVelScale=<0.5 0.5 0.5>
    gravityOverride {z=-200}
    physicsBodyDefinitions [{
      boundBone {boneName=dyn_hair_46}
      boxExtents=<25 30 30>
      constraintSetup {angularLimitsMax=<20 30 20>, angularLimitsMin=<-20 -30 -20>, angularTarget=<1 0 0>}
      localJointOffset=<7 -4 0>
     }]
   }
   n14 = Anim Dynamics - Bone: dyn_hair_47!(SimulationSpace:AnimPhysSimSpaceType Enum=Component, ComponentPose=@n13, Alpha=1, ExternalForce="0,0,0") -> (Pose) {
    angularDampingOverride=0.99
    angularSpringConstant=7500
    bAngularSpring=true
    bOverrideAngularDamping=true
    bUseGravityOverride=true
    bUsePlanarLimit=false
    gravityScale=0.35
    alphaBoolBlend {blendOption=Linear}
    boundBone {boneName=dyn_hair_47}
    componentAppliedLinearAccClamp=<1000 1000 1000>
    componentLinearAccScale=<0.5 0.5 0.5>
    componentLinearVelScale=<0.5 0.5 0.5>
    gravityOverride {z=-200}
    physicsBodyDefinitions [{
      boundBone {boneName=dyn_hair_47}
      boxExtents=<25 30 30>
      constraintSetup {angularLimitsMax=<20 30 20>, angularLimitsMin=<-20 -30 -20>, angularTarget=<1 0 0>}
      localJointOffset=<7 0 4>
     }]
   }
   n15 = Anim Dynamics - Bone: dyn_hair_48!(SimulationSpace:AnimPhysSimSpaceType Enum=Component, ComponentPose=@n14, Alpha=1, ExternalForce="0,0,0") -> (Pose) {
    angularDampingOverride=0.99
    angularSpringConstant=7500
    bAngularSpring=true
    bOverrideAngularDamping=true
    bUseGravityOverride=true
    bUsePlanarLimit=false
    gravityScale=0.35
    alphaBoolBlend {blendOption=Linear}
    boundBone {boneName=dyn_hair_48}
    componentAppliedLinearAccClamp=<1000 1000 1000>
    componentLinearAccScale=<0.5 0.5 0.5>
    componentLinearVelScale=<0.5 0.5 0.5>
    gravityOverride {z=-200}
    physicsBodyDefinitions [{
      boundBone {boneName=dyn_hair_48}
      boxExtents=<25 30 30>
      constraintSetup {angularLimitsMax=<20 30 20>, angularLimitsMin=<-20 -30 -20>, angularTarget=<1 0 0>}
      localJointOffset=<7 0 -4>
     }]
   }
   n16 = Anim Dynamics - Bone: dyn_hair_49!(SimulationSpace:AnimPhysSimSpaceType Enum=Component, ComponentPose=@n15, Alpha=1, ExternalForce="0,0,0") -> (Pose) {
    angularDampingOverride=0.99
    angularSpringConstant=7500
    bAngularSpring=true
    bOverrideAngularDamping=true
    bUseGravityOverride=true
    bUsePlanarLimit=false
    gravityScale=0.35
    alphaBoolBlend {blendOption=Linear}
    boundBone {boneName=dyn_hair_49}
    componentAppliedLinearAccClamp=<1000 1000 1000>
    componentLinearAccScale=<0.5 0.5 0.5>
    componentLinearVelScale=<0.5 0.5 0.5>
    gravityOverride {z=-200}
    physicsBodyDefinitions [{
      boundBone {boneName=dyn_hair_49}
      boxExtents=<25 30 10>
      constraintSetup {angularLimitsMax=<20 30 20>, angularLimitsMin=<-20 -30 -20>, angularTarget=<1 0 0>}
      localJointOffset=<3 0 0>
     }]
   }
   n17 = Anim Dynamics - Bone: dyn_hair_50!(SimulationSpace:AnimPhysSimSpaceType Enum=Component, ComponentPose=@n16, Alpha=1, ExternalForce="0,0,0") -> (Pose) {
    angularDampingOverride=0.99
    angularSpringConstant=7500
    bAngularSpring=true
    bOverrideAngularDamping=true
    bUseGravityOverride=true
    bUsePlanarLimit=false
    gravityScale=0.35
    alphaBoolBlend {blendOption=Linear}
    boundBone {boneName=dyn_hair_50}
    componentAppliedLinearAccClamp=<1000 1000 1000>
    componentLinearAccScale=<0.5 0.5 0.5>
    componentLinearVelScale=<0.5 0.5 0.5>
    gravityOverride {z=-200}
    physicsBodyDefinitions [{
      boundBone {boneName=dyn_hair_50}
      boxExtents=<25 10 30>
      constraintSetup {angularLimitsMax=<20 30 20>, angularLimitsMin=<-20 -30 -20>, angularTarget=<1 0 0>}
      localJointOffset=<7 0 0>
     }]
   }
   n18 = Anim Dynamics - Bone: dyn_hair_51!(SimulationSpace:AnimPhysSimSpaceType Enum=Component, ComponentPose=@n17, Alpha=1, ExternalForce="0,0,0") -> (Pose) {
    angularDampingOverride=0.99
    angularSpringConstant=7500
    bAngularSpring=true
    bOverrideAngularDamping=true
    bUseGravityOverride=true
    bUsePlanarLimit=false
    gravityScale=0.35
    alphaBoolBlend {blendOption=Linear}
    boundBone {boneName=dyn_hair_51}
    componentAppliedLinearAccClamp=<1000 1000 1000>
    componentLinearAccScale=<0.5 0.5 0.5>
    componentLinearVelScale=<0.5 0.5 0.5>
    gravityOverride {z=-200}
    physicsBodyDefinitions [{
      boundBone {boneName=dyn_hair_51}
      boxExtents=<25 30 30>
      constraintSetup {angularLimitsMax=<20 30 20>, angularLimitsMin=<-20 -30 -20>, angularTarget=<1 0 0>}
      localJointOffset=<3 0 0>
     }]
   }
   n19 = Anim Dynamics - Bone: dyn_hair_52!(SimulationSpace:AnimPhysSimSpaceType Enum=Component, ComponentPose=@n18, Alpha=1, ExternalForce="0,0,0") -> (Pose) {
    angularDampingOverride=0.99
    angularSpringConstant=7500
    bAngularSpring=true
    bOverrideAngularDamping=true
    bUseGravityOverride=true
    bUsePlanarLimit=false
    gravityScale=0.35
    alphaBoolBlend {blendOption=Linear}
    boundBone {boneName=dyn_hair_52}
    componentAppliedLinearAccClamp=<1000 1000 1000>
    componentLinearAccScale=<0.5 0.5 0.5>
    componentLinearVelScale=<0.5 0.5 0.5>
    gravityOverride {z=-200}
    physicsBodyDefinitions [{
      boundBone {boneName=dyn_hair_52}
      boxExtents=<25 30 30>
      constraintSetup {angularLimitsMax=<20 30 20>, angularLimitsMin=<-20 -30 -20>, angularTarget=<1 0 0>}
      localJointOffset=<7 4 -4>
     }]
   }
   n20 = Anim Dynamics - Bone: dyn_hair_10!(SimulationSpace:AnimPhysSimSpaceType Enum=Component, ComponentPose=@n19, Alpha=1, ExternalForce="0,0,0") -> (Pose) {
    angularDampingOverride=0.99
    angularSpringConstant=7500
    bAngularSpring=true
    bOverrideAngularDamping=true
    bUseGravityOverride=true
    bUsePlanarLimit=false
    gravityScale=0.35
    alphaBoolBlend {blendOption=Linear}
    boundBone {boneName=dyn_hair_10}
    componentAppliedLinearAccClamp=<1000 1000 1000>
    componentLinearAccScale=<0.5 0.5 0.5>
    componentLinearVelScale=<0.5 0.5 0.5>
    gravityOverride {z=-200}
    physicsBodyDefinitions [{
      boundBone {boneName=dyn_hair_10}
      boxExtents=<25 30 30>
      constraintSetup {angularLimitsMax=<20 30 20>, angularLimitsMin=<-20 -30 -20>, angularTarget=<1 0 0>}
      localJointOffset=<7 -4 4>
     }]
   }
   n21 = Anim Dynamics - Bone: dyn_hair_11!(SimulationSpace:AnimPhysSimSpaceType Enum=Component, ComponentPose=@n20, Alpha=1, ExternalForce="0,0,0") -> (Pose) {
    angularDampingOverride=0.99
    angularSpringConstant=7500
    bAngularSpring=true
    bOverrideAngularDamping=true
    bUseGravityOverride=true
    bUsePlanarLimit=false
    gravityScale=0.35
    alphaBoolBlend {blendOption=Linear}
    boundBone {boneName=dyn_hair_11}
    componentAppliedLinearAccClamp=<1000 1000 1000>
    componentLinearAccScale=<0.5 0.5 0.5>
    componentLinearVelScale=<0.5 0.5 0.5>
    gravityOverride {z=-200}
    physicsBodyDefinitions [{
      boundBone {boneName=dyn_hair_11}
      boxExtents=<25 30 30>
      constraintSetup {angularLimitsMax=<20 30 20>, angularLimitsMin=<-20 -30 -20>, angularTarget=<1 0 0>}
      localJointOffset=<7 0 0>
     }]
   }
   n22 = Anim Dynamics - Bone: dyn_hair_12!(SimulationSpace:AnimPhysSimSpaceType Enum=Component, ComponentPose=@n21, Alpha=1, ExternalForce="0,0,0") -> (Pose) {
    angularDampingOverride=0.99
    angularSpringConstant=7500
    bAngularSpring=true
    bOverrideAngularDamping=true
    bUseGravityOverride=true
    bUsePlanarLimit=false
    gravityScale=0.35
    alphaBoolBlend {blendOption=Linear}
    boundBone {boneName=dyn_hair_12}
    componentAppliedLinearAccClamp=<1000 1000 1000>
    componentLinearAccScale=<0.5 0.5 0.5>
    componentLinearVelScale=<0.5 0.5 0.5>
    gravityOverride {z=-200}
    physicsBodyDefinitions [{
      boundBone {boneName=dyn_hair_12}
      boxExtents=<25 30 30>
      constraintSetup {angularLimitsMax=<20 30 20>, angularLimitsMin=<-20 -30 -20>, angularTarget=<1 0 0>}
      localJointOffset=<7 4 0>
     }]
   }
   n23 = Anim Dynamics - Bone: dyn_hair_13!(SimulationSpace:AnimPhysSimSpaceType Enum=Component, ComponentPose=@n22, Alpha=1, ExternalForce="0,0,0") -> (Pose) {
    angularDampingOverride=0.99
    angularSpringConstant=7500
    bAngularSpring=true
    bOverrideAngularDamping=true
    bUseGravityOverride=true
    bUsePlanarLimit=false
    gravityScale=0.35
    alphaBoolBlend {blendOption=Linear}
    boundBone {boneName=dyn_hair_13}
    componentAppliedLinearAccClamp=<1000 1000 1000>
    componentLinearAccScale=<0.5 0.5 0.5>
    componentLinearVelScale=<0.5 0.5 0.5>
    gravityOverride {z=-200}
    physicsBodyDefinitions [{
      boundBone {boneName=dyn_hair_13}
      boxExtents=<25 30 30>
      constraintSetup {angularLimitsMax=<20 30 20>, angularLimitsMin=<-20 -30 -20>, angularTarget=<1 0 0>}
      localJointOffset=<7 -4 0>
     }]
   }
   n24 = Anim Dynamics - Bone: dyn_hair_14!(SimulationSpace:AnimPhysSimSpaceType Enum=Component, ComponentPose=@n23, Alpha=1, ExternalForce="0,0,0") -> (Pose) {
    angularDampingOverride=0.99
    angularSpringConstant=7500
    bAngularSpring=true
    bOverrideAngularDamping=true
    bUseGravityOverride=true
    bUsePlanarLimit=false
    gravityScale=0.35
    alphaBoolBlend {blendOption=Linear}
    boundBone {boneName=dyn_hair_14}
    componentAppliedLinearAccClamp=<1000 1000 1000>
    componentLinearAccScale=<0.5 0.5 0.5>
    componentLinearVelScale=<0.5 0.5 0.5>
    gravityOverride {z=-200}
    physicsBodyDefinitions [{
      boundBone {boneName=dyn_hair_14}
      boxExtents=<25 30 30>
      constraintSetup {angularLimitsMax=<20 30 20>, angularLimitsMin=<-20 -30 -20>, angularTarget=<1 0 0>}
      localJointOffset=<7 0 4>
     }]
   }
   n25 = Anim Dynamics - Bone: dyn_hair_15!(SimulationSpace:AnimPhysSimSpaceType Enum=Component, ComponentPose=@n24, Alpha=1, ExternalForce="0,0,0") -> (Pose) {
    angularDampingOverride=0.99
    angularSpringConstant=7500
    bAngularSpring=true
    bOverrideAngularDamping=true
    bUseGravityOverride=true
    bUsePlanarLimit=false
    gravityScale=0.35
    alphaBoolBlend {blendOption=Linear}
    boundBone {boneName=dyn_hair_15}
    componentAppliedLinearAccClamp=<1000 1000 1000>
    componentLinearAccScale=<0.5 0.5 0.5>
    componentLinearVelScale=<0.5 0.5 0.5>
    gravityOverride {z=-200}
    physicsBodyDefinitions [{
      boundBone {boneName=dyn_hair_15}
      boxExtents=<25 30 30>
      constraintSetup {angularLimitsMax=<20 30 20>, angularLimitsMin=<-20 -30 -20>, angularTarget=<1 0 0>}
      localJointOffset=<7 0 -4>
     }]
   }
   n26 = Anim Dynamics - Bone: dyn_hair_16!(SimulationSpace:AnimPhysSimSpaceType Enum=Component, ComponentPose=@n25, Alpha=1, ExternalForce="0,0,0") -> (Pose) {
    angularDampingOverride=0.99
    angularSpringConstant=7500
    bAngularSpring=true
    bOverrideAngularDamping=true
    bUseGravityOverride=true
    bUsePlanarLimit=false
    gravityScale=0.35
    alphaBoolBlend {blendOption=Linear}
    boundBone {boneName=dyn_hair_16}
    componentAppliedLinearAccClamp=<1000 1000 1000>
    componentLinearAccScale=<0.5 0.5 0.5>
    componentLinearVelScale=<0.5 0.5 0.5>
    gravityOverride {z=-200}
    physicsBodyDefinitions [{
      boundBone {boneName=dyn_hair_16}
      boxExtents=<25 30 10>
      constraintSetup {angularLimitsMax=<20 30 20>, angularLimitsMin=<-20 -30 -20>, angularTarget=<1 0 0>}
      localJointOffset=<3 0 0>
     }]
   }
   n27 = Anim Dynamics - Bone: dyn_hair_17!(SimulationSpace:AnimPhysSimSpaceType Enum=Component, ComponentPose=@n26, Alpha=1, ExternalForce="0,0,0") -> (Pose) {
    angularDampingOverride=0.99
    angularSpringConstant=7500
    bAngularSpring=true
    bOverrideAngularDamping=true
    bUseGravityOverride=true
    bUsePlanarLimit=false
    gravityScale=0.35
    alphaBoolBlend {blendOption=Linear}
    boundBone {boneName=dyn_hair_17}
    componentAppliedLinearAccClamp=<1000 1000 1000>
    componentLinearAccScale=<0.5 0.5 0.5>
    componentLinearVelScale=<0.5 0.5 0.5>
    gravityOverride {z=-200}
    physicsBodyDefinitions [{
      boundBone {boneName=dyn_hair_17}
      boxExtents=<25 10 30>
      constraintSetup {angularLimitsMax=<20 30 20>, angularLimitsMin=<-20 -30 -20>, angularTarget=<1 0 0>}
      localJointOffset=<7 0 0>
     }]
   }
   n28 = Anim Dynamics - Bone: dyn_hair_18!(SimulationSpace:AnimPhysSimSpaceType Enum=Component, ComponentPose=@n27, Alpha=1, ExternalForce="0,0,0") -> (Pose) {
    angularDampingOverride=0.99
    angularSpringConstant=7500
    bAngularSpring=true
    bOverrideAngularDamping=true
    bUseGravityOverride=true
    bUsePlanarLimit=false
    gravityScale=0.35
    alphaBoolBlend {blendOption=Linear}
    boundBone {boneName=dyn_hair_18}
    componentAppliedLinearAccClamp=<1000 1000 1000>
    componentLinearAccScale=<0.5 0.5 0.5>
    componentLinearVelScale=<0.5 0.5 0.5>
    gravityOverride {z=-200}
    physicsBodyDefinitions [{
      boundBone {boneName=dyn_hair_18}
      boxExtents=<25 30 30>
      constraintSetup {angularLimitsMax=<20 30 20>, angularLimitsMin=<-20 -30 -20>, angularTarget=<1 0 0>}
      localJointOffset=<3 0 0>
     }]
   }
   n29 = Anim Dynamics - Bone: dyn_hair_19!(SimulationSpace:AnimPhysSimSpaceType Enum=Component, ComponentPose=@n28, Alpha=1, ExternalForce="0,0,0") -> (Pose) {
    angularDampingOverride=0.99
    angularSpringConstant=7500
    bAngularSpring=true
    bOverrideAngularDamping=true
    bUseGravityOverride=true
    bUsePlanarLimit=false
    gravityScale=0.35
    alphaBoolBlend {blendOption=Linear}
    boundBone {boneName=dyn_hair_19}
    componentAppliedLinearAccClamp=<1000 1000 1000>
    componentLinearAccScale=<0.5 0.5 0.5>
    componentLinearVelScale=<0.5 0.5 0.5>
    gravityOverride {z=-200}
    physicsBodyDefinitions [{
      boundBone {boneName=dyn_hair_19}
      boxExtents=<25 30 30>
      constraintSetup {angularLimitsMax=<20 30 20>, angularLimitsMin=<-20 -30 -20>, angularTarget=<1 0 0>}
      localJointOffset=<7 4 -4>
     }]
   }
   n30 = Anim Dynamics - Bone: dyn_hair_20!(SimulationSpace:AnimPhysSimSpaceType Enum=Component, ComponentPose=@n29, Alpha=1, ExternalForce="0,0,0") -> (Pose) {
    angularDampingOverride=0.99
    angularSpringConstant=7500
    bAngularSpring=true
    bOverrideAngularDamping=true
    bUseGravityOverride=true
    bUsePlanarLimit=false
    gravityScale=0.35
    alphaBoolBlend {blendOption=Linear}
    boundBone {boneName=dyn_hair_20}
    componentAppliedLinearAccClamp=<1000 1000 1000>
    componentLinearAccScale=<0.5 0.5 0.5>
    componentLinearVelScale=<0.5 0.5 0.5>
    gravityOverride {z=-200}
    physicsBodyDefinitions [{
      boundBone {boneName=dyn_hair_20}
      boxExtents=<25 30 30>
      constraintSetup {angularLimitsMax=<20 30 20>, angularLimitsMin=<-20 -30 -20>, angularTarget=<1 0 0>}
      localJointOffset=<7 -4 4>
     }]
   }
   n31 = Anim Dynamics - Bone: dyn_hair_21!(SimulationSpace:AnimPhysSimSpaceType Enum=Component, ComponentPose=@n30, Alpha=1, ExternalForce="0,0,0") -> (Pose) {
    angularDampingOverride=0.99
    angularSpringConstant=7500
    bAngularSpring=true
    bOverrideAngularDamping=true
    bUseGravityOverride=true
    bUsePlanarLimit=false
    gravityScale=0.35
    alphaBoolBlend {blendOption=Linear}
    boundBone {boneName=dyn_hair_21}
    componentAppliedLinearAccClamp=<1000 1000 1000>
    componentLinearAccScale=<0.5 0.5 0.5>
    componentLinearVelScale=<0.5 0.5 0.5>
    gravityOverride {z=-200}
    physicsBodyDefinitions [{
      boundBone {boneName=dyn_hair_21}
      boxExtents=<25 30 30>
      constraintSetup {angularLimitsMax=<20 30 20>, angularLimitsMin=<-20 -30 -20>, angularTarget=<1 0 0>}
      localJointOffset=<7 0 0>
     }]
   }
   n32 = Anim Dynamics - Bone: dyn_hair_22!(SimulationSpace:AnimPhysSimSpaceType Enum=Component, ComponentPose=@n31, Alpha=1, ExternalForce="0,0,0") -> (Pose) {
    angularDampingOverride=0.99
    angularSpringConstant=7500
    bAngularSpring=true
    bOverrideAngularDamping=true
    bUseGravityOverride=true
    bUsePlanarLimit=false
    gravityScale=0.35
    alphaBoolBlend {blendOption=Linear}
    boundBone {boneName=dyn_hair_22}
    componentAppliedLinearAccClamp=<1000 1000 1000>
    componentLinearAccScale=<0.5 0.5 0.5>
    componentLinearVelScale=<0.5 0.5 0.5>
    gravityOverride {z=-200}
    physicsBodyDefinitions [{
      boundBone {boneName=dyn_hair_22}
      boxExtents=<25 30 30>
      constraintSetup {angularLimitsMax=<20 30 20>, angularLimitsMin=<-20 -30 -20>, angularTarget=<1 0 0>}
      localJointOffset=<7 4 0>
     }]
   }
   n33 = Anim Dynamics - Bone: dyn_hair_23!(SimulationSpace:AnimPhysSimSpaceType Enum=Component, ComponentPose=@n32, Alpha=1, ExternalForce="0,0,0") -> (Pose) {
    angularDampingOverride=0.99
    angularSpringConstant=7500
    bAngularSpring=true
    bOverrideAngularDamping=true
    bUseGravityOverride=true
    bUsePlanarLimit=false
    gravityScale=0.35
    alphaBoolBlend {blendOption=Linear}
    boundBone {boneName=dyn_hair_23}
    componentAppliedLinearAccClamp=<1000 1000 1000>
    componentLinearAccScale=<0.5 0.5 0.5>
    componentLinearVelScale=<0.5 0.5 0.5>
    gravityOverride {z=-200}
    physicsBodyDefinitions [{
      boundBone {boneName=dyn_hair_23}
      boxExtents=<25 30 30>
      constraintSetup {angularLimitsMax=<20 30 20>, angularLimitsMin=<-20 -30 -20>, angularTarget=<1 0 0>}
      localJointOffset=<7 -4 0>
     }]
   }
   n34 = Anim Dynamics - Bone: dyn_hair_24!(SimulationSpace:AnimPhysSimSpaceType Enum=Component, ComponentPose=@n33, Alpha=1, ExternalForce="0,0,0") -> (Pose) {
    angularDampingOverride=0.99
    angularSpringConstant=7500
    bAngularSpring=true
    bOverrideAngularDamping=true
    bUseGravityOverride=true
    bUsePlanarLimit=false
    gravityScale=0.35
    alphaBoolBlend {blendOption=Linear}
    boundBone {boneName=dyn_hair_24}
    componentAppliedLinearAccClamp=<1000 1000 1000>
    componentLinearAccScale=<0.5 0.5 0.5>
    componentLinearVelScale=<0.5 0.5 0.5>
    gravityOverride {z=-200}
    physicsBodyDefinitions [{
      boundBone {boneName=dyn_hair_24}
      boxExtents=<25 30 30>
      constraintSetup {angularLimitsMax=<20 30 20>, angularLimitsMin=<-20 -30 -20>, angularTarget=<1 0 0>}
      localJointOffset=<7 0 4>
     }]
   }
   n35 = Anim Dynamics - Bone: dyn_hair_25!(SimulationSpace:AnimPhysSimSpaceType Enum=Component, ComponentPose=@n34, Alpha=1, ExternalForce="0,0,0") -> (Pose) {
    angularDampingOverride=0.99
    angularSpringConstant=7500
    bAngularSpring=true
    bOverrideAngularDamping=true
    bUseGravityOverride=true
    bUsePlanarLimit=false
    gravityScale=0.35
    alphaBoolBlend {blendOption=Linear}
    boundBone {boneName=dyn_hair_25}
    componentAppliedLinearAccClamp=<1000 1000 1000>
    componentLinearAccScale=<0.5 0.5 0.5>
    componentLinearVelScale=<0.5 0.5 0.5>
    gravityOverride {z=-200}
    physicsBodyDefinitions [{
      boundBone {boneName=dyn_hair_25}
      boxExtents=<25 30 30>
      constraintSetup {angularLimitsMax=<20 30 20>, angularLimitsMin=<-20 -30 -20>, angularTarget=<1 0 0>}
      localJointOffset=<7 0 -4>
     }]
   }
   n36 = Anim Dynamics - Bone: dyn_hair_26!(SimulationSpace:AnimPhysSimSpaceType Enum=Component, ComponentPose=@n35, Alpha=1, ExternalForce="0,0,0") -> (Pose) {
    angularDampingOverride=0.99
    angularSpringConstant=7500
    bAngularSpring=true
    bOverrideAngularDamping=true
    bUseGravityOverride=true
    bUsePlanarLimit=false
    gravityScale=0.35
    alphaBoolBlend {blendOption=Linear}
    boundBone {boneName=dyn_hair_26}
    componentAppliedLinearAccClamp=<1000 1000 1000>
    componentLinearAccScale=<0.5 0.5 0.5>
    componentLinearVelScale=<0.5 0.5 0.5>
    gravityOverride {z=-200}
    physicsBodyDefinitions [{
      boundBone {boneName=dyn_hair_26}
      boxExtents=<25 30 10>
      constraintSetup {angularLimitsMax=<20 30 20>, angularLimitsMin=<-20 -30 -20>, angularTarget=<1 0 0>}
      localJointOffset=<3 0 0>
     }]
   }
   n37 = Anim Dynamics - Bone: dyn_hair_27!(SimulationSpace:AnimPhysSimSpaceType Enum=Component, ComponentPose=@n36, Alpha=1, ExternalForce="0,0,0") -> (Pose) {
    angularDampingOverride=0.99
    angularSpringConstant=7500
    bAngularSpring=true
    bOverrideAngularDamping=true
    bUseGravityOverride=true
    bUsePlanarLimit=false
    gravityScale=0.35
    alphaBoolBlend {blendOption=Linear}
    boundBone {boneName=dyn_hair_27}
    componentAppliedLinearAccClamp=<1000 1000 1000>
    componentLinearAccScale=<0.5 0.5 0.5>
    componentLinearVelScale=<0.5 0.5 0.5>
    gravityOverride {z=-200}
    physicsBodyDefinitions [{
      boundBone {boneName=dyn_hair_27}
      boxExtents=<25 10 30>
      constraintSetup {angularLimitsMax=<20 30 20>, angularLimitsMin=<-20 -30 -20>, angularTarget=<1 0 0>}
      localJointOffset=<7 0 0>
     }]
   }
   n38 = Anim Dynamics - Bone: dyn_hair_28!(SimulationSpace:AnimPhysSimSpaceType Enum=Component, ComponentPose=@n37, Alpha=1, ExternalForce="0,0,0") -> (Pose) {
    angularDampingOverride=0.99
    angularSpringConstant=7500
    bAngularSpring=true
    bOverrideAngularDamping=true
    bUseGravityOverride=true
    bUsePlanarLimit=false
    gravityScale=0.35
    alphaBoolBlend {blendOption=Linear}
    boundBone {boneName=dyn_hair_28}
    componentAppliedLinearAccClamp=<1000 1000 1000>
    componentLinearAccScale=<0.5 0.5 0.5>
    componentLinearVelScale=<0.5 0.5 0.5>
    gravityOverride {z=-200}
    physicsBodyDefinitions [{
      boundBone {boneName=dyn_hair_28}
      boxExtents=<25 30 30>
      constraintSetup {angularLimitsMax=<20 30 20>, angularLimitsMin=<-20 -30 -20>, angularTarget=<1 0 0>}
      localJointOffset=<3 0 0>
     }]
   }
   n39 = Anim Dynamics - Bone: dyn_hair_29!(SimulationSpace:AnimPhysSimSpaceType Enum=Component, ComponentPose=@n38, Alpha=1, ExternalForce="0,0,0") -> (Pose) {
    angularDampingOverride=0.99
    angularSpringConstant=7500
    bAngularSpring=true
    bOverrideAngularDamping=true
    bUseGravityOverride=true
    bUsePlanarLimit=false
    gravityScale=0.35
    alphaBoolBlend {blendOption=Linear}
    boundBone {boneName=dyn_hair_29}
    componentAppliedLinearAccClamp=<1000 1000 1000>
    componentLinearAccScale=<0.5 0.5 0.5>
    componentLinearVelScale=<0.5 0.5 0.5>
    gravityOverride {z=-200}
    physicsBodyDefinitions [{
      boundBone {boneName=dyn_hair_29}
      boxExtents=<25 30 30>
      constraintSetup {angularLimitsMax=<20 30 20>, angularLimitsMin=<-20 -30 -20>, angularTarget=<1 0 0>}
      localJointOffset=<7 4 -4>
     }]
   }
   n40 = Anim Dynamics - Bone: dyn_hair_30!(SimulationSpace:AnimPhysSimSpaceType Enum=Component, ComponentPose=@n39, Alpha=1, ExternalForce="0,0,0") -> (Pose) {
    angularDampingOverride=0.99
    angularSpringConstant=7500
    bAngularSpring=true
    bOverrideAngularDamping=true
    bUseGravityOverride=true
    bUsePlanarLimit=false
    gravityScale=0.35
    alphaBoolBlend {blendOption=Linear}
    boundBone {boneName=dyn_hair_30}
    componentAppliedLinearAccClamp=<1000 1000 1000>
    componentLinearAccScale=<0.5 0.5 0.5>
    componentLinearVelScale=<0.5 0.5 0.5>
    gravityOverride {z=-200}
    physicsBodyDefinitions [{
      boundBone {boneName=dyn_hair_30}
      boxExtents=<25 30 30>
      constraintSetup {angularLimitsMax=<20 30 20>, angularLimitsMin=<-20 -30 -20>, angularTarget=<1 0 0>}
      localJointOffset=<7 -4 4>
     }]
   }
   n41 = Anim Dynamics - Bone: dyn_hair_31!(SimulationSpace:AnimPhysSimSpaceType Enum=Component, ComponentPose=@n40, Alpha=1, ExternalForce="0,0,0") -> (Pose) {
    angularDampingOverride=0.99
    angularSpringConstant=7500
    bAngularSpring=true
    bOverrideAngularDamping=true
    bUseGravityOverride=true
    bUsePlanarLimit=false
    gravityScale=0.35
    alphaBoolBlend {blendOption=Linear}
    boundBone {boneName=dyn_hair_31}
    componentAppliedLinearAccClamp=<1000 1000 1000>
    componentLinearAccScale=<0.5 0.5 0.5>
    componentLinearVelScale=<0.5 0.5 0.5>
    gravityOverride {z=-200}
    physicsBodyDefinitions [{
      boundBone {boneName=dyn_hair_31}
      boxExtents=<25 30 30>
      constraintSetup {angularLimitsMax=<20 30 20>, angularLimitsMin=<-20 -30 -20>, angularTarget=<1 0 0>}
      localJointOffset=<7 0 0>
     }]
   }
   n42 = Anim Dynamics - Bone: dyn_hair_32!(SimulationSpace:AnimPhysSimSpaceType Enum=Component, ComponentPose=@n41, Alpha=1, ExternalForce="0,0,0") -> (Pose) {
    angularDampingOverride=0.99
    angularSpringConstant=7500
    bAngularSpring=true
    bOverrideAngularDamping=true
    bUseGravityOverride=true
    bUsePlanarLimit=false
    gravityScale=0.35
    alphaBoolBlend {blendOption=Linear}
    boundBone {boneName=dyn_hair_32}
    componentAppliedLinearAccClamp=<1000 1000 1000>
    componentLinearAccScale=<0.5 0.5 0.5>
    componentLinearVelScale=<0.5 0.5 0.5>
    gravityOverride {z=-200}
    physicsBodyDefinitions [{
      boundBone {boneName=dyn_hair_32}
      boxExtents=<25 30 30>
      constraintSetup {angularLimitsMax=<20 30 20>, angularLimitsMin=<-20 -30 -20>, angularTarget=<1 0 0>}
      localJointOffset=<7 4 0>
     }]
   }
   n43 = Anim Dynamics - Bone: dyn_hair_33!(SimulationSpace:AnimPhysSimSpaceType Enum=Component, ComponentPose=@n42, Alpha=1, ExternalForce="0,0,0") -> (Pose) {
    angularDampingOverride=0.99
    angularSpringConstant=7500
    bAngularSpring=true
    bOverrideAngularDamping=true
    bUseGravityOverride=true
    bUsePlanarLimit=false
    gravityScale=0.35
    alphaBoolBlend {blendOption=Linear}
    boundBone {boneName=dyn_hair_33}
    componentAppliedLinearAccClamp=<1000 1000 1000>
    componentLinearAccScale=<0.5 0.5 0.5>
    componentLinearVelScale=<0.5 0.5 0.5>
    gravityOverride {z=-200}
    physicsBodyDefinitions [{
      boundBone {boneName=dyn_hair_33}
      boxExtents=<25 30 30>
      constraintSetup {angularLimitsMax=<20 30 20>, angularLimitsMin=<-20 -30 -20>, angularTarget=<1 0 0>}
      localJointOffset=<7 -4 0>
     }]
   }
   n44 = Anim Dynamics - Bone: dyn_hair_34!(SimulationSpace:AnimPhysSimSpaceType Enum=Component, ComponentPose=@n43, Alpha=1, ExternalForce="0,0,0") -> (Pose) {
    angularDampingOverride=0.99
    angularSpringConstant=7500
    bAngularSpring=true
    bOverrideAngularDamping=true
    bUseGravityOverride=true
    bUsePlanarLimit=false
    gravityScale=0.35
    alphaBoolBlend {blendOption=Linear}
    boundBone {boneName=dyn_hair_34}
    componentAppliedLinearAccClamp=<1000 1000 1000>
    componentLinearAccScale=<0.5 0.5 0.5>
    componentLinearVelScale=<0.5 0.5 0.5>
    gravityOverride {z=-200}
    physicsBodyDefinitions [{
      boundBone {boneName=dyn_hair_34}
      boxExtents=<25 30 30>
      constraintSetup {angularLimitsMax=<20 30 20>, angularLimitsMin=<-20 -30 -20>, angularTarget=<1 0 0>}
      localJointOffset=<7 0 4>
     }]
   }
   n45 = Anim Dynamics - Bone: dyn_hair_35!(SimulationSpace:AnimPhysSimSpaceType Enum=Component, ComponentPose=@n44, Alpha=1, ExternalForce="0,0,0") -> (Pose) {
    angularDampingOverride=0.99
    angularSpringConstant=7500
    bAngularSpring=true
    bOverrideAngularDamping=true
    bUseGravityOverride=true
    bUsePlanarLimit=false
    gravityScale=0.35
    alphaBoolBlend {blendOption=Linear}
    boundBone {boneName=dyn_hair_35}
    componentAppliedLinearAccClamp=<1000 1000 1000>
    componentLinearAccScale=<0.5 0.5 0.5>
    componentLinearVelScale=<0.5 0.5 0.5>
    gravityOverride {z=-200}
    physicsBodyDefinitions [{
      boundBone {boneName=dyn_hair_35}
      boxExtents=<25 30 30>
      constraintSetup {angularLimitsMax=<20 30 20>, angularLimitsMin=<-20 -30 -20>, angularTarget=<1 0 0>}
      localJointOffset=<7 0 -4>
     }]
   }
   n46 = Anim Dynamics - Bone: dyn_hair_36!(SimulationSpace:AnimPhysSimSpaceType Enum=Component, ComponentPose=@n45, Alpha=1, ExternalForce="0,0,0") -> (Pose) {
    angularDampingOverride=0.99
    angularSpringConstant=7500
    bAngularSpring=true
    bOverrideAngularDamping=true
    bUseGravityOverride=true
    bUsePlanarLimit=false
    gravityScale=0.35
    alphaBoolBlend {blendOption=Linear}
    boundBone {boneName=dyn_hair_36}
    componentAppliedLinearAccClamp=<1000 1000 1000>
    componentLinearAccScale=<0.5 0.5 0.5>
    componentLinearVelScale=<0.5 0.5 0.5>
    gravityOverride {z=-200}
    physicsBodyDefinitions [{
      boundBone {boneName=dyn_hair_36}
      boxExtents=<25 30 10>
      constraintSetup {angularLimitsMax=<20 30 20>, angularLimitsMin=<-20 -30 -20>, angularTarget=<1 0 0>}
      localJointOffset=<3 0 0>
     }]
   }
   n47 = Anim Dynamics - Bone: dyn_hair_37!(SimulationSpace:AnimPhysSimSpaceType Enum=Component, ComponentPose=@n46, Alpha=1, ExternalForce="0,0,0") -> (Pose) {
    angularDampingOverride=0.99
    angularSpringConstant=7500
    bAngularSpring=true
    bOverrideAngularDamping=true
    bUseGravityOverride=true
    bUsePlanarLimit=false
    gravityScale=0.35
    alphaBoolBlend {blendOption=Linear}
    boundBone {boneName=dyn_hair_37}
    componentAppliedLinearAccClamp=<1000 1000 1000>
    componentLinearAccScale=<0.5 0.5 0.5>
    componentLinearVelScale=<0.5 0.5 0.5>
    gravityOverride {z=-200}
    physicsBodyDefinitions [{
      boundBone {boneName=dyn_hair_37}
      boxExtents=<25 10 30>
      constraintSetup {angularLimitsMax=<20 30 20>, angularLimitsMin=<-20 -30 -20>, angularTarget=<1 0 0>}
      localJointOffset=<7 0 0>
     }]
   }
   n48 = Anim Dynamics - Bone: dyn_hair_38!(SimulationSpace:AnimPhysSimSpaceType Enum=Component, ComponentPose=@n47, Alpha=1, ExternalForce="0,0,0") -> (Pose) {
    angularDampingOverride=0.99
    angularSpringConstant=7500
    bAngularSpring=true
    bOverrideAngularDamping=true
    bUseGravityOverride=true
    bUsePlanarLimit=false
    gravityScale=0.35
    alphaBoolBlend {blendOption=Linear}
    boundBone {boneName=dyn_hair_38}
    componentAppliedLinearAccClamp=<1000 1000 1000>
    componentLinearAccScale=<0.5 0.5 0.5>
    componentLinearVelScale=<0.5 0.5 0.5>
    gravityOverride {z=-200}
    physicsBodyDefinitions [{
      boundBone {boneName=dyn_hair_38}
      boxExtents=<25 30 30>
      constraintSetup {angularLimitsMax=<20 30 20>, angularLimitsMin=<-20 -30 -20>, angularTarget=<1 0 0>}
      localJointOffset=<3 0 0>
     }]
   }
   n49 = Anim Dynamics - Bone: dyn_hair_39!(SimulationSpace:AnimPhysSimSpaceType Enum=Component, ComponentPose=@n48, Alpha=1, ExternalForce="0,0,0") -> (Pose) {
    angularDampingOverride=0.99
    angularSpringConstant=7500
    bAngularSpring=true
    bOverrideAngularDamping=true
    bUseGravityOverride=true
    bUsePlanarLimit=false
    gravityScale=0.35
    alphaBoolBlend {blendOption=Linear}
    boundBone {boneName=dyn_hair_39}
    componentAppliedLinearAccClamp=<1000 1000 1000>
    componentLinearAccScale=<0.5 0.5 0.5>
    componentLinearVelScale=<0.5 0.5 0.5>
    gravityOverride {z=-200}
    physicsBodyDefinitions [{
      boundBone {boneName=dyn_hair_39}
      boxExtents=<25 30 30>
      constraintSetup {angularLimitsMax=<20 30 20>, angularLimitsMin=<-20 -30 -20>, angularTarget=<1 0 0>}
      localJointOffset=<7 4 -4>
     }]
   }
   n50 = Anim Dynamics - Bone: dyn_hair_40!(SimulationSpace:AnimPhysSimSpaceType Enum=Component, ComponentPose=@n49, Alpha=1, ExternalForce="0,0,0") -> (Pose) {
    angularDampingOverride=0.99
    angularSpringConstant=7500
    bAngularSpring=true
    bOverrideAngularDamping=true
    bUseGravityOverride=true
    bUsePlanarLimit=false
    gravityScale=0.35
    alphaBoolBlend {blendOption=Linear}
    boundBone {boneName=dyn_hair_40}
    componentAppliedLinearAccClamp=<1000 1000 1000>
    componentLinearAccScale=<0.5 0.5 0.5>
    componentLinearVelScale=<0.5 0.5 0.5>
    gravityOverride {z=-200}
    physicsBodyDefinitions [{
      boundBone {boneName=dyn_hair_40}
      boxExtents=<25 30 30>
      constraintSetup {angularLimitsMax=<20 30 20>, angularLimitsMin=<-20 -30 -20>, angularTarget=<1 0 0>}
      localJointOffset=<7 -4 4>
     }]
   }
   n51 = Anim Dynamics - Bone: dyn_hair_41!(SimulationSpace:AnimPhysSimSpaceType Enum=Component, ComponentPose=@n50, Alpha=1, ExternalForce="0,0,0") -> (Pose) {
    angularDampingOverride=0.99
    angularSpringConstant=7500
    bAngularSpring=true
    bOverrideAngularDamping=true
    bUseGravityOverride=true
    bUsePlanarLimit=false
    gravityScale=0.35
    alphaBoolBlend {blendOption=Linear}
    boundBone {boneName=dyn_hair_41}
    componentAppliedLinearAccClamp=<1000 1000 1000>
    componentLinearAccScale=<0.5 0.5 0.5>
    componentLinearVelScale=<0.5 0.5 0.5>
    gravityOverride {z=-200}
    physicsBodyDefinitions [{
      boundBone {boneName=dyn_hair_41}
      boxExtents=<25 30 30>
      constraintSetup {angularLimitsMax=<20 30 20>, angularLimitsMin=<-20 -30 -20>, angularTarget=<1 0 0>}
      localJointOffset=<7 0 0>
     }]
   }
   n52 = Anim Dynamics - Bone: dyn_hair_42!(SimulationSpace:AnimPhysSimSpaceType Enum=Component, ComponentPose=@n51, Alpha=1, ExternalForce="0,0,0") -> (Pose) {
    angularDampingOverride=0.99
    angularSpringConstant=7500
    bAngularSpring=true
    bOverrideAngularDamping=true
    bUseGravityOverride=true
    bUsePlanarLimit=false
    gravityScale=0.35
    alphaBoolBlend {blendOption=Linear}
    boundBone {boneName=dyn_hair_42}
    componentAppliedLinearAccClamp=<1000 1000 1000>
    componentLinearAccScale=<0.5 0.5 0.5>
    componentLinearVelScale=<0.5 0.5 0.5>
    gravityOverride {z=-200}
    physicsBodyDefinitions [{
      boundBone {boneName=dyn_hair_42}
      boxExtents=<25 30 30>
      constraintSetup {angularLimitsMax=<20 30 20>, angularLimitsMin=<-20 -30 -20>, angularTarget=<1 0 0>}
      localJointOffset=<7 4 0>
     }]
   }
   Output Pose(Result=Component To Local(ComponentPose=@n52)) {layerGroup=None}
  ]}
 EventGraph {nodes [
   Event BlueprintUpdateAnimation()
   > n1 = Branch(Condition=Get bIsInFrontEnd())
   > Set Simulation Space(Simulation Space:AnimPhysSimSpaceType Enum=World) -> (Output_Get:AnimPhysSimSpaceType Enum)
   > Set Alpha(Alpha:float=0.7) -> (Output_Get:float)
   Set Simulation Space(execute=@n1.else, Simulation Space=Component)
   > Set Alpha(Alpha=0.9)
  ]}
]
variables [
 Alpha {type=number, default=0.9}
 Simulation Space {type=AnimPhysSimSpaceType}
]
